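# Animation level sequence for cinematic spa1000, staging a tentacle boss scene with three tentacle instances and a superhero NPC (KalEl).
# Visibility keys, chained skeletal animations and transforms drive the tentacles, with left and right raised in Z late in the shot.
# The front tentacle hides as a separate TentacleCut actor appears, swapping in a severed version.
# It also hides the main boss, toggles KalEl, places a scaled no-fight barrier and plays two subsequences.
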LevelSequence /PistolCanary/Sequence/spa/spa1000/PistolCanary_spa1000_anm
bindings [
 TentacleBoss {
  id=#1
  possessed_class=PistolCanary_TentacleBoss_C
  tracks [{
    type=MovieSceneVisibilityTrack
    sections [{type=MovieSceneVisibilitySection, channels [None {type=MovieSceneScriptingBoolChannel, default=false, const [false,1]}]}]
   }]
 }
 KalEl {
  id=#2
  possessed_class=PistolCanary_CinematicNPC_Superhero_Main_C
  tracks [{
    type=MovieSceneVisibilityTrack
    sections [{
      type=MovieSceneVisibilitySection
      channels [None {type=MovieSceneScriptingBoolChannel, default=false, keys [[6113,false],[6122,true],[6232,false],[6895,true]]}]
     }]
   }]
 }
 TripleTentacleLeft {
  id=#3
  possessed_class=PistolCanary_TentacleBoss_C
  tracks [
   {
    type=MovieScene3DTransformTrack
    sections [{
      type=MovieScene3DTransformSection
      channels [
       Location.X {type=MovieSceneScriptingDoubleChannel, default=-1036.5809}
       Location.Y {type=MovieSceneScriptingDoubleChannel, default=-7497.0932}
       Location.Z {type=MovieSceneScriptingDoubleChannel, default=3571, keys [[6648,1941.8927],[6953,1941.8927],[7229,3571]]}
       Rotation.X {type=MovieSceneScriptingDoubleChannel, default=0}
       Rotation.Y {type=MovieSceneScriptingDoubleChannel, default=0}
       Rotation.Z {type=MovieSceneScriptingDoubleChannel, default=47.4855}
       Scale.X {type=MovieSceneScriptingDoubleChannel, default=1}
       Scale.Y {type=MovieSceneScriptingDoubleChannel, default=1}
       Scale.Z {type=MovieSceneScriptingDoubleChannel, default=1}
      ]
     }]
   }
   {
    type=MovieSceneVisibilityTrack
    sections [{
      type=MovieSceneVisibilitySection
      channels [None {type=MovieSceneScriptingBoolChannel, default=false, keys [[4227,false],[4428,true],[7320,false]]}]
     }]
   }
   {
    type=MovieSceneSkeletalAnimationTrack
    sections [
     {type=MovieSceneSkeletalAnimationSection, end_frame=4842, start_frame=3599}
     {type=MovieSceneSkeletalAnimationSection, end_frame=5125, start_frame=4736}
     {type=MovieSceneSkeletalAnimationSection, end_frame=6122, start_frame=5041}
     {type=MovieSceneSkeletalAnimationSection, end_frame=6896, start_frame=6122}
     {type=MovieSceneSkeletalAnimationSection, end_frame=8096, start_frame=6896}
    ]
   }
  ]
 }
 PistolCanary_TentacleCut {
  id=#4
  possessed_class=PistolCanary_TentacleCut_C
  tracks [
   {
    type=MovieSceneVisibilityTrack
    sections [{
      type=MovieSceneVisibilitySection
      channels [None {type=MovieSceneScriptingBoolChannel, default=false, keys [[5879,false],[6196,true],[6820,false]]}]
     }]
   }
   {type=MovieSceneSkeletalAnimationTrack, sections [{type=MovieSceneSkeletalAnimationSection, end_frame=6980, start_frame=6196}]}
   {
    type=MovieScene3DTransformTrack
    sections [{
      type=MovieScene3DTransformSection
      channels [
       Location.X {type=MovieSceneScriptingDoubleChannel, default=-532.8999, const [-532.8999,1]}
       Location.Y {type=MovieSceneScriptingDoubleChannel, default=-4714.7318, const [-4714.7319,1]}
       Location.Z {type=MovieSceneScriptingDoubleChannel, default=-221.7992, const [-221.7992,1]}
       Rotation.X {type=MovieSceneScriptingDoubleChannel, default=0, const [0,1]}
       Rotation.Y {type=MovieSceneScriptingDoubleChannel, default=0, const [0,1]}
       Rotation.Z {type=MovieSceneScriptingDoubleChannel, default=0, const [0,1]}
       Scale.X {type=MovieSceneScriptingDoubleChannel, default=1, const [1,1]}
       Scale.Y {type=MovieSceneScriptingDoubleChannel, default=1, const [1,1]}
       Scale.Z {type=MovieSceneScriptingDoubleChannel, default=1, const [1,1]}
      ]
     }]
   }
  ]
 }
 BP_PistolCanary_NoFightBarrier {
  id=#5
  possessed_class=BP_PistolCanary_NoFightBarrier_C
  tracks [{
    type=MovieScene3DTransformTrack
    sections [{
      type=MovieScene3DTransformSection
      channels [
       Location.X {type=MovieSceneScriptingDoubleChannel, default=14052, const [14052,1]}
       Location.Y {type=MovieSceneScriptingDoubleChannel, default=71572, const [71572,1]}
       Location.Z {type=MovieSceneScriptingDoubleChannel, default=3316, const [3316,1]}
       Rotation.X {type=MovieSceneScriptingDoubleChannel, default=0, const [0,1]}
       Rotation.Y {type=MovieSceneScriptingDoubleChannel, default=0, const [0,1]}
       Rotation.Z {type=MovieSceneScriptingDoubleChannel, default=0, const [0,1]}
       Scale.X {type=MovieSceneScriptingDoubleChannel, default=375, const [375,1]}
       Scale.Y {type=MovieSceneScriptingDoubleChannel, default=375, const [375,1]}
       Scale.Z {type=MovieSceneScriptingDoubleChannel, default=100.5, const [100.5,1]}
      ]
     }]
   }]
 }
 TripleTentacleFront {
  id=#6
  possessed_class=PistolCanary_TentacleBoss_C
  tracks [
   {
    type=MovieScene3DTransformTrack
    sections [{
      type=MovieScene3DTransformSection
      channels [
       Location.X {type=MovieSceneScriptingDoubleChannel, default=28.0001}
       Location.Y {type=MovieSceneScriptingDoubleChannel, default=-5598.6317}
       Location.Z {type=MovieSceneScriptingDoubleChannel, default=986.8008}
       Rotation.X {type=MovieSceneScriptingDoubleChannel, default=0}
       Rotation.Y {type=MovieSceneScriptingDoubleChannel, default=0}
       Rotation.Z {type=MovieSceneScriptingDoubleChannel, default=0}
       Scale.X {type=MovieSceneScriptingDoubleChannel, default=1}
       Scale.Y {type=MovieSceneScriptingDoubleChannel, default=1}
       Scale.Z {type=MovieSceneScriptingDoubleChannel, default=1}
      ]
     }]
   }
   {
    type=MovieSceneVisibilityTrack
    sections [{
      type=MovieSceneVisibilitySection
      channels [None {type=MovieSceneScriptingBoolChannel, default=false, keys [[3677.9975,false],[3784.9975,true],[6196,false]]}]
     }]
   }
  ]
 }
 TripleTentacleRight {
  id=#7
  possessed_class=PistolCanary_TentacleBoss_C
  tracks [
   {
    type=MovieScene3DTransformTrack
    sections [{
      type=MovieScene3DTransformSection
      channels [
       Location.X {type=MovieSceneScriptingDoubleChannel, default=1325.322, const [1325.322,1]}
       Location.Y {type=MovieSceneScriptingDoubleChannel, default=-9363.4975, const [-9363.4971,1]}
       Location.Z {type=MovieSceneScriptingDoubleChannel, default=3927, keys [[6967,2297.8594],[7229,3927]]}
       Rotation.X {type=MovieSceneScriptingDoubleChannel, default=18.745, const [18.745,1]}
       Rotation.Y {type=MovieSceneScriptingDoubleChannel, default=-7.1019, const [-7.1019,1]}
       Rotation.Z {type=MovieSceneScriptingDoubleChannel, default=-69.9822, const [-69.9822,1]}
       Scale.X {type=MovieSceneScriptingDoubleChannel, default=1}
       Scale.Y {type=MovieSceneScriptingDoubleChannel, default=1}
       Scale.Z {type=MovieSceneScriptingDoubleChannel, default=1}
      ]
     }]
   }
   {
    type=MovieSceneVisibilityTrack
    sections [{
      type=MovieSceneVisibilitySection
      channels [None {type=MovieSceneScriptingBoolChannel, default=false, keys [[3910,false],[4017,true],[8096,false]]}]
     }]
   }
   {
    type=MovieSceneSkeletalAnimationTrack
    sections [
     {type=MovieSceneSkeletalAnimationSection, end_frame=4834, start_frame=3599}
     {type=MovieSceneSkeletalAnimationSection, end_frame=5119, start_frame=4748}
     {type=MovieSceneSkeletalAnimationSection, end_frame=6122, start_frame=5041}
     {type=MovieSceneSkeletalAnimationSection, end_frame=6896, start_frame=6122}
     {type=MovieSceneSkeletalAnimationSection, end_frame=8508, start_frame=6896}
    ]
   }
  ]
 }
 SkeletalMesh {
  id=#8
  parent_id=#6
  possessed_class=SkeletalMeshComponent
  tracks [
   {
    type=MovieSceneSkeletalAnimationTrack
    sections [
     {type=MovieSceneSkeletalAnimationSection, end_frame=4842, start_frame=3600}
     {type=MovieSceneSkeletalAnimationSection, end_frame=5117, start_frame=4743}
     {type=MovieSceneSkeletalAnimationSection, end_frame=6138, start_frame=5057}
     {type=MovieSceneSkeletalAnimationSection, end_frame=6197, start_frame=6138}
    ]
   }
   {
    type=MovieSceneVisibilityTrack
    sections [{type=MovieSceneVisibilitySection, channels [None {type=MovieSceneScriptingBoolChannel, default=true}]}]
   }
  ]
 }
]
display_rate {denominator=1, numerator=60}
tick_resolution {denominator=1, numerator=24000}
tracks [{
  type=MovieSceneSubTrack
  display_name=Subsequences
  sections [{type=MovieSceneSubSection, end_frame=6411, start_frame=6122}, {type=MovieSceneSubSection, end_frame=9296, start_frame=6896}]
 }]
