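# Audio sublevel sequence for the spa1000 cinematic, scoring a fight between a tentacle boss and a flying, laser-firing superhero.
# Named master tracks layer overlapping, attachable sound sections for three tentacles and the hero's lasers and flight.
# Four rock spire monoliths carry their own audio sections, staggered in step with keys on a SpireExplosions event track.
# Event triggers also sit on each tentacle and a hero exit track, while the hero binding holds a static transform.
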LevelSequence /PistolCanary/Sequence/spa/spa1000/PistolCanary_spa1000_aud
bindings [
 TripleTentacle3 {
  id=#1
  possessed_class=PistolCanary_TentacleBoss_C
  tracks [{
    type=MovieSceneEventTrack
    sections [{
      type=MovieSceneEventTriggerSection
      channels [None {type=MovieSceneScriptingEventChannel, const ["<Struct 'MovieSceneEvent' (0x852ff01d160) {}>", 1]}]
     }]
   }]
 }
 ConTrail {
  id=#2
  possessed_class=PistolCanary_CinematicNPC_Superhero_Main_C
  tracks [
   {
    type=MovieScene3DTransformTrack
    sections [{
      type=MovieScene3DTransformSection
      channels [
       Location.X {type=MovieSceneScriptingDoubleChannel, default=-6208.6853}
       Location.Y {type=MovieSceneScriptingDoubleChannel, default=-3689.5778}
       Location.Z {type=MovieSceneScriptingDoubleChannel, default=6288.9768}
       Rotation.X {type=MovieSceneScriptingDoubleChannel, default=-10.5094}
       Rotation.Y {type=MovieSceneScriptingDoubleChannel, default=-0.1951}
       Rotation.Z {type=MovieSceneScriptingDoubleChannel, default=-117.1718}
       Scale.X {type=MovieSceneScriptingDoubleChannel, default=4}
       Scale.Y {type=MovieSceneScriptingDoubleChannel, default=4}
       Scale.Z {type=MovieSceneScriptingDoubleChannel, default=4}
      ]
     }]
   }
   {type=MovieSceneSkeletalAnimationTrack, sections []}
  ]
 }
 PistolCanary_RockSpireMonolith1 {
  id=#3
  possessed_class=PistolCanary_Monolith_Spire_C
  tracks [{
    type=MovieSceneAudioTrack
    display_name=Audio
    sections [{
      type=MovieSceneAudioSection
      end_frame=8388
      start_frame=8171
      channels [
       Volume {type=MovieSceneScriptingFloatChannel, default=1}
       Pitch {type=MovieSceneScriptingFloatChannel, default=1}
      ]
     }]
   }]
 }
 PistolCanary_RockSpireMonolith3 {
  id=#4
  possessed_class=PistolCanary_Monolith_Spire_C
  tracks [{
    type=MovieSceneAudioTrack
    display_name=Audio
    sections [{
      type=MovieSceneAudioSection
      end_frame=8728
      start_frame=8460
      channels [
       Volume {type=MovieSceneScriptingFloatChannel, default=1}
       Pitch {type=MovieSceneScriptingFloatChannel, default=1}
      ]
     }]
   }]
 }
 PistolCanary_RockSpireMonolith2 {
  id=#5
  possessed_class=PistolCanary_Monolith_Spire_C
  tracks [{
    type=MovieSceneAudioTrack
    display_name=Audio
    sections [{
      type=MovieSceneAudioSection
      end_frame=8538
      start_frame=8324
      channels [
       Volume {type=MovieSceneScriptingFloatChannel, default=1}
       Pitch {type=MovieSceneScriptingFloatChannel, default=1}
      ]
     }]
   }]
 }
 TripleTentacle1 {
  id=#6
  possessed_class=PistolCanary_TentacleBoss_C
  tracks [{
    type=MovieSceneEventTrack
    sections [{
      type=MovieSceneEventTriggerSection
      channels [None {
        type=MovieSceneScriptingEventChannel
        keys [[3951, "<Struct 'MovieSceneEvent' (0x852ff01c940) {}>"], [6214.0675, "<Struct 'MovieSceneEvent' (0x852ff01c120) {}>"]]
       }]
     }]
   }]
 }
 TripleTentacle2 {
  id=#7
  possessed_class=PistolCanary_TentacleBoss_C
  tracks [{
    type=MovieSceneEventTrack
    sections [{
      type=MovieSceneEventTriggerSection
      channels [None {type=MovieSceneScriptingEventChannel, const ["<Struct 'MovieSceneEvent' (0x852ff01d980) {}>", 1]}]
     }]
   }]
 }
 PistolCanary_RockSpireMonolith4 {
  id=#8
  possessed_class=PistolCanary_Monolith_Spire_C
  tracks [{
    type=MovieSceneAudioTrack
    display_name=Audio
    sections [{
      type=MovieSceneAudioSection
      end_frame=8822
      start_frame=8578
      channels [
       Volume {type=MovieSceneScriptingFloatChannel, default=1}
       Pitch {type=MovieSceneScriptingFloatChannel, default=1}
      ]
     }]
   }]
 }
]
display_rate {denominator=1, numerator=60}
tick_resolution {denominator=1, numerator=24000}
tracks [
 {
  type=MovieSceneAudioTrack
  display_name=Tentacle_1_Audio
  sections [
   {
    type=MovieSceneAudioSection
    end_frame=6199
    start_frame=5366
    channels [
     Volume {type=MovieSceneScriptingFloatChannel, default=1}
     Pitch {type=MovieSceneScriptingFloatChannel, default=1}
     AttachActor {type=MovieSceneScriptingActorReferenceChannel}
    ]
   }
   {
    type=MovieSceneAudioSection
    end_frame=4967
    start_frame=4597
    channels [
     Volume {type=MovieSceneScriptingFloatChannel, default=1}
     Pitch {type=MovieSceneScriptingFloatChannel, default=1}
     AttachActor {type=MovieSceneScriptingActorReferenceChannel}
    ]
   }
   {
    type=MovieSceneAudioSection
    end_frame=5892
    start_frame=5606
    channels [
     Volume {type=MovieSceneScriptingFloatChannel, default=1}
     Pitch {type=MovieSceneScriptingFloatChannel, default=1}
     AttachActor {type=MovieSceneScriptingActorReferenceChannel}
    ]
   }
   {
    type=MovieSceneAudioSection
    end_frame=4785
    start_frame=3866
    channels [
     Volume {type=MovieSceneScriptingFloatChannel, default=1}
     Pitch {type=MovieSceneScriptingFloatChannel, default=1}
     AttachActor {type=MovieSceneScriptingActorReferenceChannel}
    ]
   }
   {
    type=MovieSceneAudioSection
    end_frame=5399
    start_frame=5123
    channels [
     Volume {type=MovieSceneScriptingFloatChannel, default=1}
     Pitch {type=MovieSceneScriptingFloatChannel, default=1}
     AttachActor {type=MovieSceneScriptingActorReferenceChannel}
    ]
   }
  ]
 }
 {
  type=MovieSceneAudioTrack
  display_name=Tentacle_3_Audio
  sections [
   {
    type=MovieSceneAudioSection
    end_frame=6206
    start_frame=5379
    channels [
     Volume {type=MovieSceneScriptingFloatChannel, default=1}
     Pitch {type=MovieSceneScriptingFloatChannel, default=1}
     AttachActor {type=MovieSceneScriptingActorReferenceChannel}
    ]
   }
   {
    type=MovieSceneAudioSection
    end_frame=5192
    start_frame=4433
    channels [
     Volume {type=MovieSceneScriptingFloatChannel, default=1}
     Pitch {type=MovieSceneScriptingFloatChannel, default=1}
     AttachActor {type=MovieSceneScriptingActorReferenceChannel}
    ]
   }
   {
    type=MovieSceneAudioSection
    end_frame=6711
    start_frame=6225
    channels [
     Volume {type=MovieSceneScriptingFloatChannel, default=1}
     Pitch {type=MovieSceneScriptingFloatChannel, default=1}
     AttachActor {type=MovieSceneScriptingActorReferenceChannel}
    ]
   }
   {
    type=MovieSceneAudioSection
    end_frame=5802
    start_frame=5447
    channels [
     Volume {type=MovieSceneScriptingFloatChannel, default=1}
     Pitch {type=MovieSceneScriptingFloatChannel, default=1}
     AttachActor {type=MovieSceneScriptingActorReferenceChannel}
    ]
   }
   {
    type=MovieSceneAudioSection
    end_frame=4988
    start_frame=4561
    channels [
     Volume {type=MovieSceneScriptingFloatChannel, default=1}
     Pitch {type=MovieSceneScriptingFloatChannel, default=1}
     AttachActor {type=MovieSceneScriptingActorReferenceChannel}
    ]
   }
   {
    type=MovieSceneAudioSection
    end_frame=5350
    start_frame=4990
    channels [
     Volume {type=MovieSceneScriptingFloatChannel, default=1}
     Pitch {type=MovieSceneScriptingFloatChannel, default=1}
     AttachActor {type=MovieSceneScriptingActorReferenceChannel}
    ]
   }
   {
    type=MovieSceneAudioSection
    end_frame=8522
    start_frame=7793
    channels [
     Volume {type=MovieSceneScriptingFloatChannel, default=1}
     Pitch {type=MovieSceneScriptingFloatChannel, default=1}
     AttachActor {type=MovieSceneScriptingActorReferenceChannel}
    ]
   }
   {
    type=MovieSceneAudioSection
    end_frame=7412
    start_frame=7056
    channels [
     Volume {type=MovieSceneScriptingFloatChannel, default=1}
     Pitch {type=MovieSceneScriptingFloatChannel, default=1}
     AttachActor {type=MovieSceneScriptingActorReferenceChannel}
    ]
   }
   {
    type=MovieSceneAudioSection
    end_frame=5089
    start_frame=4711
    channels [
     Volume {type=MovieSceneScriptingFloatChannel, default=1}
     Pitch {type=MovieSceneScriptingFloatChannel, default=1}
     AttachActor {type=MovieSceneScriptingActorReferenceChannel}
    ]
   }
   {
    type=MovieSceneAudioSection
    end_frame=6204
    start_frame=5862
    channels [
     Volume {type=MovieSceneScriptingFloatChannel, default=1}
     Pitch {type=MovieSceneScriptingFloatChannel, default=1}
     AttachActor {type=MovieSceneScriptingActorReferenceChannel}
    ]
   }
   {
    type=MovieSceneAudioSection
    end_frame=8226
    start_frame=7751
    channels [
     Volume {type=MovieSceneScriptingFloatChannel, default=1}
     Pitch {type=MovieSceneScriptingFloatChannel, default=1}
     AttachActor {type=MovieSceneScriptingActorReferenceChannel}
    ]
   }
   {
    type=MovieSceneAudioSection
    end_frame=8167
    start_frame=7661
    channels [
     Volume {type=MovieSceneScriptingFloatChannel, default=1}
     Pitch {type=MovieSceneScriptingFloatChannel, default=1}
     AttachActor {type=MovieSceneScriptingActorReferenceChannel}
    ]
   }
  ]
 }
 {
  type=MovieSceneAudioTrack
  display_name=Tentacle_2_Audio
  sections [
   {
    type=MovieSceneAudioSection
    end_frame=6197
    start_frame=5352
    channels [
     Volume {type=MovieSceneScriptingFloatChannel, default=1}
     Pitch {type=MovieSceneScriptingFloatChannel, default=1}
     AttachActor {type=MovieSceneScriptingActorReferenceChannel}
    ]
   }
   {
    type=MovieSceneAudioSection
    end_frame=5033
    start_frame=4548
    channels [
     Volume {type=MovieSceneScriptingFloatChannel, default=1}
     Pitch {type=MovieSceneScriptingFloatChannel, default=1}
     AttachActor {type=MovieSceneScriptingActorReferenceChannel}
    ]
   }
   {
    type=MovieSceneAudioSection
    end_frame=5462
    start_frame=5210
    channels [
     Volume {type=MovieSceneScriptingFloatChannel, default=1}
     Pitch {type=MovieSceneScriptingFloatChannel, default=1}
     AttachActor {type=MovieSceneScriptingActorReferenceChannel}
    ]
   }
   {
    type=MovieSceneAudioSection
    end_frame=5358
    start_frame=3604
    channels [
     Volume {type=MovieSceneScriptingFloatChannel, default=1}
     Pitch {type=MovieSceneScriptingFloatChannel, default=1}
     AttachActor {type=MovieSceneScriptingActorReferenceChannel}
    ]
   }
   {
    type=MovieSceneAudioSection
    end_frame=7425
    start_frame=7160
    channels [
     Volume {type=MovieSceneScriptingFloatChannel, default=1}
     Pitch {type=MovieSceneScriptingFloatChannel, default=1}
     AttachActor {type=MovieSceneScriptingActorReferenceChannel}
    ]
   }
   {
    type=MovieSceneAudioSection
    end_frame=4728
    start_frame=4151
    channels [
     Volume {type=MovieSceneScriptingFloatChannel, default=1}
     Pitch {type=MovieSceneScriptingFloatChannel, default=1}
     AttachActor {type=MovieSceneScriptingActorReferenceChannel}
    ]
   }
   {
    type=MovieSceneAudioSection
    end_frame=5285
    start_frame=4972
    channels [
     Volume {type=MovieSceneScriptingFloatChannel, default=1}
     Pitch {type=MovieSceneScriptingFloatChannel, default=1}
     AttachActor {type=MovieSceneScriptingActorReferenceChannel}
    ]
   }
   {
    type=MovieSceneAudioSection
    end_frame=6201
    start_frame=5923
    channels [
     Volume {type=MovieSceneScriptingFloatChannel, default=1}
     Pitch {type=MovieSceneScriptingFloatChannel, default=1}
     AttachActor {type=MovieSceneScriptingActorReferenceChannel}
    ]
   }
   {
    type=MovieSceneAudioSection
    end_frame=7626
    start_frame=7313
    channels [
     Volume {type=MovieSceneScriptingFloatChannel, default=1.3}
     Pitch {type=MovieSceneScriptingFloatChannel, default=1}
     AttachActor {type=MovieSceneScriptingActorReferenceChannel}
    ]
   }
   {
    type=MovieSceneAudioSection
    end_frame=5604
    start_frame=5292
    channels [
     Volume {type=MovieSceneScriptingFloatChannel, default=1}
     Pitch {type=MovieSceneScriptingFloatChannel, default=1}
     AttachActor {type=MovieSceneScriptingActorReferenceChannel}
    ]
   }
   {
    type=MovieSceneAudioSection
    end_frame=7707
    start_frame=7586
    channels [
     Volume {type=MovieSceneScriptingFloatChannel, default=1}
     Pitch {type=MovieSceneScriptingFloatChannel, default=1}
     AttachActor {type=MovieSceneScriptingActorReferenceChannel}
    ]
   }
   {
    type=MovieSceneAudioSection
    end_frame=7938
    start_frame=7517
    channels [
     Volume {type=MovieSceneScriptingFloatChannel, default=1}
     Pitch {type=MovieSceneScriptingFloatChannel, default=1}
     AttachActor {type=MovieSceneScriptingActorReferenceChannel}
    ]
   }
  ]
 }
 {
  type=MovieSceneAudioTrack
  display_name=MainHeroLaserAudio
  sections [
   {
    type=MovieSceneAudioSection
    end_frame=8277
    start_frame=8086
    channels [
     Volume {type=MovieSceneScriptingFloatChannel, default=1}
     Pitch {type=MovieSceneScriptingFloatChannel, default=1}
     AttachActor {type=MovieSceneScriptingActorReferenceChannel}
    ]
   }
   {
    type=MovieSceneAudioSection
    end_frame=8448
    start_frame=8268
    channels [
     Volume {type=MovieSceneScriptingFloatChannel, default=1}
     Pitch {type=MovieSceneScriptingFloatChannel, default=1}
     AttachActor {type=MovieSceneScriptingActorReferenceChannel}
    ]
   }
   {
    type=MovieSceneAudioSection
    end_frame=8625
    start_frame=8417
    channels [
     Volume {type=MovieSceneScriptingFloatChannel, default=1}
     Pitch {type=MovieSceneScriptingFloatChannel, default=1}
     AttachActor {type=MovieSceneScriptingActorReferenceChannel}
    ]
   }
   {
    type=MovieSceneAudioSection
    end_frame=8617
    start_frame=8551
    channels [
     Volume {type=MovieSceneScriptingFloatChannel, default=1}
     Pitch {type=MovieSceneScriptingFloatChannel, default=1}
     AttachActor {type=MovieSceneScriptingActorReferenceChannel}
    ]
   }
   {
    type=MovieSceneAudioSection
    end_frame=7251
    start_frame=6994
    channels [
     Volume {type=MovieSceneScriptingFloatChannel, default=1}
     Pitch {type=MovieSceneScriptingFloatChannel, default=1}
     AttachActor {type=MovieSceneScriptingActorReferenceChannel}
    ]
   }
  ]
 }
 {
  type=MovieSceneAudioTrack
  display_name=MainHeroFlyingAudio
  sections [
   {
    type=MovieSceneAudioSection
    end_frame=9340
    start_frame=6904
    channels [
     Volume {type=MovieSceneScriptingFloatChannel, default=1}
     Pitch {type=MovieSceneScriptingFloatChannel, default=1}
     AttachActor {type=MovieSceneScriptingActorReferenceChannel}
    ]
   }
   {
    type=MovieSceneAudioSection
    end_frame=11586
    start_frame=6892
    channels [
     Volume {type=MovieSceneScriptingFloatChannel, default=1}
     Pitch {type=MovieSceneScriptingFloatChannel, default=1}
     AttachActor {type=MovieSceneScriptingActorReferenceChannel}
    ]
   }
   {
    type=MovieSceneAudioSection
    end_frame=7914
    start_frame=7634
    channels [
     Volume {type=MovieSceneScriptingFloatChannel, default=1}
     Pitch {type=MovieSceneScriptingFloatChannel, default=1}
     AttachActor {type=MovieSceneScriptingActorReferenceChannel}
    ]
   }
   {
    type=MovieSceneAudioSection
    end_frame=8402
    start_frame=8266
    channels [
     Volume {type=MovieSceneScriptingFloatChannel, default=1}
     Pitch {type=MovieSceneScriptingFloatChannel, default=1}
     AttachActor {type=MovieSceneScriptingActorReferenceChannel}
    ]
   }
   {
    type=MovieSceneAudioSection
    end_frame=8250
    start_frame=8122
    channels [
     Volume {type=MovieSceneScriptingFloatChannel, default=1}
     Pitch {type=MovieSceneScriptingFloatChannel, default=1}
     AttachActor {type=MovieSceneScriptingActorReferenceChannel}
    ]
   }
   {
    type=MovieSceneAudioSection
    end_frame=8774
    start_frame=8493
    channels [
     Volume {type=MovieSceneScriptingFloatChannel, default=1}
     Pitch {type=MovieSceneScriptingFloatChannel, default=1}
     AttachActor {type=MovieSceneScriptingActorReferenceChannel}
    ]
   }
   {
    type=MovieSceneAudioSection
    end_frame=7269
    start_frame=7156
    channels [
     Volume {type=MovieSceneScriptingFloatChannel, default=1}
     Pitch {type=MovieSceneScriptingFloatChannel, default=1}
     AttachActor {type=MovieSceneScriptingActorReferenceChannel}
    ]
   }
   {
    type=MovieSceneAudioSection
    end_frame=7514
    start_frame=7393
    channels [
     Volume {type=MovieSceneScriptingFloatChannel, default=1}
     Pitch {type=MovieSceneScriptingFloatChannel, default=1}
     AttachActor {type=MovieSceneScriptingActorReferenceChannel}
    ]
   }
   {
    type=MovieSceneAudioSection
    end_frame=7599
    start_frame=7513
    channels [
     Volume {type=MovieSceneScriptingFloatChannel, default=1}
     Pitch {type=MovieSceneScriptingFloatChannel, default=1}
     AttachActor {type=MovieSceneScriptingActorReferenceChannel}
    ]
   }
   {
    type=MovieSceneAudioSection
    end_frame=8056
    start_frame=7940
    channels [
     Volume {type=MovieSceneScriptingFloatChannel, default=1}
     Pitch {type=MovieSceneScriptingFloatChannel, default=1}
     AttachActor {type=MovieSceneScriptingActorReferenceChannel}
    ]
   }
   {
    type=MovieSceneAudioSection
    end_frame=8342
    start_frame=8218
    channels [
     Volume {type=MovieSceneScriptingFloatChannel, default=1}
     Pitch {type=MovieSceneScriptingFloatChannel, default=1}
     AttachActor {type=MovieSceneScriptingActorReferenceChannel}
    ]
   }
   {
    type=MovieSceneAudioSection
    end_frame=8707
    start_frame=8636
    channels [
     Volume {type=MovieSceneScriptingFloatChannel, default=1}
     Pitch {type=MovieSceneScriptingFloatChannel, default=1}
     AttachActor {type=MovieSceneScriptingActorReferenceChannel}
    ]
   }
   {
    type=MovieSceneAudioSection
    end_frame=8935
    start_frame=8800
    channels [
     Volume {type=MovieSceneScriptingFloatChannel, default=1}
     Pitch {type=MovieSceneScriptingFloatChannel, default=1}
     AttachActor {type=MovieSceneScriptingActorReferenceChannel}
    ]
   }
   {
    type=MovieSceneAudioSection
    end_frame=8861
    start_frame=8740
    channels [
     Volume {type=MovieSceneScriptingFloatChannel, default=1}
     Pitch {type=MovieSceneScriptingFloatChannel, default=0.8}
     AttachActor {type=MovieSceneScriptingActorReferenceChannel}
    ]
   }
  ]
 }
 {
  type=MovieSceneEventTrack
  display_name=Event Hero Exit
  sections [{
    type=MovieSceneEventTriggerSection
    channels [None {type=MovieSceneScriptingEventChannel, const ["<Struct 'MovieSceneEvent' (0x852ff01d3e0) {}>", 1]}]
   }]
 }
 {
  type=MovieSceneAudioTrack
  display_name=Audio
  sections [{
    type=MovieSceneAudioSection
    end_frame=6661
    start_frame=6141
    channels [
     Volume {type=MovieSceneScriptingFloatChannel, default=1}
     Pitch {type=MovieSceneScriptingFloatChannel, default=1}
     AttachActor {type=MovieSceneScriptingActorReferenceChannel, default="<Struct 'MovieSceneObjectBindingID' (0x852fd7acf80) {}>"}
    ]
   }]
 }
 {
  type=MovieSceneEventTrack
  display_name=SpireExplosions
  sections [{
    type=MovieSceneEventTriggerSection
    channels [None {
      type=MovieSceneScriptingEventChannel
      keys [
       [8175, "<Struct 'MovieSceneEvent' (0x852ff01ce40) {}>"]
       [8324, "<Struct 'MovieSceneEvent' (0x852ff01cda0) {}>"]
       [8461, "<Struct 'MovieSceneEvent' (0x852ff01cd00) {}>"]
       [8570, "<Struct 'MovieSceneEvent' (0x852ff01cc60) {}>"]
      ]
     }]
   }]
 }
]
