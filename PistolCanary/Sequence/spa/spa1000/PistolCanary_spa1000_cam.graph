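# Level Sequence for PistolCanary spa1000 that drives timed camera shakes and controller force feedback, apparently for a lobby event.
# It animates one shared cameraForce Blueprint at a fixed transform, keying its shake class through emerge, triplecharge, slice and punch variants.
# The force effect drops from medium to small early on.
# Two event tracks fire shake and feedback at matching times, though the triplecharge trigger lands one tick before its class key.
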LevelSequence /PistolCanary/Sequence/spa/spa1000/PistolCanary_spa1000_cam
bindings [PistolCanary_cameraForce_BP {
  id=#1
  possessed_class=Events_Shared_cameraForce_BP_C
  tracks [
   {
    type=MovieSceneObjectPropertyTrack
    display_name=Force Effect
    sections [{
      type=MovieSceneObjectPropertySection
      channels [None {
        type=MovieSceneScriptingObjectPathChannel
        default="<Object '&Events_Shared_medium_FF' (0x8509a2598b0) Class 'ForceFeedbackEffect'>"
        keys [
         [3956, "<Object '&Events_Shared_medium_FF' (0x8509a2598b0) Class 'ForceFeedbackEffect'>"]
         [4426, "<Object '&Events_Shared_small_FF' (0x84c3c703480) Class 'ForceFeedbackEffect'>"]
        ]
       }]
     }]
   }
   {
    type=MovieSceneObjectPropertyTrack
    display_name=Cam Shake Base
    sections [{
      type=MovieSceneObjectPropertySection
      channels [None {
        type=MovieSceneScriptingObjectPathChannel
        default="<Object '&Events_Shared_00small_csb_C' (0x84c3c6a8a00) Class 'BlueprintGeneratedClass'>"
        keys [
         [
          3956
          "<Object '/PistolCanary/Camera/PistolCanary_spa1000_emerge1_csb.PistolCanary_spa1000_emerge1_csb_C' (0x852d664c600) Class 'BlueprintGeneratedClass'>"
         ]
         [
          4146
          "<Object '/PistolCanary/Camera/PistolCanary_spa1000_emerge2_csb.PistolCanary_spa1000_emerge2_csb_C' (0x852e6405400) Class 'BlueprintGeneratedClass'>"
         ]
         [
          4426
          "<Object '/PistolCanary/Camera/PistolCanary_spa1000_emerge3_csb.PistolCanary_spa1000_emerge3_csb_C' (0x852e6406c00) Class 'BlueprintGeneratedClass'>"
         ]
         [
          5001
          "<Object '/PistolCanary/Camera/PistolCanary_spa1000_emerge4_csb.PistolCanary_spa1000_emerge4_csb_C' (0x852e6407e00) Class 'BlueprintGeneratedClass'>"
         ]
         [
          5968
          "<Object '/PistolCanary/Camera/PistolCanary_spa1000_triplecharge_csb.PistolCanary_spa1000_triplecharge_csb_C' (0x852e6408a00) Class 'BlueprintGeneratedClass'>"
         ]
         [
          6203
          "<Object '/PistolCanary/Camera/PistolCanary_spa1000_slice_csb.PistolCanary_spa1000_slice_csb_C' (0x852d99e4800) Class 'BlueprintGeneratedClass'>"
         ]
         [
          7008
          "<Object '/PistolCanary/Camera/PistolCanary_spa1000_punch_csb.PistolCanary_spa1000_punch_csb_C' (0x852e6406000) Class 'BlueprintGeneratedClass'>"
         ]
         [
          7770
          "<Object '&PistolCanary_spa1000_punch_csb_C' (0x852e6406000) Class 'BlueprintGeneratedClass'>"
         ]
         [
          8171
          "<Object '&PistolCanary_spa1000_slice_csb_C' (0x852d99e4800) Class 'BlueprintGeneratedClass'>"
         ]
         [
          9187
          "<Object '&PistolCanary_spa1000_punch_csb_C' (0x852e6406000) Class 'BlueprintGeneratedClass'>"
         ]
        ]
       }]
     }]
   }
   {type=MovieSceneSpawnTrack, sections [{type=MovieSceneSpawnSection, channels [None {type=MovieSceneScriptingBoolChannel, default=true}]}]}
   {
    type=MovieScene3DTransformTrack
    sections [{
      type=MovieScene3DTransformSection
      channels [
       Location.X {type=MovieSceneScriptingDoubleChannel, default=7084}
       Location.Y {type=MovieSceneScriptingDoubleChannel, default=1572}
       Location.Z {type=MovieSceneScriptingDoubleChannel, default=0}
       Rotation.X {type=MovieSceneScriptingDoubleChannel, default=0}
       Rotation.Y {type=MovieSceneScriptingDoubleChannel, default=0}
       Rotation.Z {type=MovieSceneScriptingDoubleChannel, default=0}
       Scale.X {type=MovieSceneScriptingDoubleChannel, default=1}
       Scale.Y {type=MovieSceneScriptingDoubleChannel, default=1}
       Scale.Z {type=MovieSceneScriptingDoubleChannel, default=1}
      ]
     }]
   }
   {
    type=MovieSceneEventTrack
    display_name=Cam Shake Lobby Event Trigger
    sections [{
      type=MovieSceneEventTriggerSection
      channels [None {
        type=MovieSceneScriptingEventChannel
        keys [
         [3956, "<Struct 'MovieSceneEvent' (0x852ff01d7a0) {}>"]
         [4146, "<Struct 'MovieSceneEvent' (0x852ff01d840) {}>"]
         [4426, "<Struct 'MovieSceneEvent' (0x852ff01cc60) {}>"]
         [5001, "<Struct 'MovieSceneEvent' (0x852ff01cd00) {}>"]
         [5967, "<Struct 'MovieSceneEvent' (0x852ff01cda0) {}>"]
         [6203, "<Struct 'MovieSceneEvent' (0x852ff01ce40) {}>"]
         [7008, "<Struct 'MovieSceneEvent' (0x852ff01d3e0) {}>"]
         [7770, "<Struct 'MovieSceneEvent' (0x852ff01d980) {}>"]
         [8171, "<Struct 'MovieSceneEvent' (0x852ff01c120) {}>"]
         [8323, "<Struct 'MovieSceneEvent' (0x852ff01c940) {}>"]
         [8462, "<Struct 'MovieSceneEvent' (0x852ff01d160) {}>"]
         [8580, "<Struct 'MovieSceneEvent' (0x852ff01d700) {}>"]
         [9187, "<Struct 'MovieSceneEvent' (0x852ff01d520) {}>"]
        ]
       }]
     }]
   }
   {
    type=MovieSceneEventTrack
    display_name=Force Feedback Lobby Event Trigger
    sections [{
      type=MovieSceneEventTriggerSection
      channels [None {
        type=MovieSceneScriptingEventChannel
        keys [
         [3956, "<Struct 'MovieSceneEvent' (0x852ff01db60) {}>"]
         [4146, "<Struct 'MovieSceneEvent' (0x852ff01dc00) {}>"]
         [4426, "<Struct 'MovieSceneEvent' (0x852ff01dca0) {}>"]
         [5001, "<Struct 'MovieSceneEvent' (0x852ff01dd40) {}>"]
         [5967, "<Struct 'MovieSceneEvent' (0x852ff01dde0) {}>"]
         [6203, "<Struct 'MovieSceneEvent' (0x852ff01de80) {}>"]
         [7008, "<Struct 'MovieSceneEvent' (0x852ff01df20) {}>"]
         [7770, "<Struct 'MovieSceneEvent' (0x852ff01dfc0) {}>"]
         [8171, "<Struct 'MovieSceneEvent' (0x852ff01e060) {}>"]
         [8323, "<Struct 'MovieSceneEvent' (0x852ff01e100) {}>"]
         [8462, "<Struct 'MovieSceneEvent' (0x852ff01e1a0) {}>"]
         [8580, "<Struct 'MovieSceneEvent' (0x852ff01e240) {}>"]
         [9187, "<Struct 'MovieSceneEvent' (0x852ff01e2e0) {}>"]
        ]
       }]
     }]
   }
  ]
 }]
display_rate {denominator=1, numerator=60}
tick_resolution {denominator=1, numerator=24000}
tracks []
