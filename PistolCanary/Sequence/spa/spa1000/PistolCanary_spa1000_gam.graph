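# Gameplay Level Sequence for PistolCanary spa1000 that times the world-side logic of a scripted event, apparently around a castle.
# A knockback volume scales up, a loot destroyer spawns briefly, and event keys trigger it and a player-built structure destroyer.
# A focus point moves through keyed positions and later attaches, while an NPC hero staging target relocates.
# Event tracks also drive the Section3 sequence runner, NPCs and camera modification.
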LevelSequence /PistolCanary/Sequence/spa/spa1000/PistolCanary_spa1000_gam
bindings [
 BP_PistolCanary_LootDestroyer {
  id=#1
  possessed_class=BP_PistolCanary_LootDestroyer_C
  tracks [
   {
    type=MovieSceneSpawnTrack
    sections [{
      type=MovieSceneSpawnSection
      channels [None {type=MovieSceneScriptingBoolChannel, default=false, keys [[4933,false],[4934,true],[5423,false]]}]
     }]
   }
   {
    type=MovieScene3DTransformTrack
    sections [{
      type=MovieScene3DTransformSection
      channels [
       Location.X {type=MovieSceneScriptingDoubleChannel, default=14276}
       Location.Y {type=MovieSceneScriptingDoubleChannel, default=73740}
       Location.Z {type=MovieSceneScriptingDoubleChannel, default=2629.6496}
       Rotation.X {type=MovieSceneScriptingDoubleChannel, default=0}
       Rotation.Y {type=MovieSceneScriptingDoubleChannel, default=0}
       Rotation.Z {type=MovieSceneScriptingDoubleChannel, default=0}
       Scale.X {type=MovieSceneScriptingDoubleChannel, default=200}
       Scale.Y {type=MovieSceneScriptingDoubleChannel, default=250}
       Scale.Z {type=MovieSceneScriptingDoubleChannel, default=200}
      ]
     }]
   }
   {
    type=MovieSceneEventTrack
    sections [{
      type=MovieSceneEventTriggerSection
      channels [None {
        type=MovieSceneScriptingEventChannel
        keys [[4974, "<Struct 'MovieSceneEvent' (0x852ff01df20) {}>"], [5360, "<Struct 'MovieSceneEvent' (0x852ff01dfc0) {}>"]]
       }]
     }]
   }
  ]
 }
 TargetPoint_NPCHeroes_StagingLoc01 {
  id=#2
  possessed_class=TargetPoint
  tracks [{
    type=MovieScene3DTransformTrack
    sections [{
      type=MovieScene3DTransformSection
      channels [
       Location.X {type=MovieSceneScriptingDoubleChannel, default=14216, keys [[7952,14216],[8861,29121.084]]}
       Location.Y {type=MovieSceneScriptingDoubleChannel, default=77221.7828, keys [[7952,77221.7812],[8861,80008.9453]]}
       Location.Z {type=MovieSceneScriptingDoubleChannel, default=3078.2995, keys [[7952,3078.2996],[8861,3370.376]]}
       Rotation.X {type=MovieSceneScriptingDoubleChannel, default=0, const [0,2]}
       Rotation.Y {type=MovieSceneScriptingDoubleChannel, default=0, const [0,2]}
       Rotation.Z {type=MovieSceneScriptingDoubleChannel, default=0, const [0,2]}
       Scale.X {type=MovieSceneScriptingDoubleChannel, default=1, const [1,2]}
       Scale.Y {type=MovieSceneScriptingDoubleChannel, default=1, const [1,2]}
       Scale.Z {type=MovieSceneScriptingDoubleChannel, default=1, const [1,2]}
      ]
     }]
   }]
 }
 BP_PistolCanary_KnockbackVolume_Castle {
  id=#3
  possessed_class=BP_PistolCanary_KnockbackVolume_C
  tracks [{
    type=MovieScene3DTransformTrack
    sections [{
      type=MovieScene3DTransformSection
      channels [
       Location.X {type=MovieSceneScriptingDoubleChannel, default=14424}
       Location.Y {type=MovieSceneScriptingDoubleChannel, default=70100}
       Location.Z {type=MovieSceneScriptingDoubleChannel, default=3428}
       Rotation.X {type=MovieSceneScriptingDoubleChannel, default=0}
       Rotation.Y {type=MovieSceneScriptingDoubleChannel, default=0}
       Rotation.Z {type=MovieSceneScriptingDoubleChannel, default=0}
       Scale.X {type=MovieSceneScriptingDoubleChannel, default=50, keys [[3832,0],[3833,50],[4214,137.5]]}
       Scale.Y {type=MovieSceneScriptingDoubleChannel, default=50, keys [[3832,0],[3833,50],[4214,137.5]]}
       Scale.Z {type=MovieSceneScriptingDoubleChannel, default=50, keys [[3832,0],[3833,50],[4214,137.5]]}
      ]
     }]
   }]
 }
 BP_PistolCanary_SequenceRunner_Section3 {
  id=#4
  possessed_class=BP_PistolCanary_SequenceRunner_Section1_C
  tracks [{
    type=MovieSceneEventTrack
    sections [{
      type=MovieSceneEventTriggerSection
      channels [None {
        type=MovieSceneScriptingEventChannel
        keys [[4925, "<Struct 'MovieSceneEvent' (0x852ff01de80) {}>"], [8862, "<Struct 'MovieSceneEvent' (0x852ff01e060) {}>"]]
       }]
     }]
   }]
 }
 BP_PistolCanary_FocusPoint {
  id=#5
  possessed_class=BP_PistolCanary_FocusPoint_C
  tracks [
   {
    type=MovieSceneEventTrack
    sections [{
      type=MovieSceneEventTriggerSection
      channels [None {
        type=MovieSceneScriptingEventChannel
        keys [
         [3637.675, "<Struct 'MovieSceneEvent' (0x852ff01e380) {}>"]
         [3901, "<Struct 'MovieSceneEvent' (0x852ff01e4c0) {}>"]
         [5470, "<Struct 'MovieSceneEvent' (0x852ff01e1a0) {}>"]
         [5471, "<Struct 'MovieSceneEvent' (0x852ff01dd40) {}>"]
         [5472, "<Struct 'MovieSceneEvent' (0x852ff01db60) {}>"]
         [5657, "<Struct 'MovieSceneEvent' (0x852ff01c120) {}>"]
         [5658, "<Struct 'MovieSceneEvent' (0x852ff01ce40) {}>"]
         [5659, "<Struct 'MovieSceneEvent' (0x852ff01e600) {}>"]
         [6028, "<Struct 'MovieSceneEvent' (0x852ff01e7e0) {}>"]
         [6029, "<Struct 'MovieSceneEvent' (0x852ff01e740) {}>"]
         [6030, "<Struct 'MovieSceneEvent' (0x852ff01c940) {}>"]
         [6164, "<Struct 'MovieSceneEvent' (0x852ff01e240) {}>"]
         [6165, "<Struct 'MovieSceneEvent' (0x852ff01cbc0) {}>"]
         [6166, "<Struct 'MovieSceneEvent' (0x852ff01dac0) {}>"]
         [6970, "<Struct 'MovieSceneEvent' (0x852ff01ea60) {}>"]
         [9163, "<Struct 'MovieSceneEvent' (0x852ff01ec40) {}>"]
         [57800, "<Struct 'MovieSceneEvent' (0x852ff01ee20) {}>"]
         [328260, "<Struct 'MovieSceneEvent' (0x852ff01f000) {}>"]
        ]
       }]
     }]
   }
   {type=MovieScene3DAttachTrack, sections [{type=MovieScene3DAttachSection, end_frame=9306, start_frame=6910}]}
   {
    type=MovieScene3DTransformTrack
    sections [{
      type=MovieScene3DTransformSection
      end_frame=6900
      channels [
       Location.X {type=MovieSceneScriptingDoubleChannel, default=10653, keys [[5471,12784],[6164,12784],[6200,16784],[6232,10653],[6462,10653],[6513,12784]]}
       Location.Y {type=MovieSceneScriptingDoubleChannel, default=71736, const [71736,6]}
       Location.Z {
        type=MovieSceneScriptingDoubleChannel
        default=9100
        keys [[5471,9100],[5526,9100],[6030,14100],[6164,9100],[6200,9100],[6232,6500],[6462,6500],[6513,9100]]
       }
      ]
     }]
   }
  ]
 }
 BP_PistolCanary_PlayerBuiltStructureDestroyer_FullPOI {
  id=#6
  possessed_class=BP_PistolCanary_PlayerBuiltStructureDestroyer_C
  tracks [{
    type=MovieSceneEventTrack
    sections [{
      type=MovieSceneEventTriggerSection
      channels [None {
        type=MovieSceneScriptingEventChannel
        keys [[3930.16, "<Struct 'MovieSceneEvent' (0x852ff01d3e0) {}>"], [5002, "<Struct 'MovieSceneEvent' (0x852ff01eba0) {}>"]]
       }]
     }]
   }]
 }
]
display_rate {denominator=1, numerator=60}
tick_resolution {denominator=1, numerator=24000}
tracks [
 {
  type=MovieSceneEventTrack
  display_name=NPCs
  sections [{
    type=MovieSceneEventTriggerSection
    channels [None {
      type=MovieSceneScriptingEventChannel
      keys [[9212.51, "<Struct 'MovieSceneEvent' (0x852ff01e9c0) {}>"], [9225.51, "<Struct 'MovieSceneEvent' (0x852ff01e920) {}>"]]
     }]
   }]
 }
 {
  type=MovieSceneEventTrack
  sections [{
    type=MovieSceneEventTriggerSection
    channels [None {
      type=MovieSceneScriptingEventChannel
      keys [
       [5002, "<Struct 'MovieSceneEvent' (0x852ff01d7a0) {}>"]
       [9225, "<Struct 'MovieSceneEvent' (0x852ff01d840) {}>"]
       [9285, "<Struct 'MovieSceneEvent' (0x852ff01cc60) {}>"]
       [9290, "<Struct 'MovieSceneEvent' (0x852ff01cd00) {}>"]
      ]
     }]
   }]
 }
 {
  type=MovieSceneEventTrack
  display_name=ModifyCamera
  sections [{
    type=MovieSceneEventTriggerSection
    channels [None {
      type=MovieSceneScriptingEventChannel
      keys [[3965, "<Struct 'MovieSceneEvent' (0x852ff01d160) {}>"], [9188, "<Struct 'MovieSceneEvent' (0x852ff01d700) {}>"]]
     }]
   }]
 }
]
