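# Lighting Level Sequence for PistolCanary shot spa1000, driving the day-sequence override, spotlights and fog for the cinematic.
# DaySequenceModifier volume blend weight ramps in, holds, then steps down to zero, while the MaskMakerFortress volume's capsule radius expands.
# A blue energy-ball spotlight moves, grows in intensity and size, then cuts out, while a second spotlight fades in and out.
# Visibility tracks hide both low-hanging fogs but show the main jungle fog; an event track targets the end light beam.
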LevelSequence /PistolCanary/Sequence/spa/spa1000/PistolCanary_spa1000_lgt
bindings [
 DSV_MaskMakerFortress {id=#1, possessed_class=DaySequenceModifierVolume, tracks []}
 BP_JungleFog_LowHanging2 {
  id=#2
  possessed_class=BP_JungleFog_C
  tracks [{
    type=MovieSceneVisibilityTrack
    sections [{type=MovieSceneVisibilitySection, channels [None {type=MovieSceneScriptingBoolChannel, default=false}]}]
   }]
 }
 BP_JungleFog {
  id=#3
  possessed_class=BP_JungleFog_C
  tracks [{
    type=MovieSceneVisibilityTrack
    sections [{type=MovieSceneVisibilitySection, channels [None {type=MovieSceneScriptingBoolChannel, default=false, const [true,1]}]}]
   }]
 }
 RectLight_pistolCanary {id=#4, possessed_class=SpotLight, tracks []}
 DaySequenceModifier {
  id=#5
  parent_id=#6
  possessed_class=DaySequenceModifierComponent
  tracks [{
    type=MovieSceneFloatTrack
    display_name=User Blend Weight
    sections [{
      type=MovieSceneFloatSection
      channels [None {
        type=MovieSceneScriptingFloatChannel
        default=0
        keys [[3600,0],[4008,1],[8166,1],[8287,0.75],[8325,0.75],[8394,0.5],[8460,0.5],[8520,0.25],[8579,0.25],[8935,0],[9265,0]]
       }]
     }]
   }]
 }
 LightComponent0 {
  id=#7
  parent_id=#8
  possessed_class=SpotLightComponent
  tracks [
   {
    type=MovieSceneFloatTrack
    display_name=Intensity
    sections [{
      type=MovieSceneFloatSection
      channels [None {type=MovieSceneScriptingFloatChannel, default=20000000, keys [[5405,0],[6169,20000000],[6240,0]]}]
     }]
   }
   {
    type=MovieSceneColorTrack
    display_name=Light Color
    sections [{
      type=MovieSceneColorSection
      channels [
       Color.R {type=MovieSceneScriptingFloatChannel, default=0}
       Color.G {type=MovieSceneScriptingFloatChannel, default=0.2623}
       Color.B {type=MovieSceneScriptingFloatChannel, default=1}
       Color.A {type=MovieSceneScriptingFloatChannel, default=1}
      ]
     }]
   }
   {
    type=MovieSceneFloatTrack
    display_name=Attenuation Radius
    sections [{
      type=MovieSceneFloatSection
      channels [None {type=MovieSceneScriptingFloatChannel, default=5000, keys [[5405,5000],[5623,9000],[6169,12000]]}]
     }]
   }
   {
    type=MovieSceneFloatTrack
    display_name=Source Radius
    sections [{type=MovieSceneFloatSection, channels [None {type=MovieSceneScriptingFloatChannel, default=7000, keys [[5405,0],[6169,7000]]}]}]
   }
   {
    type=MovieSceneFloatTrack
    display_name=Soft Source Radius
    sections [{type=MovieSceneFloatSection, channels [None {type=MovieSceneScriptingFloatChannel, default=1750, keys [[5405,0],[6169,1750]]}]}]
   }
  ]
 }
 SpotLight_PistolCanary_EnergyBall {
  id=#8
  possessed_class=SpotLight
  tracks [{
    type=MovieScene3DTransformTrack
    sections [{
      type=MovieScene3DTransformSection
      channels [
       Location.X {
        type=MovieSceneScriptingDoubleChannel
        default=163.9686
        thinned_from=6
        keys [[5405,14449.1895],[5970,14449.1895],[6011,13312.1895],[6087,13848.3223],[6169,13693.1895]]
       }
       Location.Y {
        type=MovieSceneScriptingDoubleChannel
        default=-1200.4724
        keys [[5405,69888.2344],[5970,71965.2344],[6011,72523.0703],[6087,65315.6602],[6123,71660.1484],[6169,80828.2344]]
       }
       Location.Z {
        type=MovieSceneScriptingDoubleChannel
        default=10115.4294
        keys [[5405,14791.4297],[5970,17325.3984],[6011,15827.3447],[6087,14590.5273],[6123,17527.0117],[6169,15249.4297]]
       }
       Rotation.X {type=MovieSceneScriptingDoubleChannel, default=-26.3821, const [0,6]}
       Rotation.Y {type=MovieSceneScriptingDoubleChannel, default=-70.5808, const [-90,6]}
       Rotation.Z {type=MovieSceneScriptingDoubleChannel, default=126.6419, const [126.6419,6]}
      ]
     }]
   }]
 }
 BP_JungleFog_LowHanging {
  id=#9
  possessed_class=BP_JungleFog_C
  tracks [{
    type=MovieSceneVisibilityTrack
    sections [{type=MovieSceneVisibilitySection, channels [None {type=MovieSceneScriptingBoolChannel, default=false}]}]
   }]
 }
 Capsule {
  id=#10
  parent_id=#1
  possessed_class=CapsuleComponent
  tracks [{
    type=MovieSceneFloatTrack
    display_name=Capsule Radius
    sections [{type=MovieSceneFloatSection, channels [None {type=MovieSceneScriptingFloatChannel, default=50000, keys [[3600,26000],[6000,50000]]}]}]
   }]
 }
 BP_PistolCanary_EndLightBeam {
  id=#11
  possessed_class=BP_PistolCanary_EndLightBeam_C
  tracks [{
    type=MovieSceneEventTrack
    sections [{
      type=MovieSceneEventTriggerSection
      channels [None {type=MovieSceneScriptingEventChannel, const ["<Struct 'MovieSceneEvent' (0x852ff01c940) {}>", 1]}]
     }]
   }]
 }
 DaySequenceModifierVolume_PistolCanary {
  id=#6
  possessed_class=DaySequenceModifierVolume
  tracks [{
    type=MovieScene3DTransformTrack
    sections [{
      type=MovieScene3DTransformSection
      channels [
       Location.X {type=MovieSceneScriptingDoubleChannel, default=14456}
       Location.Y {type=MovieSceneScriptingDoubleChannel, default=77488}
       Location.Z {type=MovieSceneScriptingDoubleChannel, default=3688}
       Rotation.X {type=MovieSceneScriptingDoubleChannel, default=0}
       Rotation.Y {type=MovieSceneScriptingDoubleChannel, default=0}
       Rotation.Z {type=MovieSceneScriptingDoubleChannel, default=0}
       Scale.X {type=MovieSceneScriptingDoubleChannel, default=1}
       Scale.Y {type=MovieSceneScriptingDoubleChannel, default=1}
       Scale.Z {type=MovieSceneScriptingDoubleChannel, default=1}
      ]
     }]
   }]
 }
 LightComponent0 {
  id=#12
  parent_id=#4
  possessed_class=SpotLightComponent
  tracks [{
    type=MovieSceneFloatTrack
    display_name=Intensity
    sections [{
      type=MovieSceneFloatSection
      channels [None {type=MovieSceneScriptingFloatChannel, default=120000, keys [[3696,0],[4660,120000],[7828,120000],[9046,0]]}]
     }]
   }]
 }
]
display_rate {denominator=1, numerator=60}
tick_resolution {denominator=1, numerator=24000}
tracks []
