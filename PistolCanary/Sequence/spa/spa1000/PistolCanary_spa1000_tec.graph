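# Level Sequence spa1000_tec for PistolCanary, driving a panel manager and placing sky effect actors for a timed sequence.
# Two event tracks trigger the PanelManager, one with two widely spaced events and one with a dense series of events between them.
# Transform tracks hold a lightning spiral mask mesh and BP_SpiralClouds at the same fixed location and scale.
# A material parameter collection track fades GlobalOpacity in PistolCanary_MPC to zero, back to full, then out again.
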LevelSequence /PistolCanary/Sequence/spa/spa1000/PistolCanary_spa1000_tec
bindings [
 PistolCanary_PanelManager {
  id=#1
  possessed_class=PistolCanary_PanelManager_C
  tracks [
   {
    type=MovieSceneEventTrack
    sections [{
      type=MovieSceneEventTriggerSection
      channels [None {
        type=MovieSceneScriptingEventChannel
        keys [[3600, "<Struct 'MovieSceneEvent' (0x852ff01c940) {}>"], [9288, "<Struct 'MovieSceneEvent' (0x852ff01e240) {}>"]]
       }]
     }]
   }
   {
    type=MovieSceneEventTrack
    sections [{
      type=MovieSceneEventTriggerSection
      channels [None {
        type=MovieSceneScriptingEventChannel
        keys [
         [4114, "<Struct 'MovieSceneEvent' (0x852ff01d980) {}>"]
         [4170, "<Struct 'MovieSceneEvent' (0x852ff01c120) {}>"]
         [4261, "<Struct 'MovieSceneEvent' (0x852ff01dca0) {}>"]
         [4300, "<Struct 'MovieSceneEvent' (0x852ff01e1a0) {}>"]
         [4315, "<Struct 'MovieSceneEvent' (0x852ff01e380) {}>"]
         [4356, "<Struct 'MovieSceneEvent' (0x852ff01de80) {}>"]
         [4503, "<Struct 'MovieSceneEvent' (0x852ff01ed80) {}>"]
         [4569, "<Struct 'MovieSceneEvent' (0x852ff01f1e0) {}>"]
         [4601, "<Struct 'MovieSceneEvent' (0x852ff01f3c0) {}>"]
         [4645, "<Struct 'MovieSceneEvent' (0x852ff01f500) {}>"]
         [4695, "<Struct 'MovieSceneEvent' (0x852ff01f780) {}>"]
         [6197, "<Struct 'MovieSceneEvent' (0x852ff01f8c0) {}>"]
         [6198, "<Struct 'MovieSceneEvent' (0x852ff01f960) {}>"]
         [6426, "<Struct 'MovieSceneEvent' (0x852ff01f640) {}>"]
         [6427, "<Struct 'MovieSceneEvent' (0x852ff01df20) {}>"]
         [6428, "<Struct 'MovieSceneEvent' (0x852ff01ce40) {}>"]
         [6898, "<Struct 'MovieSceneEvent' (0x852ff01e7e0) {}>"]
         [6899, "<Struct 'MovieSceneEvent' (0x852ff01e560) {}>"]
         [6900, "<Struct 'MovieSceneEvent' (0x852ff01dac0) {}>"]
         [6964, "<Struct 'MovieSceneEvent' (0x852ff01ea60) {}>"]
         [6965, "<Struct 'MovieSceneEvent' (0x852ff01ece0) {}>"]
         [7045, "<Struct 'MovieSceneEvent' (0x852ff01eec0) {}>"]
         [7046, "<Struct 'MovieSceneEvent' (0x852ff01f0a0) {}>"]
         [7770, "<Struct 'MovieSceneEvent' (0x852ff01eba0) {}>"]
         [7771, "<Struct 'MovieSceneEvent' (0x852ff01d160) {}>"]
         [7935, "<Struct 'MovieSceneEvent' (0x852ff01e9c0) {}>"]
        ]
       }]
     }]
   }
  ]
 }
 SM_MaskMaker_Lightning_3_Spirals_Segments2 {
  id=#2
  possessed_class=FortStaticMeshActor
  tracks [{
    type=MovieScene3DTransformTrack
    sections [{
      type=MovieScene3DTransformSection
      channels [
       Location.X {type=MovieSceneScriptingDoubleChannel, default=14198.0736}
       Location.Y {type=MovieSceneScriptingDoubleChannel, default=70555.3582}
       Location.Z {type=MovieSceneScriptingDoubleChannel, default=29321, const [29321,1]}
       Rotation.X {type=MovieSceneScriptingDoubleChannel, default=0}
       Rotation.Y {type=MovieSceneScriptingDoubleChannel, default=0}
       Rotation.Z {type=MovieSceneScriptingDoubleChannel, default=0}
       Scale.X {type=MovieSceneScriptingDoubleChannel, default=2, const [2,1]}
       Scale.Y {type=MovieSceneScriptingDoubleChannel, default=2, const [2,1]}
       Scale.Z {type=MovieSceneScriptingDoubleChannel, default=2, const [2,1]}
      ]
     }]
   }]
 }
 BP_SpiralClouds {
  id=#3
  possessed_class=BP_SpiralClouds_C
  tracks [{
    type=MovieScene3DTransformTrack
    sections [{
      type=MovieScene3DTransformSection
      channels [
       Location.X {type=MovieSceneScriptingDoubleChannel, default=14198.0736}
       Location.Y {type=MovieSceneScriptingDoubleChannel, default=70555.3582}
       Location.Z {type=MovieSceneScriptingDoubleChannel, default=29321, const [29321,1]}
       Rotation.X {type=MovieSceneScriptingDoubleChannel, default=0}
       Rotation.Y {type=MovieSceneScriptingDoubleChannel, default=0}
       Rotation.Z {type=MovieSceneScriptingDoubleChannel, default=0}
       Scale.X {type=MovieSceneScriptingDoubleChannel, default=2, const [2,1]}
       Scale.Y {type=MovieSceneScriptingDoubleChannel, default=2, const [2,1]}
       Scale.Z {type=MovieSceneScriptingDoubleChannel, default=2, const [2,1]}
      ]
     }]
   }]
 }
]
display_rate {denominator=1, numerator=60}
tick_resolution {denominator=1, numerator=24000}
tracks [{
  type=MovieSceneMaterialParameterCollectionTrack
  display_name=PistolCanary_MPC
  sections [{
    type=MovieSceneParameterSection
    channels [GlobalOpacity {type=MovieSceneScriptingFloatChannel, default=1, keys [[-15040,1.0242],[3600,0],[3677,0],[4660,1],[8163,1],[8869,0]]}]
   }]
 }]
