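# VFX level sequence for the PistolCanary spa1000 shot, timing effects for what appears to be a tentacle boss battle with monolith spire destruction.
# Keyed spawn, attach, transform and particle tracks drive Niagara lasers, energy balls, scorches, shockwaves and exit effects.
# Material parameter tracks pulse and recolor the boss's glow and flare spire cores just before each spire hides, with death effects firing.
# A post-process volume's blend weight is keyed across two windows.
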LevelSequence /PistolCanary/Sequence/spa/spa1000/PistolCanary_spa1000_vfx
bindings [
 NS_TentacleAttackChargeUpLaser3 {
  id=#1
  possessed_class=NiagaraActor
  tracks [{type=MovieScene3DAttachTrack, sections [{type=MovieScene3DAttachSection, end_frame=9420, start_frame=7201}]}]
 }
 PistolCanary_RockSpireMonolith1 {
  id=#2
  possessed_class=PistolCanary_Monolith_Spire_C
  tracks [
   {
    type=MovieSceneVisibilityTrack
    sections [{type=MovieSceneVisibilitySection, channels [None {type=MovieSceneScriptingBoolChannel, default=false, keys [[3613,true],[8460,false]]}]}]
   }
   {
    type=MovieSceneEventTrack
    sections [{
      type=MovieSceneEventTriggerSection
      channels [None {type=MovieSceneScriptingEventChannel, const ["<Struct 'MovieSceneEvent' (0x852ff01df20) {}>", 1]}]
     }]
   }
  ]
 }
 NiagaraComponent0 {
  id=#3
  parent_id=#4
  possessed_class=NiagaraComponent
  tracks [
   {
    type=MovieSceneBoolTrack
    display_name=Auto Activate
    sections [{type=MovieSceneBoolSection, channels [None {type=MovieSceneScriptingBoolChannel, default=false, keys [[9181,false],[9182,true]]}]}]
   }
   {
    type=MovieSceneParticleTrack
    display_name=FX System
    sections [{type=MovieSceneParticleSection, channels [None {type=MovieSceneScriptingParticleChannel, const ["<ParticleKey.TRIGGER: 2>", 1]}]}]
   }
  ]
 }
 SM_MaskMaker_Lightning_3_Spirals_Segments2 {
  id=#5
  possessed_class=FortStaticMeshActor
  tracks [{
    type=MovieSceneVisibilityTrack
    sections [{type=MovieSceneVisibilitySection, channels [None {type=MovieSceneScriptingBoolChannel, default=false, keys [[5597,false],[5598,true]]}]}]
   }]
 }
 NS_HumanEyeBeam {
  id=#6
  possessed_class=NiagaraActor
  tracks [
   {type=MovieScene3DAttachTrack, sections [{type=MovieScene3DAttachSection, end_frame=8894, start_frame=6811}]}
   {
    type=MovieSceneSpawnTrack
    sections [{
      type=MovieSceneSpawnSection
      channels [None {type=MovieSceneScriptingBoolChannel, default=false, keys [[6907,false],[7032,true],[7245,false],[8095,true],[8602,false]]}]
     }]
   }
  ]
 }
 NS_TripleTentacleAttackLaser3 {
  id=#7
  possessed_class=NiagaraActor
  tracks [
   {type=MovieScene3DAttachTrack, sections [{type=MovieScene3DAttachSection, end_frame=9690, start_frame=7578}]}
   {
    type=MovieSceneSpawnTrack
    sections [{
      type=MovieSceneSpawnSection
      channels [None {type=MovieSceneScriptingBoolChannel, default=false, keys [[3874,false],[7594,true],[7682,false]]}]
     }]
   }
   {
    type=MovieScene3DTransformTrack
    sections [{
      type=MovieScene3DTransformSection
      channels [
       Location.X {type=MovieSceneScriptingDoubleChannel, default=0, const [0,4]}
       Location.Y {type=MovieSceneScriptingDoubleChannel, default=0, const [0,4]}
       Location.Z {type=MovieSceneScriptingDoubleChannel, default=0, const [0,4]}
       Rotation.X {
        type=MovieSceneScriptingDoubleChannel
        default=179.0786
        keys [[3801,-87.2302],[7451,-100.502],[7458,-97.6913],[7595,-120.6461],[7601,-118.7311]]
       }
       Rotation.Y {
        type=MovieSceneScriptingDoubleChannel
        default=-56.8563
        keys [[3801,-22.8886],[7451,-19.4061],[7458,-20.6412],[7595,-13.968],[7601,-14.3464]]
       }
       Rotation.Z {
        type=MovieSceneScriptingDoubleChannel
        default=-182.8809
        keys [[3801,87.9642],[7451,129.1574],[7458,120.9617],[7595,129.7817],[7601,129.0724]]
       }
       Scale.X {type=MovieSceneScriptingDoubleChannel, default=15, const [15,5]}
       Scale.Y {type=MovieSceneScriptingDoubleChannel, default=15, const [15,5]}
       Scale.Z {type=MovieSceneScriptingDoubleChannel, default=30, const [30,5]}
      ]
     }]
   }
  ]
 }
 NS_PC_AirCharge_Exit {id=#8, possessed_class=NiagaraActor, tracks []}
 NS_TripleTentacleAttackLaser1 {
  id=#9
  possessed_class=NiagaraActor
  tracks [
   {
    type=MovieSceneSpawnTrack
    sections [{
      type=MovieSceneSpawnSection
      channels [None {type=MovieSceneScriptingBoolChannel, default=false, keys [[3773,false],[3799,true],[3874,false],[7450,true],[7496,false]]}]
     }]
   }
   {
    type=MovieScene3DTransformTrack
    sections [{
      type=MovieScene3DTransformSection
      channels [
       Location.X {type=MovieSceneScriptingDoubleChannel, default=0, const [0,2]}
       Location.Y {type=MovieSceneScriptingDoubleChannel, default=0, const [0,2]}
       Location.Z {type=MovieSceneScriptingDoubleChannel, default=0, const [0,2]}
       Rotation.X {type=MovieSceneScriptingDoubleChannel, default=179.0786, keys [[3801,-87.2302],[7451,-100.502],[7458,-97.6913]]}
       Rotation.Y {type=MovieSceneScriptingDoubleChannel, default=-56.8563, keys [[3801,-22.8886],[7451,-19.4061],[7458,-20.6412]]}
       Rotation.Z {type=MovieSceneScriptingDoubleChannel, default=-182.8809, keys [[3801,87.9642],[7451,129.1574],[7458,120.9617]]}
       Scale.X {type=MovieSceneScriptingDoubleChannel, default=15, const [15,3]}
       Scale.Y {type=MovieSceneScriptingDoubleChannel, default=15, const [15,3]}
       Scale.Z {type=MovieSceneScriptingDoubleChannel, default=30, const [30,3]}
      ]
     }]
   }
   {type=MovieScene3DAttachTrack, sections [{type=MovieScene3DAttachSection, end_frame=9559, start_frame=7447}]}
  ]
 }
 NS_PistolCanary_InnerSpire_Glow_Loop_E {
  id=#10
  possessed_class=NiagaraActor
  tracks [{
    type=MovieSceneVisibilityTrack
    sections [{type=MovieSceneVisibilitySection, channels [None {type=MovieSceneScriptingBoolChannel, default=false, keys [[3613,true],[8166,false]]}]}]
   }]
 }
 NiagaraComponent0 {
  id=#11
  parent_id=#12
  possessed_class=NiagaraComponent
  tracks [
   {
    type=MovieSceneBoolTrack
    display_name=Auto Activate
    sections [{type=MovieSceneBoolSection, channels [None {type=MovieSceneScriptingBoolChannel, default=false, const [false,1]}]}]
   }
   {
    type=MovieSceneParticleTrack
    display_name=FX System
    sections [{type=MovieSceneParticleSection, channels [None {type=MovieSceneScriptingParticleChannel, const ["<ParticleKey.TRIGGER: 2>", 1]}]}]
   }
  ]
 }
 NS_TentacleEnergyBall {
  id=#13
  possessed_class=NiagaraActor
  tracks [
   {
    type=MovieScene3DTransformTrack
    sections [{
      type=MovieScene3DTransformSection
      channels [
       Location.X {
        type=MovieSceneScriptingDoubleChannel
        default=605.1021
        keys [[5473,605.1021],[5776,-628.8962],[5961,-547.1798],[6028,-323.314],[6083,32.0457],[6111,-398.8492],[6139,-123.4051],[6158,-2330.5181],[6183,-1628.9611],[6191,-1385.2246],[6193,-1194.1401],[6194,13071.0342],[7305,13348.8994]]
       }
       Location.Y {
        type=MovieSceneScriptingDoubleChannel
        default=-384.7639
        keys [[5473,-384.7639],[5776,-17.6894],[5961,208.1339],[6028,41.2671],[6083,-280.5915],[6111,-343.6378],[6139,-454.3257],[6158,-261.047],[6183,-654.7165],[6191,-624.466],[6193,-611.2116],[6194,76510.1953],[6195,77654.1953],[6199,79689],[7305,-3371.6941]]
       }
       Location.Z {
        type=MovieSceneScriptingDoubleChannel
        default=-907.5726
        keys [[5473,-907.5726],[5776,-1168.5436],[5961,-1455.371],[6028,-1858.7015],[6083,-1508.6317],[6111,-1284.0985],[6139,-1517.4601],[6158,-674.4239],[6183,-514.2402],[6191,-943.0718],[6193,-1127.0137],[6194,16367.0801],[6195,16759.0801],[6199,16572],[7305,-8453.8867]]
       }
       Rotation.X {type=MovieSceneScriptingDoubleChannel, default=0}
       Rotation.Y {type=MovieSceneScriptingDoubleChannel, default=0}
       Rotation.Z {type=MovieSceneScriptingDoubleChannel, default=0}
       Scale.X {type=MovieSceneScriptingDoubleChannel, default=1}
       Scale.Y {type=MovieSceneScriptingDoubleChannel, default=1}
       Scale.Z {type=MovieSceneScriptingDoubleChannel, default=1}
      ]
     }]
   }
   {type=MovieScene3DAttachTrack, sections [{type=MovieScene3DAttachSection, end_frame=6194, start_frame=5375}]}
  ]
 }
 StaticMeshComponent0 {
  id=#14
  parent_id=#2
  possessed_class=StaticMeshComponent
  tracks [
   {
    type=MovieSceneComponentMaterialTrack
    display_name=Material Slot: top_FB_001
    sections [{
      type=MovieSceneComponentMaterialParameterSection
      end_frame=9586
      start_frame=3600
      channels [
       Core Intensity {type=MovieSceneScriptingFloatChannel, default=200, keys [[8416,200],[8457,1000]]}
       Core ColorR {type=MovieSceneScriptingFloatChannel, keys [[8416,0.3613],[8457,1]]}
       Core ColorG {type=MovieSceneScriptingFloatChannel, const [0,2]}
       Core ColorB {type=MovieSceneScriptingFloatChannel, keys [[8416,0.3005],[8457,0.0612]]}
       Core ColorA {type=MovieSceneScriptingFloatChannel, const [1,2]}
      ]
     }]
   }
   {
    type=MovieScenePrimitiveMaterialTrack
    display_name=Material Slot: top_FB_001
    sections [{
      type=MovieScenePrimitiveMaterialSection
      channels [None {
        type=MovieSceneScriptingObjectPathChannel
        default="<Object '/Hermes/Rocks/MMSpikeyRocks/Materials/MaskMakerEvent/MI_Hermes_Rocks_MaskMaker_MonolithTop_B_A_Event' (0x852ee93e800) Class 'MaterialInstanceConstant'>"
       }]
     }]
   }
   {
    type=MovieSceneComponentMaterialTrack
    display_name=Material Slot: Bottom_FB
    sections [{
      type=MovieSceneComponentMaterialParameterSection
      end_frame=9586
      start_frame=3600
      channels [
       Core Intensity {type=MovieSceneScriptingFloatChannel, default=200, keys [[8416,200],[8457,1000]]}
       Core ColorR {type=MovieSceneScriptingFloatChannel, keys [[8416,0.3613],[8457,1]]}
       Core ColorG {type=MovieSceneScriptingFloatChannel, const [0,2]}
       Core ColorB {type=MovieSceneScriptingFloatChannel, keys [[8416,0.3005],[8457,0.0612]]}
       Core ColorA {type=MovieSceneScriptingFloatChannel, const [1,2]}
      ]
     }]
   }
   {
    type=MovieScenePrimitiveMaterialTrack
    display_name=Material Slot: Bottom_FB
    sections [{
      type=MovieScenePrimitiveMaterialSection
      channels [None {
        type=MovieSceneScriptingObjectPathChannel
        default="<Object '/Hermes/Rocks/MMSpikeyRocks/Materials/MaskMakerEvent/MI_Hermes_Rocks_MaskMaker_MonolithBottom_A_A_Event' (0x852ee93f800) Class 'MaterialInstanceConstant'>"
       }]
     }]
   }
  ]
 }
 NS_LargeShockwave_2 {
  id=#15
  possessed_class=NiagaraActor
  tracks [
   {
    type=MovieSceneSpawnTrack
    sections [{type=MovieSceneSpawnSection, channels [None {type=MovieSceneScriptingBoolChannel, default=false, keys [[3503,false],[5225,true]]}]}]
   }
   {
    type=MovieScene3DTransformTrack
    sections [{
      type=MovieScene3DTransformSection
      channels [
       Location.X {type=MovieSceneScriptingDoubleChannel, default=15649.0786}
       Location.Y {type=MovieSceneScriptingDoubleChannel, default=68716.0859}
       Location.Z {type=MovieSceneScriptingDoubleChannel, default=8968}
       Rotation.X {type=MovieSceneScriptingDoubleChannel, default=0}
       Rotation.Y {type=MovieSceneScriptingDoubleChannel, default=0}
       Rotation.Z {type=MovieSceneScriptingDoubleChannel, default=0}
       Scale.X {type=MovieSceneScriptingDoubleChannel, default=1}
       Scale.Y {type=MovieSceneScriptingDoubleChannel, default=1}
       Scale.Z {type=MovieSceneScriptingDoubleChannel, default=1}
      ]
     }]
   }
  ]
 }
 NS_PistolCanary_InnerSpire_Glow_Loop_N {
  id=#16
  possessed_class=NiagaraActor
  tracks [{
    type=MovieSceneVisibilityTrack
    sections [{type=MovieSceneVisibilitySection, channels [None {type=MovieSceneScriptingBoolChannel, default=false, keys [[3613,true],[8579,false]]}]}]
   }]
 }
 NS_PistolCanaryCD_Connecting_Spire_Glow_E {
  id=#17
  possessed_class=NiagaraActor
  tracks [{
    type=MovieSceneVisibilityTrack
    sections [{type=MovieSceneVisibilitySection, channels [None {type=MovieSceneScriptingBoolChannel, default=false, const [false,1]}]}]
   }]
 }
 NS_PistolCanaryCD_Connecting_Spire_Glow_W {
  id=#18
  possessed_class=NiagaraActor
  tracks [{
    type=MovieSceneVisibilityTrack
    sections [{type=MovieSceneVisibilitySection, channels [None {type=MovieSceneScriptingBoolChannel, default=false, const [false,1]}]}]
   }]
 }
 NiagaraComponent0 {
  id=#19
  parent_id=#8
  possessed_class=NiagaraComponent
  tracks [
   {
    type=MovieSceneBoolTrack
    display_name=Auto Activate
    sections [{type=MovieSceneBoolSection, channels [None {type=MovieSceneScriptingBoolChannel, default=false, keys [[8958,false],[9128,true]]}]}]
   }
   {type=MovieScene3DAttachTrack, sections [{type=MovieScene3DAttachSection, end_frame=9317, start_frame=9096}]}
   {
    type=MovieSceneParticleTrack
    display_name=FX System
    sections [{
      type=MovieSceneParticleSection
      channels [None {type=MovieSceneScriptingParticleChannel, keys [[9129, "<ParticleKey.TRIGGER: 2>"], [9294.9975, "<ParticleKey.DEACTIVATE: 1>"]]}]
     }]
   }
  ]
 }
 NiagaraComponent0 {
  id=#20
  parent_id=#21
  possessed_class=NiagaraComponent
  tracks [
   {
    type=MovieSceneNiagaraFloatParameterTrack
    display_name=BeamWidth
    sections [{type=MovieSceneFloatSection, channels [None {type=MovieSceneScriptingFloatChannel, default=0, keys [[7551,1],[7558,0]]}]}]
   }
   {
    type=MovieSceneParticleTrack
    display_name=FX System
    sections [{
      type=MovieSceneParticleSection
      channels [None {type=MovieSceneScriptingParticleChannel, keys [[7527, "<ParticleKey.TRIGGER: 2>"], [7564, "<ParticleKey.DEACTIVATE: 1>"]]}]
     }]
   }
  ]
 }
 SkeletalMesh {
  id=#22
  parent_id=#23
  possessed_class=SkeletalMeshComponent
  tracks [
   {
    type=MovieSceneComponentMaterialTrack
    display_name=Material Slot: Body_A
    sections [{
      type=MovieSceneComponentMaterialParameterSection
      end_frame=9210
      start_frame=3495
      channels [
       PulseSpeed {type=MovieSceneScriptingFloatChannel, default=0.2, thinned_from=6, keys [[5254,0.1],[5258,0],[6393,0],[6400,0.2]]}
       PulseManualOffset {type=MovieSceneScriptingFloatChannel, default=0, keys [[5324,0],[5655,0.6]]}
       PulseRadius {type=MovieSceneScriptingFloatChannel, default=0.33, keys [[5258,0.33],[6176,1],[6393,0.3]]}
       GlowPulseEXP {type=MovieSceneScriptingFloatChannel, default=0.03, keys [[5572,0.015],[5766,0.03],[6176,0.03],[6393,0.015]]}
       EmissiveGlowColorR {
        type=MovieSceneScriptingFloatChannel
        default=8
        keys [[5215,8],[5254,0],[5261,0],[5324,8],[5877,0.7],[6180,0.7],[6291,8],[6393,0],[6400,0],[6433,8]]
       }
       EmissiveGlowColorG {
        type=MovieSceneScriptingFloatChannel
        default=0.3176
        keys [[5215,0.3176],[5254,0],[5261,0],[5324,0.3176],[5877,3],[6180,3],[6291,0.3176],[6393,0],[6400,0],[6433,0.3176]]
       }
       EmissiveGlowColorB {type=MovieSceneScriptingFloatChannel, default=0, keys [[5254,0],[5324,0],[5877,10],[6180,10],[6291,0]]}
       EmissiveGlowColorA {type=MovieSceneScriptingFloatChannel, default=1}
      ]
     }]
   }
   {
    type=MovieSceneComponentMaterialTrack
    display_name=Material Slot: Body_B
    sections [{
      type=MovieSceneComponentMaterialParameterSection
      end_frame=9315
      start_frame=3600
      channels [
       PulseSpeed {type=MovieSceneScriptingFloatChannel, default=0, thinned_from=6, keys [[5254,0.1],[5258,0],[6393,0],[6400,0.2]]}
       PulseManualOffset {type=MovieSceneScriptingFloatChannel, default=0, keys [[5324,0],[5655,0.6]]}
       PulseRadius {type=MovieSceneScriptingFloatChannel, default=0.33, keys [[5258,0.33],[6176,1],[6393,0.3]]}
       GlowPulseEXP {type=MovieSceneScriptingFloatChannel, default=0.03, keys [[5572,0.015],[5766,0.03],[6176,0.03],[6393,0.015]]}
       EmissiveGlowColorR {
        type=MovieSceneScriptingFloatChannel
        default=8
        keys [[5215,8],[5254,0],[5261,0],[5324,8],[5877,0.7],[6180,0.7],[6291,8],[6393,0],[6400,0],[6433,8]]
       }
       EmissiveGlowColorG {
        type=MovieSceneScriptingFloatChannel
        default=0.3176
        keys [[5215,0.3176],[5254,0],[5261,0],[5324,0.3176],[5877,3],[6180,3],[6291,0.3176],[6393,0],[6400,0],[6433,0.3176]]
       }
       EmissiveGlowColorB {type=MovieSceneScriptingFloatChannel, default=0, keys [[5254,0],[5324,0],[5877,10],[6180,10],[6291,0]]}
       EmissiveGlowColorA {type=MovieSceneScriptingFloatChannel, default=1}
      ]
     }]
   }
   {
    type=MovieSceneComponentMaterialTrack
    display_name=Material Slot: Body_C
    sections [{
      type=MovieSceneComponentMaterialParameterSection
      end_frame=9315
      start_frame=3600
      channels [
       PulseSpeed {type=MovieSceneScriptingFloatChannel, default=0, thinned_from=6, keys [[5254,0.1],[5258,0],[6393,0],[6400,0.2]]}
       PulseManualOffset {type=MovieSceneScriptingFloatChannel, default=0.7, keys [[5324,0],[5655,0.6]]}
       PulseRadius {type=MovieSceneScriptingFloatChannel, default=0.3, keys [[5258,0.33],[6176,1],[6393,0.3]]}
       GlowPulseEXP {type=MovieSceneScriptingFloatChannel, default=0.03, keys [[5572,0.015],[5766,0.03],[6176,0.03],[6393,0.015]]}
       EmissiveGlowColorR {
        type=MovieSceneScriptingFloatChannel
        default=0
        keys [[5215,8],[5254,0],[5261,0],[5324,8],[5877,0.7],[6180,0.7],[6291,8],[6393,0],[6400,0],[6433,8]]
       }
       EmissiveGlowColorG {
        type=MovieSceneScriptingFloatChannel
        default=0
        keys [[5215,0.3176],[5254,0],[5261,0],[5324,0.3176],[5877,3],[6180,3],[6291,0.3176],[6393,0],[6400,0],[6433,0.3176]]
       }
       EmissiveGlowColorB {type=MovieSceneScriptingFloatChannel, default=10, keys [[5254,0],[5324,0],[5877,10],[6180,10],[6291,0]]}
       EmissiveGlowColorA {type=MovieSceneScriptingFloatChannel, default=1}
      ]
     }]
   }
   {
    type=MovieSceneComponentMaterialTrack
    display_name=Material Slot: Body_D
    sections [{
      type=MovieSceneComponentMaterialParameterSection
      end_frame=9315
      start_frame=3600
      channels [
       PulseSpeed {type=MovieSceneScriptingFloatChannel, default=0, thinned_from=6, keys [[5254,0.1],[5258,0],[6393,0],[6400,0.2]]}
       PulseManualOffset {type=MovieSceneScriptingFloatChannel, default=0, keys [[5324,0],[5655,0.6]]}
       PulseRadius {type=MovieSceneScriptingFloatChannel, default=0.3399, keys [[5258,0.33],[6176,1],[6393,0.3]]}
       GlowPulseEXP {type=MovieSceneScriptingFloatChannel, default=0.03, keys [[5572,0.015],[5766,0.03],[6176,0.03],[6393,0.015]]}
       EmissiveGlowColorR {
        type=MovieSceneScriptingFloatChannel
        default=8
        keys [[5215,8],[5254,0],[5261,0],[5324,8],[5877,0.7],[6180,0.7],[6291,8],[6393,0],[6400,0],[6433,8]]
       }
       EmissiveGlowColorG {
        type=MovieSceneScriptingFloatChannel
        default=0.3176
        keys [[5215,0.3176],[5254,0],[5261,0],[5324,0.3176],[5877,3],[6180,3],[6291,0.3176],[6393,0],[6400,0],[6433,0.3176]]
       }
       EmissiveGlowColorB {type=MovieSceneScriptingFloatChannel, default=0, keys [[5254,0],[5324,0],[5877,10],[6180,10],[6291,0]]}
       EmissiveGlowColorA {type=MovieSceneScriptingFloatChannel, default=1}
      ]
     }]
   }
  ]
 }
 NS_Pistol_Canary_WealthLamb_Punch {
  id=#24
  possessed_class=NiagaraActor
  tracks [
   {
    type=MovieSceneSpawnTrack
    sections [{
      type=MovieSceneSpawnSection
      channels [None {type=MovieSceneScriptingBoolChannel, default=false, keys [[7675,false],[7772,true],[8038,false]]}]
     }]
   }
   {
    type=MovieScene3DTransformTrack
    sections [{
      type=MovieScene3DTransformSection
      channels [
       Location.X {type=MovieSceneScriptingDoubleChannel, default=11605}
       Location.Y {type=MovieSceneScriptingDoubleChannel, default=67215}
       Location.Z {type=MovieSceneScriptingDoubleChannel, default=13281}
       Rotation.X {type=MovieSceneScriptingDoubleChannel, default=0}
       Rotation.Y {type=MovieSceneScriptingDoubleChannel, default=0}
       Rotation.Z {type=MovieSceneScriptingDoubleChannel, default=0}
       Scale.X {type=MovieSceneScriptingDoubleChannel, default=1}
       Scale.Y {type=MovieSceneScriptingDoubleChannel, default=1}
       Scale.Z {type=MovieSceneScriptingDoubleChannel, default=1}
      ]
     }]
   }
  ]
 }
 NS_TentacleAttackChargeUpLaser3 {
  id=#25
  possessed_class=NiagaraActor
  tracks [
   {
    type=MovieSceneSpawnTrack
    sections [{type=MovieSceneSpawnSection, channels [None {type=MovieSceneScriptingBoolChannel, default=false, const [false,3]}]}]
   }
   {type=MovieScene3DAttachTrack, sections [{type=MovieScene3DAttachSection, end_frame=14157, start_frame=5654}]}
  ]
 }
 NS_PC_WealthLamb_Exit {id=#4, possessed_class=NiagaraActor, tracks []}
 NiagaraComponent0 {
  id=#26
  parent_id=#27
  possessed_class=NiagaraComponent
  tracks [{
    type=MovieSceneParticleTrack
    display_name=FX System
    sections [{
      type=MovieSceneParticleSection
      channels [None {type=MovieSceneScriptingParticleChannel, keys [[7527, "<ParticleKey.TRIGGER: 2>"], [7607, "<ParticleKey.DEACTIVATE: 1>"]]}]
     }]
   }]
 }
 StaticMeshComponent0 {
  id=#28
  parent_id=#29
  possessed_class=StaticMeshComponent
  tracks [
   {
    type=MovieSceneComponentMaterialTrack
    display_name=Material Slot: Bottom_FB
    sections [{
      type=MovieSceneComponentMaterialParameterSection
      end_frame=9586
      start_frame=3600
      channels [
       Core Intensity {type=MovieSceneScriptingFloatChannel, default=200, keys [[8282,200],[8323,1000]]}
       Core ColorR {type=MovieSceneScriptingFloatChannel, keys [[8282,0.3613],[8323,1]]}
       Core ColorG {type=MovieSceneScriptingFloatChannel, const [0,2]}
       Core ColorB {type=MovieSceneScriptingFloatChannel, keys [[8282,0.3005],[8323,0.0612]]}
       Core ColorA {type=MovieSceneScriptingFloatChannel, const [1,2]}
      ]
     }]
   }
   {
    type=MovieScenePrimitiveMaterialTrack
    display_name=Material Slot: Bottom_FB
    sections [{
      type=MovieScenePrimitiveMaterialSection
      channels [None {
        type=MovieSceneScriptingObjectPathChannel
        default="<Object '&MI_Hermes_Rocks_MaskMaker_MonolithBottom_A_A_Event' (0x852ee93f800) Class 'MaterialInstanceConstant'>"
       }]
     }]
   }
   {
    type=MovieSceneComponentMaterialTrack
    display_name=Material Slot: top_FB_001
    sections [{
      type=MovieSceneComponentMaterialParameterSection
      end_frame=9586
      start_frame=3600
      channels [
       Core Intensity {type=MovieSceneScriptingFloatChannel, default=200, keys [[8282,200],[8323,1000]]}
       Core ColorR {type=MovieSceneScriptingFloatChannel, keys [[8282,0.3613],[8323,1]]}
       Core ColorG {type=MovieSceneScriptingFloatChannel, const [0,2]}
       Core ColorB {type=MovieSceneScriptingFloatChannel, keys [[8282,0.3005],[8323,0.0612]]}
       Core ColorA {type=MovieSceneScriptingFloatChannel, const [1,2]}
      ]
     }]
   }
   {
    type=MovieScenePrimitiveMaterialTrack
    display_name=Material Slot: top_FB_001
    sections [{
      type=MovieScenePrimitiveMaterialSection
      channels [None {
        type=MovieSceneScriptingObjectPathChannel
        default="<Object '&MI_Hermes_Rocks_MaskMaker_MonolithTop_B_A_Event' (0x852ee93e800) Class 'MaterialInstanceConstant'>"
       }]
     }]
   }
  ]
 }
 NS_LargeSpire_DeathFX_E {id=#30, possessed_class=NiagaraActor, tracks []}
 NiagaraComponent0 {
  id=#31
  parent_id=#32
  possessed_class=NiagaraComponent
  tracks [
   {type=MovieSceneNiagaraSystemTrack, sections [{type=MovieSceneNiagaraSystemSpawnSection, end_frame=8665, start_frame=8579}]}
   {
    type=MovieSceneBoolTrack
    display_name=Auto Activate
    sections [{
      type=MovieSceneBoolSection
      channels [None {type=MovieSceneScriptingBoolChannel, default=false, keys [[8558,false],[8579,true],[8642,false]]}]
     }]
   }
   {
    type=MovieSceneParticleTrack
    display_name=FX System
    sections [{
      type=MovieSceneParticleSection
      channels [None {type=MovieSceneScriptingParticleChannel, keys [[8579, "<ParticleKey.TRIGGER: 2>"], [8642, "<ParticleKey.DEACTIVATE: 1>"]]}]
     }]
   }
  ]
 }
 NS_PistolCanary_Tentacle_Scorch2 {
  id=#33
  possessed_class=NiagaraActor
  tracks [
   {
    type=MovieSceneSpawnTrack
    sections [{
      type=MovieSceneSpawnSection
      channels [None {type=MovieSceneScriptingBoolChannel, default=false, keys [[6959.8525,false],[7172.2125,true],[7386.4075,false]]}]
     }]
   }
   {
    type=MovieScene3DTransformTrack
    sections [{
      type=MovieScene3DTransformSection
      channels [
       Location.X {
        type=MovieSceneScriptingDoubleChannel
        default=0
        keys [[7172.2125,-1119.111],[7183.265,-89.2651],[7193.92,1265.3802],[7204.47,2600.251],[7219,4152.3779]]
       }
       Location.Y {
        type=MovieSceneScriptingDoubleChannel
        default=-229.6212
        keys [[7172.2125,-921.9854],[7183.265,-393.6199],[7193.92,140.8181],[7204.47,108.1827],[7219,-142.3456]]
       }
       Location.Z {
        type=MovieSceneScriptingDoubleChannel
        default=-1105.1192
        keys [[7172.2125,-691.9553],[7183.265,-903.0063],[7193.92,-712.1248],[7204.47,-296.2354],[7219,111.7097]]
       }
       Rotation.X {
        type=MovieSceneScriptingDoubleChannel
        default=0
        keys [[7172.2125,1.2939],[7183.265,-8.6712],[7193.92,-66.8881],[7204.47,-102.6874],[7219,-116.3656]]
       }
       Rotation.Y {
        type=MovieSceneScriptingDoubleChannel
        default=-74.5839
        keys [[7172.2125,-45.7494],[7183.265,-61.4034],[7193.92,-69.8426],[7204.47,-61.5031],[7219,-49.4755]]
       }
       Rotation.Z {
        type=MovieSceneScriptingDoubleChannel
        default=0
        keys [[7172.2125,268.9213],[7183.265,277.6387],[7193.92,337.8673],[7204.47,370.8456],[7219,379.3134]]
       }
       Scale.X {type=MovieSceneScriptingDoubleChannel, default=1}
       Scale.Y {type=MovieSceneScriptingDoubleChannel, default=1}
       Scale.Z {type=MovieSceneScriptingDoubleChannel, default=1}
      ]
     }]
   }
   {type=MovieScene3DAttachTrack, sections [{type=MovieScene3DAttachSection, end_frame=7419, start_frame=7170}]}
  ]
 }
 NS_TentacleAttackImpact4 {
  id=#34
  possessed_class=NiagaraActor
  tracks [{
    type=MovieScene3DTransformTrack
    sections [{
      type=MovieScene3DTransformSection
      channels [
       Location.X {type=MovieSceneScriptingDoubleChannel, default=6560, const [6560,1]}
       Location.Y {type=MovieSceneScriptingDoubleChannel, default=69376, const [69376,1]}
       Location.Z {type=MovieSceneScriptingDoubleChannel, default=3604, const [3604,1]}
       Rotation.X {type=MovieSceneScriptingDoubleChannel, default=0, const [0,1]}
       Rotation.Y {type=MovieSceneScriptingDoubleChannel, default=0, const [0,1]}
       Rotation.Z {type=MovieSceneScriptingDoubleChannel, default=0, const [0,1]}
       Scale.X {type=MovieSceneScriptingDoubleChannel, default=1, const [1,1]}
       Scale.Y {type=MovieSceneScriptingDoubleChannel, default=1, const [1,1]}
       Scale.Z {type=MovieSceneScriptingDoubleChannel, default=1, const [1,1]}
      ]
     }]
   }]
 }
 NiagaraComponent0 {
  id=#35
  parent_id=#30
  possessed_class=NiagaraComponent
  tracks [
   {
    type=MovieSceneBoolTrack
    display_name=Auto Activate
    sections [{
      type=MovieSceneBoolSection
      channels [None {type=MovieSceneScriptingBoolChannel, default=false, keys [[8158,false],[8166,true],[8293,false]]}]
     }]
   }
   {
    type=MovieSceneParticleTrack
    display_name=FX System
    sections [{
      type=MovieSceneParticleSection
      channels [None {type=MovieSceneScriptingParticleChannel, keys [[8166, "<ParticleKey.TRIGGER: 2>"], [8263, "<ParticleKey.DEACTIVATE: 1>"]]}]
     }]
   }
  ]
 }
 NS_PistolCanary_Tentacle_Scorch {
  id=#36
  possessed_class=NiagaraActor
  tracks [
   {
    type=MovieSceneSpawnTrack
    sections [{
      type=MovieSceneSpawnSection
      channels [None {type=MovieSceneScriptingBoolChannel, default=false, keys [[6959.8525,false],[6991,true],[7386.4075,false]]}]
     }]
   }
   {
    type=MovieScene3DTransformTrack
    sections [{
      type=MovieScene3DTransformSection
      channels [
       Location.X {
        type=MovieSceneScriptingDoubleChannel
        default=0
        keys [[6991,615.6737],[7038.5275,556.5635],[7065.6375,144.3839],[7087.8625,-62.2073],[7111.015,107.8554],[7119.4675,350.1977],[7127.92,-7.6155],[7146.855,67.833],[7172.2125,1194.0696]]
       }
       Location.Y {
        type=MovieSceneScriptingDoubleChannel
        default=-229.6212
        keys [[6991,-716.4096],[7038.5275,-878.7684],[7065.6375,297.5991],[7087.8625,623.856],[7111.015,1202.4445],[7119.4675,889.5947],[7127.92,-563.1998],[7146.855,-1186.9008],[7172.2125,-971.7155]]
       }
       Location.Z {
        type=MovieSceneScriptingDoubleChannel
        default=-1105.1192
        keys [[6991,-892.1871],[7038.5275,-827.0784],[7065.6375,-1089.0811],[7087.8625,-1014.033],[7111.015,-437.7097],[7119.4675,615.0024],[7127.92,1012.6827],[7146.855,109.0112],[7172.2125,-626.6472]]
       }
       Rotation.X {
        type=MovieSceneScriptingDoubleChannel
        default=0
        keys [[6991,84.1341],[7038.5275,89.2622],[7065.6375,-90.1724],[7087.8625,-85.8719],[7111.015,-87.8372],[7119.4675,-85.5793],[7127.92,62.1924],[7146.855,94.7681],[7172.2125,-10.4453]]
       }
       Rotation.Y {
        type=MovieSceneScriptingDoubleChannel
        default=-74.5839
        keys [[6991,-65.7672],[7038.5275,-57.8725],[7065.6375,-70.4161],[7087.8625,-64.2937],[7111.015,-7.9952],[7119.4675,58.9989],[7127.92,68.9209],[7146.855,14.1917],[7172.2125,-53.4216]]
       }
       Rotation.Z {
        type=MovieSceneScriptingDoubleChannel
        default=0
        keys [[6991,-91.7231],[7038.5275,-94.3008],[7065.6375,87.1175],[7087.8625,86.6907],[7111.015,90.158],[7119.4675,94.1626],[7127.92,241.0106],[7146.855,253.9826],[7172.2125,283.4304]]
       }
       Scale.X {type=MovieSceneScriptingDoubleChannel, default=1}
       Scale.Y {type=MovieSceneScriptingDoubleChannel, default=1}
       Scale.Z {type=MovieSceneScriptingDoubleChannel, default=1}
      ]
     }]
   }
   {type=MovieScene3DAttachTrack, sections [{type=MovieScene3DAttachSection, end_frame=7419, start_frame=6958}]}
  ]
 }
 NiagaraComponent0 {
  id=#37
  parent_id=#36
  possessed_class=NiagaraComponent
  tracks [
   {
    type=MovieSceneParticleTrack
    display_name=FX System
    sections [{
      type=MovieSceneParticleSection
      channels [None {
        type=MovieSceneScriptingParticleChannel
        keys [[6902.66, "<ParticleKey.DEACTIVATE: 1>"], [7047.4825, "<ParticleKey.ACTIVATE: 0>"], [7172.2125, "<ParticleKey.DEACTIVATE: 1>"]]
       }]
     }]
   }
   {
    type=MovieSceneNiagaraFloatParameterTrack
    display_name=Scale
    sections [{
      type=MovieSceneFloatSection
      channels [None {type=MovieSceneScriptingFloatChannel, default=1, keys [[7047.4825,0],[7111.015,1],[7172.2125,0.2]]}]
     }]
   }
  ]
 }
 NiagaraComponent0 {
  id=#38
  parent_id=#33
  possessed_class=NiagaraComponent
  tracks [
   {
    type=MovieSceneParticleTrack
    display_name=FX System
    sections [{
      type=MovieSceneParticleSection
      channels [None {
        type=MovieSceneScriptingParticleChannel
        keys [[6902.66, "<ParticleKey.DEACTIVATE: 1>"], [7172.2125, "<ParticleKey.ACTIVATE: 0>"], [7219, "<ParticleKey.DEACTIVATE: 1>"]]
       }]
     }]
   }
   {
    type=MovieSceneNiagaraFloatParameterTrack
    display_name=Scale
    sections [{
      type=MovieSceneFloatSection
      channels [None {type=MovieSceneScriptingFloatChannel, default=0.5, keys [[7172.2125,0],[7186.0475,0.5],[7219,0]]}]
     }]
   }
   {
    type=MovieSceneNiagaraFloatParameterTrack
    display_name=SpawnRateMultiply
    sections [{type=MovieSceneFloatSection, channels [None {type=MovieSceneScriptingFloatChannel, default=3, const [3,1]}]}]
   }
  ]
 }
 NS_PistolCanary_InnerSpire_Glow_Loop_W {
  id=#39
  possessed_class=NiagaraActor
  tracks [{
    type=MovieSceneVisibilityTrack
    sections [{type=MovieSceneVisibilitySection, channels [None {type=MovieSceneScriptingBoolChannel, default=false, keys [[3613,true],[8460,false]]}]}]
   }]
 }
 TripleTentacleRight {id=#40, possessed_class=PistolCanary_TentacleBoss_C, tracks []}
 NS_TentacleAttackChargeUpLaser {
  id=#41
  possessed_class=NiagaraActor
  tracks [
   {
    type=MovieSceneSpawnTrack
    sections [{type=MovieSceneSpawnSection, channels [None {type=MovieSceneScriptingBoolChannel, default=false, const [false,4]}]}]
   }
   {type=MovieScene3DAttachTrack, sections [{type=MovieScene3DAttachSection, end_frame=14157, start_frame=5654}]}
  ]
 }
 PPV_PistolCanary_VFX {
  id=#42
  possessed_class=PostProcessVolume
  tracks [
   {
    type=MovieSceneFloatTrack
    display_name=Blend Weight
    sections [{
      type=MovieSceneFloatSection
      channels [None {type=MovieSceneScriptingFloatChannel, default=0, keys [[5426,0],[5501,1],[6214,1],[6251,0],[7250,0],[7338,1],[7799,1],[7877,0]]}]
     }]
   }
   {
    type=MovieSceneFloatTrack
    display_name="Intensity (Settings)"
    sections [{type=MovieSceneFloatSection, channels [None {type=MovieSceneScriptingFloatChannel, default=2, keys [[5433,0.675],[5988,2]]}]}]
   }
   {
    type=MovieSceneFloatTrack
    display_name="Intensity (Settings)"
    sections [{type=MovieSceneFloatSection, channels [None {type=MovieSceneScriptingFloatChannel, default=3, keys [[5431,1.5],[5993,3]]}]}]
   }
  ]
 }
 NiagaraComponent0 {
  id=#43
  parent_id=#44
  possessed_class=NiagaraComponent
  tracks [
   {
    type=MovieSceneNiagaraVectorParameterTrack
    display_name=Stop Location
    sections [{
      type=MovieSceneFloatVectorSection
      channels [
       Vector.X {type=MovieSceneScriptingFloatChannel, default=14122, keys [[5463,14347.3086],[5711,14122],[5946,13308]]}
       Vector.Y {type=MovieSceneScriptingFloatChannel, default=70824, keys [[5463,69706],[5711,70824],[5946,72057]]}
       Vector.Z {type=MovieSceneScriptingFloatChannel, default=12300, keys [[5463,12300],[5711,13309],[5946,13436]]}
      ]
     }]
   }
   {type=MovieSceneNiagaraSystemTrack, sections [{type=MovieSceneNiagaraSystemSpawnSection, end_frame=6022, start_frame=5422}]}
   {
    type=MovieSceneNiagaraFloatParameterTrack
    display_name=BeamImpactScale
    sections [{type=MovieSceneFloatSection, channels [None {type=MovieSceneScriptingFloatChannel, default=2}]}]
   }
   {
    type=MovieSceneNiagaraFloatParameterTrack
    display_name=BeamWidth
    sections [{type=MovieSceneFloatSection, channels [None {type=MovieSceneScriptingFloatChannel, default=2}]}]
   }
   {
    type=MovieSceneNiagaraFloatParameterTrack
    display_name=Duration
    sections [{type=MovieSceneFloatSection, channels [None {type=MovieSceneScriptingFloatChannel, default=10}]}]
   }
   {
    type=MovieSceneParticleTrack
    display_name=FX System
    sections [{
      type=MovieSceneParticleSection
      channels [None {type=MovieSceneScriptingParticleChannel, keys [[5419, "<ParticleKey.ACTIVATE: 0>"], [6052, "<ParticleKey.DEACTIVATE: 1>"]]}]
     }]
   }
  ]
 }
 TripleTentacleFront {id=#23, possessed_class=PistolCanary_TentacleBoss_C, tracks []}
 NS_PistolCanaryCD_Connecting_Spire_Glow_S {
  id=#45
  possessed_class=NiagaraActor
  tracks [{
    type=MovieSceneVisibilityTrack
    sections [{type=MovieSceneVisibilitySection, channels [None {type=MovieSceneScriptingBoolChannel, default=false, const [false,1]}]}]
   }]
 }
 NS_TentacleEnergyBallDissipate2 {
  id=#46
  possessed_class=NiagaraActor
  tracks [
   {
    type=MovieScene3DTransformTrack
    sections [{
      type=MovieScene3DTransformSection
      channels [
       Location.X {type=MovieSceneScriptingDoubleChannel, default=0}
       Location.Y {type=MovieSceneScriptingDoubleChannel, default=0}
       Location.Z {type=MovieSceneScriptingDoubleChannel, default=0}
       Rotation.X {type=MovieSceneScriptingDoubleChannel, default=0, const [0,1]}
       Rotation.Y {type=MovieSceneScriptingDoubleChannel, default=0, const [0,1]}
       Rotation.Z {type=MovieSceneScriptingDoubleChannel, default=0, const [0,1]}
       Scale.X {type=MovieSceneScriptingDoubleChannel, default=0.2, const [1,1]}
       Scale.Y {type=MovieSceneScriptingDoubleChannel, default=0.2, const [1,1]}
       Scale.Z {type=MovieSceneScriptingDoubleChannel, default=1, const [1,1]}
      ]
     }]
   }
   {type=MovieScene3DAttachTrack, sections [{type=MovieScene3DAttachSection, end_frame=9586, start_frame=5937}]}
  ]
 }
 NS_TripleTentacleAttackLaser2 {
  id=#21
  possessed_class=NiagaraActor
  tracks [
   {type=MovieScene3DAttachTrack, sections [{type=MovieScene3DAttachSection, end_frame=9637, start_frame=7525}]}
   {
    type=MovieSceneSpawnTrack
    sections [{
      type=MovieSceneSpawnSection
      channels [None {type=MovieSceneScriptingBoolChannel, default=false, keys [[3773,false],[3799,true],[3874,false],[7527,true],[7633,false]]}]
     }]
   }
   {
    type=MovieScene3DTransformTrack
    sections [{
      type=MovieScene3DTransformSection
      channels [
       Location.X {type=MovieSceneScriptingDoubleChannel, default=0, const [0,3]}
       Location.Y {type=MovieSceneScriptingDoubleChannel, default=0, const [0,3]}
       Location.Z {type=MovieSceneScriptingDoubleChannel, default=0, const [0,3]}
       Rotation.X {type=MovieSceneScriptingDoubleChannel, default=179.0786, keys [[3801,-87.2302],[7451,-100.502],[7458,-97.6913],[7528,-108.7267]]}
       Rotation.Y {type=MovieSceneScriptingDoubleChannel, default=-56.8563, keys [[3801,-22.8886],[7451,-19.4061],[7458,-20.6412],[7528,-15.9754]]}
       Rotation.Z {type=MovieSceneScriptingDoubleChannel, default=-182.8809, keys [[3801,87.9642],[7451,129.1574],[7458,120.9617],[7528,130.6655]]}
       Scale.X {type=MovieSceneScriptingDoubleChannel, default=15, const [15,4]}
       Scale.Y {type=MovieSceneScriptingDoubleChannel, default=15, const [15,4]}
       Scale.Z {type=MovieSceneScriptingDoubleChannel, default=30, const [30,4]}
      ]
     }]
   }
  ]
 }
 StaticMeshComponent0 {
  id=#47
  parent_id=#48
  possessed_class=StaticMeshComponent
  tracks [
   {
    type=MovieSceneComponentMaterialTrack
    display_name=Material Slot: top_FB_001
    sections [{
      type=MovieSceneComponentMaterialParameterSection
      end_frame=9586
      start_frame=3600
      channels [
       Core Intensity {type=MovieSceneScriptingFloatChannel, default=200, keys [[8534,200],[8575,1000]]}
       Core ColorR {type=MovieSceneScriptingFloatChannel, keys [[8534,0.3613],[8575,1]]}
       Core ColorG {type=MovieSceneScriptingFloatChannel, const [0,2]}
       Core ColorB {type=MovieSceneScriptingFloatChannel, keys [[8534,0.3005],[8575,0.0612]]}
       Core ColorA {type=MovieSceneScriptingFloatChannel, const [1,2]}
      ]
     }]
   }
   {
    type=MovieScenePrimitiveMaterialTrack
    display_name=Material Slot: top_FB_001
    sections [{
      type=MovieScenePrimitiveMaterialSection
      channels [None {
        type=MovieSceneScriptingObjectPathChannel
        default="<Object '&MI_Hermes_Rocks_MaskMaker_MonolithTop_B_A_Event' (0x852ee93e800) Class 'MaterialInstanceConstant'>"
       }]
     }]
   }
   {
    type=MovieSceneComponentMaterialTrack
    display_name=Material Slot: Bottom_FB
    sections [{
      type=MovieSceneComponentMaterialParameterSection
      end_frame=9586
      start_frame=3600
      channels [
       Core Intensity {type=MovieSceneScriptingFloatChannel, default=200, keys [[8534,200],[8575,1000]]}
       Core ColorR {type=MovieSceneScriptingFloatChannel, keys [[8534,0.3613],[8575,1]]}
       Core ColorG {type=MovieSceneScriptingFloatChannel, const [0,2]}
       Core ColorB {type=MovieSceneScriptingFloatChannel, keys [[8534,0.3005],[8575,0.0612]]}
       Core ColorA {type=MovieSceneScriptingFloatChannel, const [1,2]}
      ]
     }]
   }
   {
    type=MovieScenePrimitiveMaterialTrack
    display_name=Material Slot: Bottom_FB
    sections [{
      type=MovieScenePrimitiveMaterialSection
      channels [None {
        type=MovieSceneScriptingObjectPathChannel
        default="<Object '&MI_Hermes_Rocks_MaskMaker_MonolithBottom_A_A_Event' (0x852ee93f800) Class 'MaterialInstanceConstant'>"
       }]
     }]
   }
  ]
 }
 NiagaraComponent0 {
  id=#49
  parent_id=#6
  possessed_class=NiagaraComponent
  tracks [
   {
    type=MovieSceneNiagaraFloatParameterTrack
    display_name=WidthMult
    sections [{type=MovieSceneFloatSection, channels [None {type=MovieSceneScriptingFloatChannel, default=1}]}]
   }
   {
    type=MovieSceneNiagaraFloatParameterTrack
    display_name=BeamProgress
    sections [{
      type=MovieSceneFloatSection
      channels [None {
        type=MovieSceneScriptingFloatChannel
        default=0
        thinned_from=19
        keys [[7043,0],[7047.4825,1],[7224.9825,1],[7230,0],[8111,0],[8123,1],[8179,1],[8180,0],[8289,0],[8292,1],[8326,1],[8327,0],[8426,0],[8430,1],[8449,1],[8452,0],[8550,0],[8554,1]]
       }]
     }]
   }
   {
    type=MovieSceneNiagaraFloatParameterTrack
    display_name=BeamWidth
    sections [{
      type=MovieSceneFloatSection
      channels [None {type=MovieSceneScriptingFloatChannel, default=0, thinned_from=11, keys [[7214,1],[7224.9825,0],[8111,0],[8113,1],[8575,1],[8578,0]]}]
     }]
   }
   {
    type=MovieSceneNiagaraFloatParameterTrack
    display_name=Firing
    sections [{
      type=MovieSceneFloatSection
      channels [None {
        type=MovieSceneScriptingFloatChannel
        default=0
        thinned_from=21
        keys [[7043,0],[7044,1],[7214,1],[7224.9825,0],[8121,0],[8122,1],[8166,1],[8170,0],[8291,0],[8292,1],[8326,1],[8429,0],[8430,1],[8443,1],[8449,0],[8553,0],[8554,1],[8573,1],[8576,0]]
       }]
     }]
   }
   {
    type=MovieSceneNiagaraVectorParameterTrack
    display_name=Stop Location
    sections [{
      type=MovieSceneFloatVectorSection
      channels [
       Vector.X {
        type=MovieSceneScriptingFloatChannel
        default=11617
        keys [[7043,10421],[7117,9285],[7121,10237],[7127,11189],[7145,11617],[7164,11189],[7182,10448],[7197,9885],[8058,10928.6064],[8124,26148.9941],[8242,26148.9941],[8276,14540.3877],[8425,14540.3877],[8426,1745.8026],[8456,1745.8026],[8550,14295.3955]]
       }
       Vector.Y {
        type=MovieSceneScriptingFloatChannel
        default=70147
        keys [[7043,71235],[7117,69591],[7121,68835],[7127,69059],[7145,70147],[7164,70891],[7182,70937],[7197,70487],[8058,71390.0547],[8124,69656.6719],[8125,70355.2578],[8242,69656.6719],[8243,70355.2578],[8276,86682.6172],[8425,86682.6172],[8426,70526.4375],[8456,70526.4375],[8550,56967.4258]]
       }
       Vector.Z {
        type=MovieSceneScriptingFloatChannel
        default=9209
        thinned_from=17
        keys [[7043,10121],[7117,9789],[7127,9209],[7197,9209],[8058,11000],[8124,7660.5454],[8125,6586.1606],[8242,7660.5454],[8243,6586.1606],[8276,6756.8145],[8425,6756.8145],[8426,6420],[8456,6420],[8550,5867.0913]]
       }
      ]
     }]
   }
   {
    type=MovieSceneParticleTrack
    display_name=FX System
    sections [{
      type=MovieSceneParticleSection
      channels [None {
        type=MovieSceneScriptingParticleChannel
        keys [
         [7032, "<ParticleKey.DEACTIVATE: 1>"]
         [7043, "<ParticleKey.ACTIVATE: 0>"]
         [7230, "<ParticleKey.DEACTIVATE: 1>"]
         [8125, "<ParticleKey.ACTIVATE: 0>"]
         [8158, "<ParticleKey.DEACTIVATE: 1>"]
         [8266, "<ParticleKey.ACTIVATE: 0>"]
         [8319, "<ParticleKey.DEACTIVATE: 1>"]
         [8426, "<ParticleKey.ACTIVATE: 0>"]
         [8456, "<ParticleKey.DEACTIVATE: 1>"]
         [8550, "<ParticleKey.ACTIVATE: 0>"]
         [8576, "<ParticleKey.DEACTIVATE: 1>"]
        ]
       }]
     }]
   }
  ]
 }
 NS_PS_WealthLamb_Exit_SkyFlare {id=#50, possessed_class=NiagaraActor, tracks []}
 NiagaraComponent0 {
  id=#51
  parent_id=#9
  possessed_class=NiagaraComponent
  tracks [
   {
    type=MovieSceneNiagaraFloatParameterTrack
    display_name=BeamWidth
    sections [{type=MovieSceneFloatSection, channels [None {type=MovieSceneScriptingFloatChannel, default=0, keys [[7474,1],[7481,0]]}]}]
   }
   {
    type=MovieSceneParticleTrack
    display_name=FX System
    sections [{
      type=MovieSceneParticleSection
      channels [None {type=MovieSceneScriptingParticleChannel, keys [[7450, "<ParticleKey.TRIGGER: 2>"], [7492, "<ParticleKey.DEACTIVATE: 1>"]]}]
     }]
   }
  ]
 }
 PistolCanary_RockSpireMonolith2 {
  id=#48
  possessed_class=PistolCanary_Monolith_Spire_C
  tracks [
   {
    type=MovieSceneVisibilityTrack
    sections [{type=MovieSceneVisibilitySection, channels [None {type=MovieSceneScriptingBoolChannel, default=false, keys [[3613,true],[8579,false]]}]}]
   }
   {
    type=MovieSceneEventTrack
    sections [{
      type=MovieSceneEventTriggerSection
      channels [None {type=MovieSceneScriptingEventChannel, const ["<Struct 'MovieSceneEvent' (0x852fd1a1ea0) {}>", 1]}]
     }]
   }
  ]
 }
 NS_PC_EnergyBall_Charge_Beam_Tentacle3 {
  id=#52
  possessed_class=NiagaraActor
  tracks [
   {
    type=MovieScene3DTransformTrack
    sections [{
      type=MovieScene3DTransformSection
      channels [
       Location.X {type=MovieSceneScriptingDoubleChannel, default=0}
       Location.Y {type=MovieSceneScriptingDoubleChannel, default=0}
       Location.Z {type=MovieSceneScriptingDoubleChannel, default=0}
       Rotation.X {type=MovieSceneScriptingDoubleChannel, default=0}
       Rotation.Y {type=MovieSceneScriptingDoubleChannel, default=0}
       Rotation.Z {type=MovieSceneScriptingDoubleChannel, default=0}
       Scale.X {type=MovieSceneScriptingDoubleChannel, default=1}
       Scale.Y {type=MovieSceneScriptingDoubleChannel, default=1}
       Scale.Z {type=MovieSceneScriptingDoubleChannel, default=1}
      ]
     }]
   }
   {type=MovieScene3DAttachTrack, sections [{type=MovieScene3DAttachSection, end_frame=9455, start_frame=5415}]}
   {
    type=MovieSceneSpawnTrack
    sections [{
      type=MovieSceneSpawnSection
      channels [None {type=MovieSceneScriptingBoolChannel, default=false, keys [[5411,false],[5419,true],[6225,false]]}]
     }]
   }
  ]
 }
 NiagaraComponent0 {
  id=#53
  parent_id=#34
  possessed_class=NiagaraComponent
  tracks [{
    type=MovieSceneParticleTrack
    display_name=FX System
    sections [{type=MovieSceneParticleSection, channels [None {type=MovieSceneScriptingParticleChannel, const ["<ParticleKey.TRIGGER: 2>", 1]}]}]
   }]
 }
 NS_LargeSpire_DeathFX_W {
  id=#54
  possessed_class=NiagaraActor
  tracks [{
    type=MovieScene3DTransformTrack
    sections [{
      type=MovieScene3DTransformSection
      channels [
       Location.X {type=MovieSceneScriptingDoubleChannel, default=1638.1653}
       Location.Y {type=MovieSceneScriptingDoubleChannel, default=70412.4252}
       Location.Z {type=MovieSceneScriptingDoubleChannel, default=7128}
       Rotation.X {type=MovieSceneScriptingDoubleChannel, default=0}
       Rotation.Y {type=MovieSceneScriptingDoubleChannel, default=0}
       Rotation.Z {type=MovieSceneScriptingDoubleChannel, default=0}
       Scale.X {type=MovieSceneScriptingDoubleChannel, default=1}
       Scale.Y {type=MovieSceneScriptingDoubleChannel, default=3}
       Scale.Z {type=MovieSceneScriptingDoubleChannel, default=3}
      ]
     }]
   }]
 }
 NiagaraComponent0 {
  id=#55
  parent_id=#46
  possessed_class=NiagaraComponent
  tracks [
   {
    type=MovieSceneNiagaraFloatParameterTrack
    display_name=Scale
    sections [{type=MovieSceneFloatSection, channels [None {type=MovieSceneScriptingFloatChannel, default=1.5, keys [[6202,1.5],[6219,0]]}]}]
   }
   {type=MovieSceneNiagaraSystemTrack, sections [{type=MovieSceneNiagaraSystemSpawnSection, end_frame=6407, start_frame=6202}]}
   {
    type=MovieSceneParticleTrack
    display_name=FX System
    sections [{
      type=MovieSceneParticleSection
      channels [None {type=MovieSceneScriptingParticleChannel, keys [[6210, "<ParticleKey.TRIGGER: 2>"], [6408, "<ParticleKey.DEACTIVATE: 1>"]]}]
     }]
   }
  ]
 }
 NS_IcicleLarva_Aura {
  id=#56
  possessed_class=NiagaraActor
  tracks [
   {
    type=MovieSceneSpawnTrack
    sections [{
      type=MovieSceneSpawnSection
      channels [None {type=MovieSceneScriptingBoolChannel, default=false, keys [[4907,false],[6052,true],[6595,false],[7184,true],[9299,false]]}]
     }]
   }
   {type=MovieScene3DAttachTrack, sections [{type=MovieScene3DAttachSection, end_frame=13355, start_frame=6042}]}
  ]
 }
 NiagaraComponent0 {
  id=#57
  parent_id=#58
  possessed_class=NiagaraComponent
  tracks [
   {
    type=MovieSceneBoolTrack
    display_name=Auto Activate
    sections [{
      type=MovieSceneBoolSection
      channels [None {type=MovieSceneScriptingBoolChannel, default=false, keys [[8291,false],[8323,true],[8464,false]]}]
     }]
   }
   {
    type=MovieSceneParticleTrack
    display_name=FX System
    sections [{
      type=MovieSceneParticleSection
      channels [None {type=MovieSceneScriptingParticleChannel, keys [[8323, "<ParticleKey.TRIGGER: 2>"], [8407, "<ParticleKey.DEACTIVATE: 1>"]]}]
     }]
   }
  ]
 }
 NS_PistolCanary_InnerSpire_Glow_Loop_S {
  id=#59
  possessed_class=NiagaraActor
  tracks [{
    type=MovieSceneVisibilityTrack
    sections [{type=MovieSceneVisibilitySection, channels [None {type=MovieSceneScriptingBoolChannel, default=false, keys [[3613,true],[8323,false]]}]}]
   }]
 }
 NS_PC_EnergyBall_Charge_Beam_Tentacle2 {
  id=#44
  possessed_class=NiagaraActor
  tracks [
   {
    type=MovieScene3DTransformTrack
    sections [{
      type=MovieScene3DTransformSection
      channels [
       Location.X {type=MovieSceneScriptingDoubleChannel, default=0}
       Location.Y {type=MovieSceneScriptingDoubleChannel, default=0}
       Location.Z {type=MovieSceneScriptingDoubleChannel, default=0}
       Rotation.X {type=MovieSceneScriptingDoubleChannel, default=0}
       Rotation.Y {type=MovieSceneScriptingDoubleChannel, default=0}
       Rotation.Z {type=MovieSceneScriptingDoubleChannel, default=0}
       Scale.X {type=MovieSceneScriptingDoubleChannel, default=1}
       Scale.Y {type=MovieSceneScriptingDoubleChannel, default=1}
       Scale.Z {type=MovieSceneScriptingDoubleChannel, default=1}
      ]
     }]
   }
   {type=MovieScene3DAttachTrack, sections [{type=MovieScene3DAttachSection, end_frame=9455, start_frame=5376}]}
   {
    type=MovieSceneSpawnTrack
    sections [{
      type=MovieSceneSpawnSection
      channels [None {type=MovieSceneScriptingBoolChannel, default=false, keys [[5411,false],[5419,true],[6225,false]]}]
     }]
   }
  ]
 }
 NiagaraComponent0 {
  id=#60
  parent_id=#52
  possessed_class=NiagaraComponent
  tracks [
   {type=MovieSceneNiagaraSystemTrack, sections [{type=MovieSceneNiagaraSystemSpawnSection, end_frame=6022, start_frame=5422}]}
   {
    type=MovieSceneNiagaraFloatParameterTrack
    display_name=BeamImpactScale
    sections [{type=MovieSceneFloatSection, channels [None {type=MovieSceneScriptingFloatChannel, default=2}]}]
   }
   {
    type=MovieSceneNiagaraFloatParameterTrack
    display_name=BeamWidth
    sections [{type=MovieSceneFloatSection, channels [None {type=MovieSceneScriptingFloatChannel, default=2}]}]
   }
   {
    type=MovieSceneNiagaraVectorParameterTrack
    display_name=Stop Location
    sections [{
      type=MovieSceneFloatVectorSection
      channels [
       Vector.X {type=MovieSceneScriptingFloatChannel, default=14122, keys [[5463,14347.3086],[5711,14122],[5946,13308]]}
       Vector.Y {type=MovieSceneScriptingFloatChannel, default=70824, keys [[5463,69706],[5711,70824],[5946,72057]]}
       Vector.Z {type=MovieSceneScriptingFloatChannel, default=12300, keys [[5463,12300],[5711,13309],[5946,13436]]}
      ]
     }]
   }
   {
    type=MovieSceneParticleTrack
    display_name=FX System
    sections [{
      type=MovieSceneParticleSection
      channels [None {type=MovieSceneScriptingParticleChannel, keys [[5419, "<ParticleKey.ACTIVATE: 0>"], [6049, "<ParticleKey.DEACTIVATE: 1>"]]}]
     }]
   }
  ]
 }
 PistolCanary_RockSpireMonolith3 {
  id=#29
  possessed_class=PistolCanary_Monolith_Spire_C
  tracks [
   {
    type=MovieSceneVisibilityTrack
    sections [{type=MovieSceneVisibilitySection, channels [None {type=MovieSceneScriptingBoolChannel, default=false, keys [[3613,true],[8323,false]]}]}]
   }
   {
    type=MovieSceneEventTrack
    sections [{
      type=MovieSceneEventTriggerSection
      channels [None {type=MovieSceneScriptingEventChannel, const ["<Struct 'MovieSceneEvent' (0x852fd1a1b80) {}>", 1]}]
     }]
   }
  ]
 }
 NS_TentacleAttackImpact5 {
  id=#27
  possessed_class=NiagaraActor
  tracks [{
    type=MovieScene3DTransformTrack
    sections [{
      type=MovieScene3DTransformSection
      channels [
       Location.X {type=MovieSceneScriptingDoubleChannel, default=6560, keys [[7455,6560],[7529,19252]]}
       Location.Y {type=MovieSceneScriptingDoubleChannel, default=69376, keys [[7455,69376],[7529,61716]]}
       Location.Z {type=MovieSceneScriptingDoubleChannel, default=3604, keys [[7455,3604],[7529,3504]]}
       Rotation.X {type=MovieSceneScriptingDoubleChannel, default=0, keys [[7455,0],[7529,13.3639]]}
       Rotation.Y {type=MovieSceneScriptingDoubleChannel, default=0, const [0,2]}
       Rotation.Z {type=MovieSceneScriptingDoubleChannel, default=0, const [0,2]}
       Scale.X {type=MovieSceneScriptingDoubleChannel, default=1, const [1,2]}
       Scale.Y {type=MovieSceneScriptingDoubleChannel, default=1, const [1,2]}
       Scale.Z {type=MovieSceneScriptingDoubleChannel, default=1, const [1,2]}
      ]
     }]
   }]
 }
 NS_TentacleAttackChargeUpLaser2 {
  id=#61
  possessed_class=NiagaraActor
  tracks [
   {
    type=MovieSceneSpawnTrack
    sections [{type=MovieSceneSpawnSection, channels [None {type=MovieSceneScriptingBoolChannel, default=false, const [false,3]}]}]
   }
   {type=MovieScene3DAttachTrack, sections [{type=MovieScene3DAttachSection, end_frame=14157, start_frame=5654}]}
  ]
 }
 SkeletalMesh {
  id=#62
  parent_id=#40
  possessed_class=SkeletalMeshComponent
  tracks [
   {
    type=MovieSceneComponentMaterialTrack
    display_name=Material Slot: Body_A
    sections [{
      type=MovieSceneComponentMaterialParameterSection
      end_frame=9210
      start_frame=3495
      channels [
       PulseSpeed {type=MovieSceneScriptingFloatChannel, default=0.2, thinned_from=6, keys [[5254,0.1],[5258,0],[6393,0],[6400,0.2]]}
       PulseManualOffset {type=MovieSceneScriptingFloatChannel, default=0.6, keys [[5324,0],[5655,0.6],[7261,0],[7336,1],[7770,1],[7803,0.6]]}
       PulseRadius {type=MovieSceneScriptingFloatChannel, default=0.33, keys [[5258,0.33],[6176,1],[6393,0.3]]}
       GlowPulseEXP {type=MovieSceneScriptingFloatChannel, default=0.03, keys [[5572,0.015],[5766,0.03],[6176,0.03],[6393,0.015]]}
       EmissiveGlowColorR {
        type=MovieSceneScriptingFloatChannel
        default=8
        thinned_from=12
        keys [[5215,8],[5254,0],[5261,0],[5324,8],[5877,0.7],[6180,0.7],[6291,8],[6393,0],[6400,0],[6433,8],[7770,8]]
       }
       EmissiveGlowColorG {
        type=MovieSceneScriptingFloatChannel
        default=0.3176
        thinned_from=12
        keys [[5215,0.3176],[5254,0],[5261,0],[5324,0.3176],[5877,3],[6180,3],[6291,0.3176],[6393,0],[6400,0],[6433,0.3176],[7770,0.3176]]
       }
       EmissiveGlowColorB {
        type=MovieSceneScriptingFloatChannel
        default=0
        keys [[5254,0],[5324,0],[5877,10],[6180,10],[6291,0],[7261,0],[7336,100],[7770,100],[7803,0]]
       }
       EmissiveGlowColorA {type=MovieSceneScriptingFloatChannel, default=1, const [1,2]}
      ]
     }]
   }
   {
    type=MovieSceneComponentMaterialTrack
    display_name=Material Slot: Body_B
    sections [{
      type=MovieSceneComponentMaterialParameterSection
      end_frame=9315
      start_frame=3600
      channels [
       PulseSpeed {type=MovieSceneScriptingFloatChannel, default=0, thinned_from=6, keys [[5254,0.1],[5258,0],[6393,0],[6400,0.2]]}
       PulseManualOffset {type=MovieSceneScriptingFloatChannel, default=0.6, keys [[5324,0],[5655,0.6],[7261,0],[7336,1],[7770,1],[7803,0.6]]}
       PulseRadius {type=MovieSceneScriptingFloatChannel, default=0.33, keys [[5258,0.33],[6176,1],[6393,0.3]]}
       GlowPulseEXP {type=MovieSceneScriptingFloatChannel, default=0.03, keys [[5572,0.015],[5766,0.03],[6176,0.03],[6393,0.015]]}
       EmissiveGlowColorR {
        type=MovieSceneScriptingFloatChannel
        default=8
        thinned_from=12
        keys [[5215,8],[5254,0],[5261,0],[5324,8],[5877,0.7],[6180,0.7],[6291,8],[6393,0],[6400,0],[6433,8],[7770,8]]
       }
       EmissiveGlowColorG {
        type=MovieSceneScriptingFloatChannel
        default=0.3176
        thinned_from=12
        keys [[5215,0.3176],[5254,0],[5261,0],[5324,0.3176],[5877,3],[6180,3],[6291,0.3176],[6393,0],[6400,0],[6433,0.3176],[7770,0.3176]]
       }
       EmissiveGlowColorB {
        type=MovieSceneScriptingFloatChannel
        default=0
        keys [[5254,0],[5324,0],[5877,10],[6180,10],[6291,0],[7261,0],[7336,100],[7770,100],[7803,0]]
       }
       EmissiveGlowColorA {type=MovieSceneScriptingFloatChannel, default=1, const [1,2]}
      ]
     }]
   }
   {
    type=MovieSceneComponentMaterialTrack
    display_name=Material Slot: Body_C
    sections [{
      type=MovieSceneComponentMaterialParameterSection
      end_frame=9315
      start_frame=3600
      channels [
       PulseSpeed {type=MovieSceneScriptingFloatChannel, default=0, thinned_from=6, keys [[5254,0.1],[5258,0],[6393,0],[6400,0.2]]}
       PulseManualOffset {type=MovieSceneScriptingFloatChannel, default=0.6, keys [[5324,0],[5655,0.6],[7261,0],[7336,1],[7770,1],[7803,0.6]]}
       PulseRadius {type=MovieSceneScriptingFloatChannel, default=0.3, keys [[5258,0.33],[6176,1],[6393,0.3]]}
       GlowPulseEXP {type=MovieSceneScriptingFloatChannel, default=0.03, keys [[5572,0.015],[5766,0.03],[6176,0.03],[6393,0.015]]}
       EmissiveGlowColorR {
        type=MovieSceneScriptingFloatChannel
        default=0
        thinned_from=12
        keys [[5215,8],[5254,0],[5261,0],[5324,8],[5877,0.7],[6180,0.7],[6291,8],[6393,0],[6400,0],[6433,8],[7770,8]]
       }
       EmissiveGlowColorG {
        type=MovieSceneScriptingFloatChannel
        default=0
        thinned_from=12
        keys [[5215,0.3176],[5254,0],[5261,0],[5324,0.3176],[5877,3],[6180,3],[6291,0.3176],[6393,0],[6400,0],[6433,0.3176],[7770,0.3176]]
       }
       EmissiveGlowColorB {
        type=MovieSceneScriptingFloatChannel
        default=0
        keys [[5254,0],[5324,0],[5877,10],[6180,10],[6291,0],[7261,0],[7336,100],[7770,100],[7803,0]]
       }
       EmissiveGlowColorA {type=MovieSceneScriptingFloatChannel, default=1, const [1,2]}
      ]
     }]
   }
   {
    type=MovieSceneComponentMaterialTrack
    display_name=Material Slot: Body_D
    sections [{
      type=MovieSceneComponentMaterialParameterSection
      end_frame=9315
      start_frame=3600
      channels [
       PulseSpeed {type=MovieSceneScriptingFloatChannel, default=0, thinned_from=6, keys [[5254,0.1],[5258,0],[6393,0],[6400,0.2]]}
       PulseManualOffset {type=MovieSceneScriptingFloatChannel, default=0.6, keys [[5324,0],[5655,0.6],[7261,0],[7336,1],[7770,1],[7803,0.6]]}
       PulseRadius {type=MovieSceneScriptingFloatChannel, default=0.3399, keys [[5258,0.33],[6176,1],[6393,0.3]]}
       GlowPulseEXP {type=MovieSceneScriptingFloatChannel, default=0.03, keys [[5572,0.015],[5766,0.03],[6176,0.03],[6393,0.015]]}
       EmissiveGlowColorR {
        type=MovieSceneScriptingFloatChannel
        default=8
        thinned_from=12
        keys [[5215,8],[5254,0],[5261,0],[5324,8],[5877,0.7],[6180,0.7],[6291,8],[6393,0],[6400,0],[6433,8],[7770,8]]
       }
       EmissiveGlowColorG {
        type=MovieSceneScriptingFloatChannel
        default=0.3176
        thinned_from=12
        keys [[5215,0.3176],[5254,0],[5261,0],[5324,0.3176],[5877,3],[6180,3],[6291,0.3176],[6393,0],[6400,0],[6433,0.3176],[7770,0.3176]]
       }
       EmissiveGlowColorB {
        type=MovieSceneScriptingFloatChannel
        default=0
        keys [[5254,0],[5324,0],[5877,10],[6180,10],[6291,0],[7261,0],[7336,100],[7770,100],[7803,0]]
       }
       EmissiveGlowColorA {type=MovieSceneScriptingFloatChannel, default=1, const [1,2]}
      ]
     }]
   }
  ]
 }
 PistolCanary_RockSpireMonolith4 {
  id=#63
  possessed_class=PistolCanary_Monolith_Spire_C
  tracks [
   {
    type=MovieSceneVisibilityTrack
    sections [{type=MovieSceneVisibilitySection, channels [None {type=MovieSceneScriptingBoolChannel, default=false, keys [[3614,true],[8167,false]]}]}]
   }
   {
    type=MovieSceneEventTrack
    sections [{
      type=MovieSceneEventTriggerSection
      channels [None {type=MovieSceneScriptingEventChannel, const ["<Struct 'MovieSceneEvent' (0x852fd1a3520) {}>", 1]}]
     }]
   }
  ]
 }
 NS_PistolCanaryCD_Connecting_Spire_Glow_N {
  id=#64
  possessed_class=NiagaraActor
  tracks [{
    type=MovieSceneVisibilityTrack
    sections [{type=MovieSceneVisibilitySection, channels [None {type=MovieSceneScriptingBoolChannel, default=false, const [false,1]}]}]
   }]
 }
 KalEl {id=#65, possessed_class=PistolCanary_CinematicNPC_Superhero_Main_C, tracks []}
 NS_PistolCanary_TripleTentacleWave {
  id=#66
  possessed_class=NiagaraActor
  tracks [{
    type=MovieScene3DTransformTrack
    sections [{
      type=MovieScene3DTransformSection
      channels [
       Location.X {type=MovieSceneScriptingDoubleChannel, default=14360.2255, const [14240.2256,1]}
       Location.Y {type=MovieSceneScriptingDoubleChannel, default=81825.2175, const [70241.2188,1]}
       Location.Z {type=MovieSceneScriptingDoubleChannel, default=4050.9363, const [4050.9363,1]}
       Rotation.X {type=MovieSceneScriptingDoubleChannel, default=0, const [0,1]}
       Rotation.Y {type=MovieSceneScriptingDoubleChannel, default=0, const [0,1]}
       Rotation.Z {type=MovieSceneScriptingDoubleChannel, default=0, const [0,1]}
       Scale.X {type=MovieSceneScriptingDoubleChannel, default=1, const [1,1]}
       Scale.Y {type=MovieSceneScriptingDoubleChannel, default=1, const [1,1]}
       Scale.Z {type=MovieSceneScriptingDoubleChannel, default=1, const [1,1]}
      ]
     }]
   }]
 }
 NiagaraComponent0 {
  id=#67
  parent_id=#13
  possessed_class=NiagaraComponent
  tracks [
   {
    type=MovieSceneNiagaraFloatParameterTrack
    display_name=Scale
    sections [{
      type=MovieSceneFloatSection
      channels [None {type=MovieSceneScriptingFloatChannel, default=0.7, keys [[5428,0],[5444,0.2],[6076,0.5],[6198,0.7],[6204,0]]}]
     }]
   }
   {
    type=MovieSceneNiagaraFloatParameterTrack
    display_name=PostProcess
    sections [{
      type=MovieSceneFloatSection
      channels [None {type=MovieSceneScriptingFloatChannel, default=0, keys [[5505,0],[6073,0.5],[6159,0.5],[6198,1],[6203,0]]}]
     }]
   }
   {
    type=MovieSceneNiagaraFloatParameterTrack
    display_name=LightRays
    sections [{type=MovieSceneFloatSection, channels [None {type=MovieSceneScriptingFloatChannel, default=4, keys [[6070,0],[6165,4],[6182,0]]}]}]
   }
   {type=MovieSceneNiagaraSystemTrack, sections [{type=MovieSceneNiagaraSystemSpawnSection, end_frame=6252, start_frame=5415}]}
  ]
 }
 NiagaraComponent0 {
  id=#68
  parent_id=#69
  possessed_class=NiagaraComponent
  tracks [{
    type=MovieSceneParticleTrack
    display_name=FX System
    sections [{
      type=MovieSceneParticleSection
      channels [None {type=MovieSceneScriptingParticleChannel, keys [[7591, "<ParticleKey.TRIGGER: 2>"], [7787, "<ParticleKey.DEACTIVATE: 1>"]]}]
     }]
   }]
 }
 NiagaraComponent0 {
  id=#70
  parent_id=#54
  possessed_class=NiagaraComponent
  tracks [
   {
    type=MovieSceneBoolTrack
    display_name=Auto Activate
    sections [{
      type=MovieSceneBoolSection
      channels [None {type=MovieSceneScriptingBoolChannel, default=false, keys [[8436,false],[8460,true],[8558,false]]}]
     }]
   }
   {
    type=MovieSceneParticleTrack
    display_name=FX System
    sections [{
      type=MovieSceneParticleSection
      channels [None {type=MovieSceneScriptingParticleChannel, keys [[8460, "<ParticleKey.TRIGGER: 2>"], [8558, "<ParticleKey.DEACTIVATE: 1>"]]}]
     }]
   }
  ]
 }
 NiagaraComponent0 {
  id=#71
  parent_id=#66
  possessed_class=NiagaraComponent
  tracks [
   {
    type=MovieSceneBoolTrack
    display_name=Auto Activate
    sections [{
      type=MovieSceneBoolSection
      channels [None {type=MovieSceneScriptingBoolChannel, default=false, keys [[4966,false],[5002,true],[5197,false],[5902,false]]}]
     }]
   }
   {
    type=MovieSceneParticleTrack
    display_name=FX System
    sections [{
      type=MovieSceneParticleSection
      channels [None {type=MovieSceneScriptingParticleChannel, keys [[5002, "<ParticleKey.TRIGGER: 2>"], [5197, "<ParticleKey.DEACTIVATE: 1>"]]}]
     }]
   }
  ]
 }
 NiagaraComponent0 {
  id=#72
  parent_id=#1
  possessed_class=NiagaraComponent
  tracks [
   {
    type=MovieSceneNiagaraVectorParameterTrack
    display_name=ChargeScaleVec
    sections [{
      type=MovieSceneFloatVectorSection
      channels [
       Vector.X {type=MovieSceneScriptingFloatChannel, default=0.6, keys [[7239,0],[7244,0.6],[7334,1]]}
       Vector.Y {type=MovieSceneScriptingFloatChannel, default=1, const [1,2]}
       Vector.Z {type=MovieSceneScriptingFloatChannel, default=1, const [1,2]}
      ]
     }]
   }
   {
    type=MovieSceneNiagaraFloatParameterTrack
    display_name=ChargeProgress
    sections [{type=MovieSceneFloatSection, channels [None {type=MovieSceneScriptingFloatChannel, default=1, keys [[7238,0],[7270,0.3],[7329,1]]}]}]
   }
   {
    type=MovieSceneNiagaraFloatParameterTrack
    display_name=PlasmaEnabled
    sections [{type=MovieSceneFloatSection, channels [None {type=MovieSceneScriptingFloatChannel, default=0, const [0,2]}]}]
   }
   {
    type=MovieSceneNiagaraColorParameterTrack
    display_name=ChargeScaleColor
    sections [{
      type=MovieSceneColorSection
      channels [
       Color.R {type=MovieSceneScriptingFloatChannel, default=1}
       Color.G {type=MovieSceneScriptingFloatChannel, default=1}
       Color.B {type=MovieSceneScriptingFloatChannel, default=1}
       Color.A {type=MovieSceneScriptingFloatChannel, default=0, keys [[7253,0],[7288,1],[7613,1],[7713,0]]}
      ]
     }]
   }
   {
    type=MovieSceneNiagaraFloatParameterTrack
    display_name=ChargeProgressLightning
    sections [{type=MovieSceneFloatSection, channels [None {type=MovieSceneScriptingFloatChannel, default=0.5472, keys [[7240,0],[7296,0.5472],[7392,1]]}]}]
   }
   {
    type=MovieSceneNiagaraFloatParameterTrack
    display_name=PostProcess
    sections [{
      type=MovieSceneFloatSection
      channels [None {type=MovieSceneScriptingFloatChannel, default=0, thinned_from=6, keys [[7241,0],[7421,1],[7550,0],[7665,1],[7717,0]]}]
     }]
   }
   {
    type=MovieSceneNiagaraFloatParameterTrack
    display_name=TipDelay
    sections [{type=MovieSceneFloatSection, channels [None {type=MovieSceneScriptingFloatChannel, default=1.5}]}]
   }
   {
    type=MovieSceneParticleTrack
    display_name=FX System
    sections [{
      type=MovieSceneParticleSection
      channels [None {type=MovieSceneScriptingParticleChannel, keys [[7237, "<ParticleKey.TRIGGER: 2>"], [7705, "<ParticleKey.DEACTIVATE: 1>"]]}]
     }]
   }
   {
    type=MovieSceneNiagaraFloatParameterTrack
    display_name=LightningEnabled
    sections [{type=MovieSceneFloatSection, channels [None {type=MovieSceneScriptingFloatChannel, default=0, keys [[7383,1],[7391,0]]}]}]
   }
  ]
 }
 NiagaraComponent0 {
  id=#73
  parent_id=#74
  possessed_class=NiagaraComponent
  tracks [
   {
    type=MovieSceneBoolTrack
    display_name=Auto Activate
    sections [{
      type=MovieSceneBoolSection
      channels [None {type=MovieSceneScriptingBoolChannel, default=false, keys [[6123,false],[6126,true],[6223,false]]}]
     }]
   }
   {
    type=MovieSceneParticleTrack
    display_name=FX System
    sections [{
      type=MovieSceneParticleSection
      channels [None {type=MovieSceneScriptingParticleChannel, keys [[6126, "<ParticleKey.TRIGGER: 2>"], [6223, "<ParticleKey.DEACTIVATE: 1>"]]}]
     }]
   }
  ]
 }
 NS_TentacleAttackImpact6 {
  id=#69
  possessed_class=NiagaraActor
  tracks [{
    type=MovieScene3DTransformTrack
    sections [{
      type=MovieScene3DTransformSection
      channels [
       Location.X {type=MovieSceneScriptingDoubleChannel, default=6560, keys [[7455,6560],[7529,19252],[7592,23408.4395]]}
       Location.Y {type=MovieSceneScriptingDoubleChannel, default=69376, keys [[7455,69376],[7529,61716],[7592,70719.0469]]}
       Location.Z {type=MovieSceneScriptingDoubleChannel, default=3604, keys [[7455,3604],[7529,3504],[7592,2005.2466]]}
       Rotation.X {type=MovieSceneScriptingDoubleChannel, default=0, keys [[7455,0],[7529,13.3639],[7592,-1.8527]]}
       Rotation.Y {type=MovieSceneScriptingDoubleChannel, default=0, keys [[7455,0],[7529,0],[7592,6.4278]]}
       Rotation.Z {type=MovieSceneScriptingDoubleChannel, default=0, keys [[7455,0],[7529,0],[7592,-0.2075]]}
       Scale.X {type=MovieSceneScriptingDoubleChannel, default=1, const [1,3]}
       Scale.Y {type=MovieSceneScriptingDoubleChannel, default=1, const [1,3]}
       Scale.Z {type=MovieSceneScriptingDoubleChannel, default=1, const [1,3]}
      ]
     }]
   }]
 }
 NS_PC_AirCharge {
  id=#74
  possessed_class=NiagaraActor
  tracks [{type=MovieScene3DAttachTrack, sections [{type=MovieScene3DAttachSection, end_frame=9287, start_frame=6126}]}]
 }
 NS_LargeSpire_DeathFX_N {
  id=#32
  possessed_class=NiagaraActor
  tracks [{
    type=MovieScene3DTransformTrack
    sections [{
      type=MovieScene3DTransformSection
      channels [
       Location.X {type=MovieSceneScriptingDoubleChannel, default=14321.0786}
       Location.Y {type=MovieSceneScriptingDoubleChannel, default=57056.0859}
       Location.Z {type=MovieSceneScriptingDoubleChannel, default=7128}
       Rotation.X {type=MovieSceneScriptingDoubleChannel, default=0}
       Rotation.Y {type=MovieSceneScriptingDoubleChannel, default=0}
       Rotation.Z {type=MovieSceneScriptingDoubleChannel, default=0}
       Scale.X {type=MovieSceneScriptingDoubleChannel, default=3}
       Scale.Y {type=MovieSceneScriptingDoubleChannel, default=4}
       Scale.Z {type=MovieSceneScriptingDoubleChannel, default=4}
      ]
     }]
   }]
 }
 TripleTentacleLeft {id=#75, possessed_class=PistolCanary_TentacleBoss_C, tracks []}
 Niagara {
  id=#76
  parent_id=#65
  possessed_class=NiagaraComponent
  tracks [
   {type=MovieSceneNiagaraSystemTrack, sections [{type=MovieSceneNiagaraSystemSpawnSection, end_frame=9039, start_frame=6047}]}
   {
    type=MovieSceneParticleTrack
    display_name=FX System
    sections [{
      type=MovieSceneParticleSection
      channels [None {
        type=MovieSceneScriptingParticleChannel
        keys [
         [6047, "<ParticleKey.ACTIVATE: 0>"]
         [6537, "<ParticleKey.DEACTIVATE: 1>"]
         [6739, "<ParticleKey.ACTIVATE: 0>"]
         [8273, "<ParticleKey.DEACTIVATE: 1>"]
         [8584, "<ParticleKey.ACTIVATE: 0>"]
        ]
       }]
     }]
   }
   {
    type=MovieSceneNiagaraFloatParameterTrack
    display_name=ContrailWidth
    sections [{type=MovieSceneFloatSection, channels [None {type=MovieSceneScriptingFloatChannel, default=5, keys [[6053,1],[6102,5],[6208,5],[6347,1]]}]}]
   }
   {type=MovieScene3DAttachTrack, sections [{type=MovieScene3DAttachSection, end_frame=9091, start_frame=6040}]}
  ]
 }
 NS_PC_TentacleSlice {
  id=#12
  possessed_class=NiagaraActor
  tracks [
   {
    type=MovieScene3DTransformTrack
    sections [{
      type=MovieScene3DTransformSection
      channels [
       Location.X {type=MovieSceneScriptingDoubleChannel, default=531.8238}
       Location.Y {type=MovieSceneScriptingDoubleChannel, default=2.9705}
       Location.Z {type=MovieSceneScriptingDoubleChannel, default=-137.7855}
       Rotation.X {type=MovieSceneScriptingDoubleChannel, default=10.6673}
       Rotation.Y {type=MovieSceneScriptingDoubleChannel, default=-89.136}
       Rotation.Z {type=MovieSceneScriptingDoubleChannel, default=-9.7945}
       Scale.X {type=MovieSceneScriptingDoubleChannel, default=1}
       Scale.Y {type=MovieSceneScriptingDoubleChannel, default=1}
       Scale.Z {type=MovieSceneScriptingDoubleChannel, default=1}
      ]
     }]
   }
   {type=MovieScene3DAttachTrack, sections [{type=MovieScene3DAttachSection, end_frame=9286, start_frame=6191}]}
  ]
 }
 SkeletalMesh {
  id=#77
  parent_id=#75
  possessed_class=SkeletalMeshComponent
  tracks [
   {
    type=MovieSceneComponentMaterialTrack
    display_name=Material Slot: Body_A
    sections [{
      type=MovieSceneComponentMaterialParameterSection
      end_frame=9210
      start_frame=3495
      channels [
       PulseSpeed {type=MovieSceneScriptingFloatChannel, default=0.2, thinned_from=6, keys [[5254,0.1],[5258,0],[6393,0],[6400,0.2]]}
       PulseManualOffset {type=MovieSceneScriptingFloatChannel, default=0, keys [[5324,0],[5655,0.6]]}
       PulseRadius {type=MovieSceneScriptingFloatChannel, default=0.33, keys [[5258,0.33],[6176,1],[6393,0.3]]}
       GlowPulseEXP {type=MovieSceneScriptingFloatChannel, default=0.03, keys [[5572,0.015],[5766,0.03],[6176,0.03],[6393,0.015]]}
       EmissiveGlowColorR {
        type=MovieSceneScriptingFloatChannel
        default=8
        keys [[5215,8],[5254,0],[5261,0],[5324,8],[5877,0.7],[6180,0.7],[6291,8],[6393,0],[6400,0],[6433,8]]
       }
       EmissiveGlowColorG {
        type=MovieSceneScriptingFloatChannel
        default=0.3176
        keys [[5215,0.3176],[5254,0],[5261,0],[5324,0.3176],[5877,3],[6180,3],[6291,0.3176],[6393,0],[6400,0],[6433,0.3176]]
       }
       EmissiveGlowColorB {type=MovieSceneScriptingFloatChannel, default=0, keys [[5254,0],[5324,0],[5877,10],[6180,10],[6291,0]]}
       EmissiveGlowColorA {type=MovieSceneScriptingFloatChannel, default=1}
      ]
     }]
   }
   {
    type=MovieSceneComponentMaterialTrack
    display_name=Material Slot: Body_B
    sections [{
      type=MovieSceneComponentMaterialParameterSection
      end_frame=9315
      start_frame=3600
      channels [
       PulseSpeed {type=MovieSceneScriptingFloatChannel, default=0, thinned_from=6, keys [[5254,0.1],[5258,0],[6393,0],[6400,0.2]]}
       PulseManualOffset {type=MovieSceneScriptingFloatChannel, default=0, keys [[5324,0],[5655,0.6]]}
       PulseRadius {type=MovieSceneScriptingFloatChannel, default=0.33, keys [[5258,0.33],[6176,1],[6393,0.3]]}
       GlowPulseEXP {type=MovieSceneScriptingFloatChannel, default=0.03, keys [[5572,0.015],[5766,0.03],[6176,0.03],[6393,0.015]]}
       EmissiveGlowColorR {
        type=MovieSceneScriptingFloatChannel
        default=8
        keys [[5215,8],[5254,0],[5261,0],[5324,8],[5877,0.7],[6180,0.7],[6291,8],[6393,0],[6400,0],[6433,8]]
       }
       EmissiveGlowColorG {
        type=MovieSceneScriptingFloatChannel
        default=0.3176
        keys [[5215,0.3176],[5254,0],[5261,0],[5324,0.3176],[5877,3],[6180,3],[6291,0.3176],[6393,0],[6400,0],[6433,0.3176]]
       }
       EmissiveGlowColorB {type=MovieSceneScriptingFloatChannel, default=0, keys [[5254,0],[5324,0],[5877,10],[6180,10],[6291,0]]}
       EmissiveGlowColorA {type=MovieSceneScriptingFloatChannel, default=1}
      ]
     }]
   }
   {
    type=MovieSceneComponentMaterialTrack
    display_name=Material Slot: Body_C
    sections [{
      type=MovieSceneComponentMaterialParameterSection
      end_frame=9315
      start_frame=3600
      channels [
       PulseSpeed {type=MovieSceneScriptingFloatChannel, default=0, thinned_from=6, keys [[5254,0.1],[5258,0],[6393,0],[6400,0.2]]}
       PulseManualOffset {type=MovieSceneScriptingFloatChannel, default=0.7, keys [[5324,0],[5655,0.6]]}
       PulseRadius {type=MovieSceneScriptingFloatChannel, default=0.3, keys [[5258,0.33],[6176,1],[6393,0.3]]}
       GlowPulseEXP {type=MovieSceneScriptingFloatChannel, default=0.03, keys [[5572,0.015],[5766,0.03],[6176,0.03],[6393,0.015]]}
       EmissiveGlowColorR {
        type=MovieSceneScriptingFloatChannel
        default=0
        keys [[5215,8],[5254,0],[5261,0],[5324,8],[5877,0.7],[6180,0.7],[6291,8],[6393,0],[6400,0],[6433,8]]
       }
       EmissiveGlowColorG {
        type=MovieSceneScriptingFloatChannel
        default=0
        keys [[5215,0.3176],[5254,0],[5261,0],[5324,0.3176],[5877,3],[6180,3],[6291,0.3176],[6393,0],[6400,0],[6433,0.3176]]
       }
       EmissiveGlowColorB {type=MovieSceneScriptingFloatChannel, default=10, keys [[5254,0],[5324,0],[5877,10],[6180,10],[6291,0]]}
       EmissiveGlowColorA {type=MovieSceneScriptingFloatChannel, default=1}
      ]
     }]
   }
   {
    type=MovieSceneComponentMaterialTrack
    display_name=Material Slot: Body_D
    sections [{
      type=MovieSceneComponentMaterialParameterSection
      end_frame=9315
      start_frame=3600
      channels [
       PulseSpeed {type=MovieSceneScriptingFloatChannel, default=0, thinned_from=6, keys [[5254,0.1],[5258,0],[6393,0],[6400,0.2]]}
       PulseManualOffset {type=MovieSceneScriptingFloatChannel, default=0, keys [[5324,0],[5655,0.6]]}
       PulseRadius {type=MovieSceneScriptingFloatChannel, default=0.3399, keys [[5258,0.33],[6176,1],[6393,0.3]]}
       GlowPulseEXP {type=MovieSceneScriptingFloatChannel, default=0.03, keys [[5572,0.015],[5766,0.03],[6176,0.03],[6393,0.015]]}
       EmissiveGlowColorR {
        type=MovieSceneScriptingFloatChannel
        default=8
        keys [[5215,8],[5254,0],[5261,0],[5324,8],[5877,0.7],[6180,0.7],[6291,8],[6393,0],[6400,0],[6433,8]]
       }
       EmissiveGlowColorG {
        type=MovieSceneScriptingFloatChannel
        default=0.3176
        keys [[5215,0.3176],[5254,0],[5261,0],[5324,0.3176],[5877,3],[6180,3],[6291,0.3176],[6393,0],[6400,0],[6433,0.3176]]
       }
       EmissiveGlowColorB {type=MovieSceneScriptingFloatChannel, default=0, keys [[5254,0],[5324,0],[5877,10],[6180,10],[6291,0]]}
       EmissiveGlowColorA {type=MovieSceneScriptingFloatChannel, default=1}
      ]
     }]
   }
  ]
 }
 StaticMeshComponent0 {
  id=#78
  parent_id=#63
  possessed_class=StaticMeshComponent
  tracks [
   {
    type=MovieScenePrimitiveMaterialTrack
    display_name=Material Slot: Bottom_FB
    sections [{
      type=MovieScenePrimitiveMaterialSection
      channels [None {
        type=MovieSceneScriptingObjectPathChannel
        default="<Object '&MI_Hermes_Rocks_MaskMaker_MonolithBottom_A_A_Event' (0x852ee93f800) Class 'MaterialInstanceConstant'>"
       }]
     }]
   }
   {
    type=MovieScenePrimitiveMaterialTrack
    display_name=Material Slot: top_FB_001
    sections [{
      type=MovieScenePrimitiveMaterialSection
      channels [None {
        type=MovieSceneScriptingObjectPathChannel
        default="<Object '&MI_Hermes_Rocks_MaskMaker_MonolithTop_B_A_Event' (0x852ee93e800) Class 'MaterialInstanceConstant'>"
       }]
     }]
   }
   {
    type=MovieSceneComponentMaterialTrack
    display_name=Material Slot: Bottom_FB
    sections [{
      type=MovieSceneComponentMaterialParameterSection
      end_frame=9587
      start_frame=3601
      channels [
       Core Intensity {type=MovieSceneScriptingFloatChannel, default=200, keys [[8122,200],[8163,1000]]}
       Core ColorR {type=MovieSceneScriptingFloatChannel, keys [[8122,0.3613],[8163,1]]}
       Core ColorG {type=MovieSceneScriptingFloatChannel, const [0,2]}
       Core ColorB {type=MovieSceneScriptingFloatChannel, keys [[8122,0.3005],[8163,0.0612]]}
       Core ColorA {type=MovieSceneScriptingFloatChannel, const [1,2]}
      ]
     }]
   }
   {
    type=MovieSceneComponentMaterialTrack
    display_name=Material Slot: top_FB_001
    sections [{
      type=MovieSceneComponentMaterialParameterSection
      end_frame=9587
      start_frame=3601
      channels [
       Core Intensity {type=MovieSceneScriptingFloatChannel, default=200, keys [[8122,200],[8163,1000]]}
       Core ColorR {type=MovieSceneScriptingFloatChannel, keys [[8122,0.3613],[8163,1]]}
       Core ColorG {type=MovieSceneScriptingFloatChannel, const [0,2]}
       Core ColorB {type=MovieSceneScriptingFloatChannel, keys [[8122,0.3005],[8163,0.0612]]}
       Core ColorA {type=MovieSceneScriptingFloatChannel, const [1,2]}
      ]
     }]
   }
  ]
 }
 NiagaraComponent0 {
  id=#79
  parent_id=#7
  possessed_class=NiagaraComponent
  tracks [
   {
    type=MovieSceneNiagaraFloatParameterTrack
    display_name=BeamWidth
    sections [{type=MovieSceneFloatSection, channels [None {type=MovieSceneScriptingFloatChannel, default=0, keys [[7618,1],[7625,0]]}]}]
   }
   {
    type=MovieSceneParticleTrack
    display_name=FX System
    sections [{
      type=MovieSceneParticleSection
      channels [None {type=MovieSceneScriptingParticleChannel, keys [[7594, "<ParticleKey.TRIGGER: 2>"], [7636, "<ParticleKey.DEACTIVATE: 1>"]]}]
     }]
   }
  ]
 }
 NS_LargeSpire_DeathFX_S {
  id=#58
  possessed_class=NiagaraActor
  tracks [{
    type=MovieScene3DTransformTrack
    sections [{
      type=MovieScene3DTransformSection
      channels [
       Location.X {type=MovieSceneScriptingDoubleChannel, default=14357.4929}
       Location.Y {type=MovieSceneScriptingDoubleChannel, default=86757.8885}
       Location.Z {type=MovieSceneScriptingDoubleChannel, default=7128}
       Rotation.X {type=MovieSceneScriptingDoubleChannel, default=0}
       Rotation.Y {type=MovieSceneScriptingDoubleChannel, default=0}
       Rotation.Z {type=MovieSceneScriptingDoubleChannel, default=0}
       Scale.X {type=MovieSceneScriptingDoubleChannel, default=1}
       Scale.Y {type=MovieSceneScriptingDoubleChannel, default=3}
       Scale.Z {type=MovieSceneScriptingDoubleChannel, default=3}
      ]
     }]
   }]
 }
 NiagaraComponent0 {
  id=#80
  parent_id=#50
  possessed_class=NiagaraComponent
  tracks [{
    type=MovieSceneParticleTrack
    display_name=FX System
    sections [{type=MovieSceneParticleSection, channels [None {type=MovieSceneScriptingParticleChannel, const ["<ParticleKey.TRIGGER: 2>", 1]}]}]
   }]
 }
]
display_rate {denominator=1, numerator=60}
tick_resolution {denominator=1, numerator=24000}
tracks [{type=MovieSceneSubTrack, display_name=Subsequences, sections [{type=MovieSceneSubSection, end_frame=9586, start_frame=3597}]}]
