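# Backup VFX Level Sequence for PistolCanary shot spa1000, choreographing a tentacle boss energy attack, beam strikes, monolith spire destruction and exit effects.
# Spawn, attach, transform and Niagara tracks time the ground portal, charge beams, energy ball, lasers and eye beam, whose keyed stop location sweeps between targets.
# Material tracks pulse tentacle glow and flare spire cores before visibility tracks hide spires as death effects play.
# A post-process volume blends in during the charge.
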LevelSequence /PistolCanary/Sequence/spa/spa1000/PistolCanary_spa1000_vfx_backup
bindings [
 PistolCanary_RockSpireMonolith1 {
  id=#1
  possessed_class=PistolCanary_Monolith_Spire_C
  tracks [{
    type=MovieSceneVisibilityTrack
    sections [{type=MovieSceneVisibilitySection, channels [None {type=MovieSceneScriptingBoolChannel, default=false, keys [[3613,true],[8460,false]]}]}]
   }]
 }
 Niagara {
  id=#2
  parent_id=#3
  possessed_class=NiagaraComponent
  tracks [
   {
    type=MovieSceneBoolTrack
    display_name=Auto Activate
    sections [{type=MovieSceneBoolSection, channels [None {type=MovieSceneScriptingBoolChannel, default=false, keys [[7386,false],[7405,true]]}]}]
   }
   {type=MovieSceneNiagaraSystemTrack, sections [{type=MovieSceneNiagaraSystemSpawnSection, end_frame=7759, start_frame=7404}]}
   {
    type=MovieSceneNiagaraFloatParameterTrack
    display_name=Duration
    sections [{type=MovieSceneFloatSection, channels [None {type=MovieSceneScriptingFloatChannel, default=5}]}]
   }
  ]
 }
 NiagaraComponent0 {
  id=#4
  parent_id=#5
  possessed_class=NiagaraComponent
  tracks [
   {
    type=MovieSceneBoolTrack
    display_name=Auto Activate
    sections [{type=MovieSceneBoolSection, channels [None {type=MovieSceneScriptingBoolChannel, default=false, keys [[9181,false],[9182,true]]}]}]
   }
   {type=MovieSceneNiagaraSystemTrack, sections [{type=MovieSceneNiagaraSystemSpawnSection, end_frame=9354, start_frame=9182}]}
  ]
 }
 SM_MaskMaker_Lightning_3_Spirals_Segments2 {
  id=#6
  possessed_class=FortStaticMeshActor
  tracks [{
    type=MovieSceneVisibilityTrack
    sections [{type=MovieSceneVisibilitySection, channels [None {type=MovieSceneScriptingBoolChannel, default=false, keys [[5597,false],[5598,true]]}]}]
   }]
 }
 NS_HumanEyeBeam {
  id=#7
  possessed_class=NiagaraActor
  tracks [
   {
    type=MovieSceneSpawnTrack
    sections [{
      type=MovieSceneSpawnSection
      channels [None {type=MovieSceneScriptingBoolChannel, default=false, keys [[6907,false],[6991,true],[8894,false]]}]
     }]
   }
   {type=MovieScene3DAttachTrack, sections [{type=MovieScene3DAttachSection, end_frame=8894, start_frame=6811}]}
  ]
 }
 NS_PC_AirCharge_Exit {id=#8, possessed_class=NiagaraActor, tracks []}
 PistolCanary_Tentacle3_Laser_Seq3 {
  id=#3
  possessed_class=PistolCanary_BGA_LaserOrigin_C
  tracks [
   {type=MovieScene3DAttachTrack, sections [{type=MovieScene3DAttachSection, end_frame=9338, start_frame=7277}]}
   {
    type=MovieSceneEventTrack
    sections [{
      type=MovieSceneEventTriggerSection
      channels [None {type=MovieSceneScriptingEventChannel, const ["<Struct 'MovieSceneEvent' (0x852fd1abd60) {}>", 1]}]
     }]
   }
   {
    type=MovieSceneSpawnTrack
    sections [{
      type=MovieSceneSpawnSection
      channels [None {type=MovieSceneScriptingBoolChannel, default=false, keys [[7279,false],[7404,true],[7798,false]]}]
     }]
   }
   {
    type=MovieScene3DTransformTrack
    sections [{
      type=MovieScene3DTransformSection
      channels [
       Location.X {type=MovieSceneScriptingDoubleChannel, default=0}
       Location.Y {type=MovieSceneScriptingDoubleChannel, default=0}
       Location.Z {type=MovieSceneScriptingDoubleChannel, default=0}
       Rotation.X {type=MovieSceneScriptingDoubleChannel, default=0}
       Rotation.Y {type=MovieSceneScriptingDoubleChannel, default=0}
       Rotation.Z {type=MovieSceneScriptingDoubleChannel, default=0}
       Scale.X {type=MovieSceneScriptingDoubleChannel, default=1}
       Scale.Y {type=MovieSceneScriptingDoubleChannel, default=1}
       Scale.Z {type=MovieSceneScriptingDoubleChannel, default=1}
      ]
     }]
   }
  ]
 }
 NS_PistolCanary_InnerSpire_Glow_Loop_E {
  id=#9
  possessed_class=NiagaraActor
  tracks [{
    type=MovieSceneVisibilityTrack
    sections [{type=MovieSceneVisibilitySection, channels [None {type=MovieSceneScriptingBoolChannel, default=false, keys [[3613,true],[8166,false]]}]}]
   }]
 }
 NiagaraComponent0 {
  id=#10
  parent_id=#11
  possessed_class=NiagaraComponent
  tracks [
   {
    type=MovieSceneBoolTrack
    display_name=Auto Activate
    sections [{type=MovieSceneBoolSection, channels [None {type=MovieSceneScriptingBoolChannel, default=false, const [false,1]}]}]
   }
   {
    type=MovieSceneParticleTrack
    display_name=FX System
    sections [{
      type=MovieSceneParticleSection
      channels [None {type=MovieSceneScriptingParticleChannel, keys [[6185, "<ParticleKey.ACTIVATE: 0>"], [6756, "<ParticleKey.DEACTIVATE: 1>"]]}]
     }]
   }
   {type=MovieSceneNiagaraSystemTrack, sections [{type=MovieSceneNiagaraSystemSpawnSection, end_frame=6799, start_frame=6199}]}
  ]
 }
 NS_TentacleEnergyBall {
  id=#12
  possessed_class=NiagaraActor
  tracks [
   {
    type=MovieScene3DTransformTrack
    sections [{
      type=MovieScene3DTransformSection
      channels [
       Location.X {
        type=MovieSceneScriptingDoubleChannel
        default=11060
        keys [[5470,605.1021],[5773,-681],[5958,-993.978],[6019,-189.623],[6059,348.2392],[6078,693.5811],[6096,-1419.1627],[6114,-3646.3289],[6155,-1397.2366],[6180,780.6614],[6191,780.6614],[6192,10760],[7302,13348.8994]]
       }
       Location.Y {
        type=MovieSceneScriptingDoubleChannel
        default=80505
        keys [[5470,-384.7639],[5773,11],[5958,284.4412],[6019,-304.148],[6059,675.6165],[6078,674.9985],[6096,285.3013],[6114,-339.6565],[6155,-392.8623],[6180,-623.4841],[6191,-623.4841],[6192,80505],[7302,-3371.6941]]
       }
       Location.Z {
        type=MovieSceneScriptingDoubleChannel
        default=17000
        keys [[5470,-907.5726],[5773,-1455],[5958,-1870.6259],[6019,-448.1126],[6059,532.2547],[6078,-630.0845],[6096,-2276.5613],[6114,-306.4432],[6155,-2148.5942],[6180,-1785.3121],[6191,-1785.3121],[6192,16180],[7302,-8453.8867]]
       }
       Rotation.X {type=MovieSceneScriptingDoubleChannel, default=0}
       Rotation.Y {type=MovieSceneScriptingDoubleChannel, default=0}
       Rotation.Z {type=MovieSceneScriptingDoubleChannel, default=0}
       Scale.X {type=MovieSceneScriptingDoubleChannel, default=1}
       Scale.Y {type=MovieSceneScriptingDoubleChannel, default=1}
       Scale.Z {type=MovieSceneScriptingDoubleChannel, default=1}
      ]
     }]
   }
   {type=MovieScene3DAttachTrack, sections [{type=MovieScene3DAttachSection, end_frame=6192, start_frame=5375}]}
  ]
 }
 StaticMeshComponent0 {
  id=#13
  parent_id=#1
  possessed_class=StaticMeshComponent
  tracks [
   {
    type=MovieSceneComponentMaterialTrack
    display_name=Material Slot: top_FB_001
    sections [{
      type=MovieSceneComponentMaterialParameterSection
      end_frame=9586
      start_frame=3600
      channels [
       Core Intensity {type=MovieSceneScriptingFloatChannel, default=200, keys [[8416,200],[8457,1000]]}
       Core ColorR {type=MovieSceneScriptingFloatChannel, keys [[8416,0.3613],[8457,1]]}
       Core ColorG {type=MovieSceneScriptingFloatChannel, const [0,2]}
       Core ColorB {type=MovieSceneScriptingFloatChannel, keys [[8416,0.3005],[8457,0.0612]]}
       Core ColorA {type=MovieSceneScriptingFloatChannel, const [1,2]}
      ]
     }]
   }
   {
    type=MovieScenePrimitiveMaterialTrack
    display_name=Material Slot: top_FB_001
    sections [{
      type=MovieScenePrimitiveMaterialSection
      channels [None {
        type=MovieSceneScriptingObjectPathChannel
        default="<Object '/Hermes/Rocks/MMSpikeyRocks/Materials/MaskMakerEvent/MI_Hermes_Rocks_MaskMaker_MonolithTop_B_A_Event' (0x852ee93e800) Class 'MaterialInstanceConstant'>"
       }]
     }]
   }
   {
    type=MovieSceneComponentMaterialTrack
    display_name=Material Slot: Bottom_FB
    sections [{
      type=MovieSceneComponentMaterialParameterSection
      end_frame=9586
      start_frame=3600
      channels [
       Core Intensity {type=MovieSceneScriptingFloatChannel, default=200, keys [[8416,200],[8457,1000]]}
       Core ColorR {type=MovieSceneScriptingFloatChannel, keys [[8416,0.3613],[8457,1]]}
       Core ColorG {type=MovieSceneScriptingFloatChannel, const [0,2]}
       Core ColorB {type=MovieSceneScriptingFloatChannel, keys [[8416,0.3005],[8457,0.0612]]}
       Core ColorA {type=MovieSceneScriptingFloatChannel, const [1,2]}
      ]
     }]
   }
   {
    type=MovieScenePrimitiveMaterialTrack
    display_name=Material Slot: Bottom_FB
    sections [{
      type=MovieScenePrimitiveMaterialSection
      channels [None {
        type=MovieSceneScriptingObjectPathChannel
        default="<Object '/Hermes/Rocks/MMSpikeyRocks/Materials/MaskMakerEvent/MI_Hermes_Rocks_MaskMaker_MonolithBottom_A_A_Event' (0x852ee93f800) Class 'MaterialInstanceConstant'>"
       }]
     }]
   }
  ]
 }
 NS_LargeShockwave_2 {
  id=#14
  possessed_class=NiagaraActor
  tracks [
   {
    type=MovieSceneSpawnTrack
    sections [{type=MovieSceneSpawnSection, channels [None {type=MovieSceneScriptingBoolChannel, default=false, keys [[3503,false],[5225,true]]}]}]
   }
   {
    type=MovieScene3DTransformTrack
    sections [{
      type=MovieScene3DTransformSection
      channels [
       Location.X {type=MovieSceneScriptingDoubleChannel, default=15649.0786}
       Location.Y {type=MovieSceneScriptingDoubleChannel, default=68716.0859}
       Location.Z {type=MovieSceneScriptingDoubleChannel, default=8968}
       Rotation.X {type=MovieSceneScriptingDoubleChannel, default=0}
       Rotation.Y {type=MovieSceneScriptingDoubleChannel, default=0}
       Rotation.Z {type=MovieSceneScriptingDoubleChannel, default=0}
       Scale.X {type=MovieSceneScriptingDoubleChannel, default=1}
       Scale.Y {type=MovieSceneScriptingDoubleChannel, default=1}
       Scale.Z {type=MovieSceneScriptingDoubleChannel, default=1}
      ]
     }]
   }
  ]
 }
 NS_PistolCanary_InnerSpire_Glow_Loop_N {
  id=#15
  possessed_class=NiagaraActor
  tracks [{
    type=MovieSceneVisibilityTrack
    sections [{type=MovieSceneVisibilitySection, channels [None {type=MovieSceneScriptingBoolChannel, default=false, keys [[3613,true],[8579,false]]}]}]
   }]
 }
 NS_PistolCanaryCD_Connecting_Spire_Glow_E {
  id=#16
  possessed_class=NiagaraActor
  tracks [{
    type=MovieSceneVisibilityTrack
    sections [{type=MovieSceneVisibilitySection, channels [None {type=MovieSceneScriptingBoolChannel, default=false, const [false,1]}]}]
   }]
 }
 NiagaraComponent0 {
  id=#17
  parent_id=#18
  possessed_class=NiagaraComponent
  tracks [{
    type=MovieSceneNiagaraFloatParameterTrack
    display_name=Charge
    sections [{
      type=MovieSceneFloatSection
      channels [None {type=MovieSceneScriptingFloatChannel, default=2, keys [[5277,0],[5361,0.25],[6148,2],[6175,0]]}]
     }]
   }]
 }
 NS_PistolCanaryCD_Connecting_Spire_Glow_W {
  id=#19
  possessed_class=NiagaraActor
  tracks [{
    type=MovieSceneVisibilityTrack
    sections [{type=MovieSceneVisibilitySection, channels [None {type=MovieSceneScriptingBoolChannel, default=false, const [false,1]}]}]
   }]
 }
 NS_PistolCanary_Portal_Dissolve {id=#20, possessed_class=NiagaraActor, tracks []}
 NiagaraComponent0 {
  id=#21
  parent_id=#20
  possessed_class=NiagaraComponent
  tracks [{type=MovieSceneNiagaraSystemTrack, sections [{type=MovieSceneNiagaraSystemSpawnSection, end_frame=9131, start_frame=8584}]}]
 }
 NiagaraComponent0 {
  id=#22
  parent_id=#8
  possessed_class=NiagaraComponent
  tracks [
   {
    type=MovieSceneBoolTrack
    display_name=Auto Activate
    sections [{type=MovieSceneBoolSection, channels [None {type=MovieSceneScriptingBoolChannel, default=false, keys [[8958,false],[9128,true]]}]}]
   }
   {type=MovieScene3DAttachTrack, sections [{type=MovieScene3DAttachSection, end_frame=9317, start_frame=9096}]}
   {type=MovieSceneNiagaraSystemTrack, sections [{type=MovieSceneNiagaraSystemSpawnSection, end_frame=9295, start_frame=9129}]}
   {
    type=MovieSceneParticleTrack
    display_name=FX System
    sections [{type=MovieSceneParticleSection, channels [None {type=MovieSceneScriptingParticleChannel}]}]
   }
  ]
 }
 NiagaraComponent0 {
  id=#23
  parent_id=#24
  possessed_class=NiagaraComponent
  tracks [{
    type=MovieSceneNiagaraFloatParameterTrack
    display_name=Charge
    sections [{
      type=MovieSceneFloatSection
      channels [None {type=MovieSceneScriptingFloatChannel, default=2, keys [[5277,0],[5361,0.25],[6148,2],[6175,0]]}]
     }]
   }]
 }
 SkeletalMesh {
  id=#25
  parent_id=#26
  possessed_class=SkeletalMeshComponent
  tracks [
   {
    type=MovieSceneComponentMaterialTrack
    display_name=Material Slot: Body_A
    sections [{
      type=MovieSceneComponentMaterialParameterSection
      end_frame=9210
      start_frame=3495
      channels [
       PulseSpeed {type=MovieSceneScriptingFloatChannel, default=0.2, thinned_from=6, keys [[5254,0.1],[5258,0],[6393,0],[6400,0.2]]}
       PulseManualOffset {type=MovieSceneScriptingFloatChannel, default=0, keys [[5324,0],[5655,0.6]]}
       PulseRadius {type=MovieSceneScriptingFloatChannel, default=0.33, keys [[5258,0.33],[6176,1],[6393,0.3]]}
       GlowPulseEXP {type=MovieSceneScriptingFloatChannel, default=0.03, keys [[5572,0.015],[5766,0.03],[6176,0.03],[6393,0.015]]}
       EmissiveGlowColorR {
        type=MovieSceneScriptingFloatChannel
        default=8
        keys [[5215,8],[5254,0],[5261,0],[5324,8],[5877,0.7],[6180,0.7],[6291,8],[6393,0],[6400,0],[6433,8]]
       }
       EmissiveGlowColorG {
        type=MovieSceneScriptingFloatChannel
        default=0.3176
        keys [[5215,0.3176],[5254,0],[5261,0],[5324,0.3176],[5877,3],[6180,3],[6291,0.3176],[6393,0],[6400,0],[6433,0.3176]]
       }
       EmissiveGlowColorB {type=MovieSceneScriptingFloatChannel, default=0, keys [[5254,0],[5324,0],[5877,10],[6180,10],[6291,0]]}
       EmissiveGlowColorA {type=MovieSceneScriptingFloatChannel, default=1}
      ]
     }]
   }
   {
    type=MovieSceneComponentMaterialTrack
    display_name=Material Slot: Body_B
    sections [{
      type=MovieSceneComponentMaterialParameterSection
      end_frame=9315
      start_frame=3600
      channels [
       PulseSpeed {type=MovieSceneScriptingFloatChannel, default=0, thinned_from=6, keys [[5254,0.1],[5258,0],[6393,0],[6400,0.2]]}
       PulseManualOffset {type=MovieSceneScriptingFloatChannel, default=0, keys [[5324,0],[5655,0.6]]}
       PulseRadius {type=MovieSceneScriptingFloatChannel, default=0.33, keys [[5258,0.33],[6176,1],[6393,0.3]]}
       GlowPulseEXP {type=MovieSceneScriptingFloatChannel, default=0.03, keys [[5572,0.015],[5766,0.03],[6176,0.03],[6393,0.015]]}
       EmissiveGlowColorR {
        type=MovieSceneScriptingFloatChannel
        default=8
        keys [[5215,8],[5254,0],[5261,0],[5324,8],[5877,0.7],[6180,0.7],[6291,8],[6393,0],[6400,0],[6433,8]]
       }
       EmissiveGlowColorG {
        type=MovieSceneScriptingFloatChannel
        default=0.3176
        keys [[5215,0.3176],[5254,0],[5261,0],[5324,0.3176],[5877,3],[6180,3],[6291,0.3176],[6393,0],[6400,0],[6433,0.3176]]
       }
       EmissiveGlowColorB {type=MovieSceneScriptingFloatChannel, default=0, keys [[5254,0],[5324,0],[5877,10],[6180,10],[6291,0]]}
       EmissiveGlowColorA {type=MovieSceneScriptingFloatChannel, default=1}
      ]
     }]
   }
   {
    type=MovieSceneComponentMaterialTrack
    display_name=Material Slot: Body_C
    sections [{
      type=MovieSceneComponentMaterialParameterSection
      end_frame=9315
      start_frame=3600
      channels [
       PulseSpeed {type=MovieSceneScriptingFloatChannel, default=0, thinned_from=6, keys [[5254,0.1],[5258,0],[6393,0],[6400,0.2]]}
       PulseManualOffset {type=MovieSceneScriptingFloatChannel, default=0.7, keys [[5324,0],[5655,0.6]]}
       PulseRadius {type=MovieSceneScriptingFloatChannel, default=0.3, keys [[5258,0.33],[6176,1],[6393,0.3]]}
       GlowPulseEXP {type=MovieSceneScriptingFloatChannel, default=0.03, keys [[5572,0.015],[5766,0.03],[6176,0.03],[6393,0.015]]}
       EmissiveGlowColorR {
        type=MovieSceneScriptingFloatChannel
        default=0
        keys [[5215,8],[5254,0],[5261,0],[5324,8],[5877,0.7],[6180,0.7],[6291,8],[6393,0],[6400,0],[6433,8]]
       }
       EmissiveGlowColorG {
        type=MovieSceneScriptingFloatChannel
        default=0
        keys [[5215,0.3176],[5254,0],[5261,0],[5324,0.3176],[5877,3],[6180,3],[6291,0.3176],[6393,0],[6400,0],[6433,0.3176]]
       }
       EmissiveGlowColorB {type=MovieSceneScriptingFloatChannel, default=10, keys [[5254,0],[5324,0],[5877,10],[6180,10],[6291,0]]}
       EmissiveGlowColorA {type=MovieSceneScriptingFloatChannel, default=1}
      ]
     }]
   }
   {
    type=MovieSceneComponentMaterialTrack
    display_name=Material Slot: Body_D
    sections [{
      type=MovieSceneComponentMaterialParameterSection
      end_frame=9315
      start_frame=3600
      channels [
       PulseSpeed {type=MovieSceneScriptingFloatChannel, default=0, thinned_from=6, keys [[5254,0.1],[5258,0],[6393,0],[6400,0.2]]}
       PulseManualOffset {type=MovieSceneScriptingFloatChannel, default=0, keys [[5324,0],[5655,0.6]]}
       PulseRadius {type=MovieSceneScriptingFloatChannel, default=0.3399, keys [[5258,0.33],[6176,1],[6393,0.3]]}
       GlowPulseEXP {type=MovieSceneScriptingFloatChannel, default=0.03, keys [[5572,0.015],[5766,0.03],[6176,0.03],[6393,0.015]]}
       EmissiveGlowColorR {
        type=MovieSceneScriptingFloatChannel
        default=8
        keys [[5215,8],[5254,0],[5261,0],[5324,8],[5877,0.7],[6180,0.7],[6291,8],[6393,0],[6400,0],[6433,8]]
       }
       EmissiveGlowColorG {
        type=MovieSceneScriptingFloatChannel
        default=0.3176
        keys [[5215,0.3176],[5254,0],[5261,0],[5324,0.3176],[5877,3],[6180,3],[6291,0.3176],[6393,0],[6400,0],[6433,0.3176]]
       }
       EmissiveGlowColorB {type=MovieSceneScriptingFloatChannel, default=0, keys [[5254,0],[5324,0],[5877,10],[6180,10],[6291,0]]}
       EmissiveGlowColorA {type=MovieSceneScriptingFloatChannel, default=1}
      ]
     }]
   }
  ]
 }
 NS_TentacleAttackChargeUpLaser3 {
  id=#27
  possessed_class=NiagaraActor
  tracks [
   {
    type=MovieSceneSpawnTrack
    sections [{type=MovieSceneSpawnSection, channels [None {type=MovieSceneScriptingBoolChannel, default=false, const [false,3]}]}]
   }
   {type=MovieScene3DAttachTrack, sections [{type=MovieScene3DAttachSection, end_frame=14157, start_frame=5654}]}
  ]
 }
 NS_PC_WealthLamb_Exit {id=#5, possessed_class=NiagaraActor, tracks []}
 StaticMeshComponent0 {
  id=#28
  parent_id=#29
  possessed_class=StaticMeshComponent
  tracks [
   {
    type=MovieSceneComponentMaterialTrack
    display_name=Material Slot: Bottom_FB
    sections [{
      type=MovieSceneComponentMaterialParameterSection
      end_frame=9586
      start_frame=3600
      channels [
       Core Intensity {type=MovieSceneScriptingFloatChannel, default=200, keys [[8282,200],[8323,1000]]}
       Core ColorR {type=MovieSceneScriptingFloatChannel, keys [[8282,0.3613],[8323,1]]}
       Core ColorG {type=MovieSceneScriptingFloatChannel, const [0,2]}
       Core ColorB {type=MovieSceneScriptingFloatChannel, keys [[8282,0.3005],[8323,0.0612]]}
       Core ColorA {type=MovieSceneScriptingFloatChannel, const [1,2]}
      ]
     }]
   }
   {
    type=MovieScenePrimitiveMaterialTrack
    display_name=Material Slot: Bottom_FB
    sections [{
      type=MovieScenePrimitiveMaterialSection
      channels [None {
        type=MovieSceneScriptingObjectPathChannel
        default="<Object '&MI_Hermes_Rocks_MaskMaker_MonolithBottom_A_A_Event' (0x852ee93f800) Class 'MaterialInstanceConstant'>"
       }]
     }]
   }
   {
    type=MovieSceneComponentMaterialTrack
    display_name=Material Slot: top_FB_001
    sections [{
      type=MovieSceneComponentMaterialParameterSection
      end_frame=9586
      start_frame=3600
      channels [
       Core Intensity {type=MovieSceneScriptingFloatChannel, default=200, keys [[8282,200],[8323,1000]]}
       Core ColorR {type=MovieSceneScriptingFloatChannel, keys [[8282,0.3613],[8323,1]]}
       Core ColorG {type=MovieSceneScriptingFloatChannel, const [0,2]}
       Core ColorB {type=MovieSceneScriptingFloatChannel, keys [[8282,0.3005],[8323,0.0612]]}
       Core ColorA {type=MovieSceneScriptingFloatChannel, const [1,2]}
      ]
     }]
   }
   {
    type=MovieScenePrimitiveMaterialTrack
    display_name=Material Slot: top_FB_001
    sections [{
      type=MovieScenePrimitiveMaterialSection
      channels [None {
        type=MovieSceneScriptingObjectPathChannel
        default="<Object '&MI_Hermes_Rocks_MaskMaker_MonolithTop_B_A_Event' (0x852ee93e800) Class 'MaterialInstanceConstant'>"
       }]
     }]
   }
  ]
 }
 NS_PistolCanary_Tentacle_GroundPortal_01 {
  id=#30
  possessed_class=NiagaraActor
  tracks [{
    type=MovieScene3DTransformTrack
    sections [{
      type=MovieScene3DTransformSection
      channels [
       Location.X {type=MovieSceneScriptingDoubleChannel, default=14328}
       Location.Y {type=MovieSceneScriptingDoubleChannel, default=70128}
       Location.Z {
        type=MovieSceneScriptingDoubleChannel
        default=4500
        keys [[3771,3688],[3930,3688],[4239,3773],[4776,3972],[5003,3972],[5040,4998],[8587,4998],[8886,4500]]
       }
       Scale.X {
        type=MovieSceneScriptingDoubleChannel
        default=1.25
        keys [[3771,0],[3930,1.1],[4239,1.4],[4776,1.1],[5003,1.1],[5040,1.5],[8587,1.5],[8758,1.25],[8865,0]]
       }
       Scale.Y {
        type=MovieSceneScriptingDoubleChannel
        default=1.25
        keys [[3771,0],[3930,1.1],[4239,1.4],[4776,1.1],[5003,1.1],[5040,1.5],[8587,1.5],[8758,1.25],[8865,0]]
       }
       Scale.Z {
        type=MovieSceneScriptingDoubleChannel
        default=1.25
        keys [[3771,0],[3930,1.1],[4239,3],[4776,1.1],[5003,1.1],[5040,1.5],[8587,1.5],[8758,1.25],[8865,0]]
       }
      ]
     }]
   }]
 }
 NS_LargeSpire_DeathFX_E {id=#31, possessed_class=NiagaraActor, tracks []}
 NiagaraComponent0 {
  id=#32
  parent_id=#33
  possessed_class=NiagaraComponent
  tracks [
   {type=MovieSceneNiagaraSystemTrack, sections [{type=MovieSceneNiagaraSystemSpawnSection, end_frame=8665, start_frame=8579}]}
   {
    type=MovieSceneBoolTrack
    display_name=Auto Activate
    sections [{
      type=MovieSceneBoolSection
      channels [None {type=MovieSceneScriptingBoolChannel, default=false, keys [[8558,false],[8579,true],[8642,false]]}]
     }]
   }
  ]
 }
 NiagaraComponent0 {
  id=#34
  parent_id=#31
  possessed_class=NiagaraComponent
  tracks [
   {type=MovieSceneNiagaraSystemTrack, sections [{type=MovieSceneNiagaraSystemSpawnSection, end_frame=8291, start_frame=8166}]}
   {
    type=MovieSceneBoolTrack
    display_name=Auto Activate
    sections [{
      type=MovieSceneBoolSection
      channels [None {type=MovieSceneScriptingBoolChannel, default=false, keys [[8158,false],[8166,true],[8293,false]]}]
     }]
   }
  ]
 }
 NiagaraComponent0 {
  id=#35
  parent_id=#30
  possessed_class=NiagaraComponent
  tracks [
   {type=MovieSceneNiagaraSystemTrack, sections [{type=MovieSceneNiagaraSystemSpawnSection, end_frame=9432, start_frame=3446}]}
   {
    type=MovieSceneNiagaraFloatParameterTrack
    display_name=Dissolve
    sections [{type=MovieSceneFloatSection, channels [None {type=MovieSceneScriptingFloatChannel, default=0, keys [[8668,1],[9208,0]]}]}]
   }
   {
    type=MovieSceneNiagaraFloatParameterTrack
    display_name=DissolveEdge
    sections [{type=MovieSceneFloatSection, channels [None {type=MovieSceneScriptingFloatChannel, default=0, keys [[8520,1],[8922,0]]}]}]
   }
   {
    type=MovieSceneNiagaraFloatParameterTrack
    display_name=SpawnLightning
    sections [{type=MovieSceneFloatSection, channels [None {type=MovieSceneScriptingFloatChannel, default=15, keys [[8260,12.5],[8708,0]]}]}]
   }
   {
    type=MovieSceneNiagaraFloatParameterTrack
    display_name=SizeCenterPortal
    sections [{type=MovieSceneFloatSection, channels [None {type=MovieSceneScriptingFloatChannel, default=0, keys [[3446,1],[3925,0]]}]}]
   }
  ]
 }
 NS_PistolCanary_InnerSpire_Glow_Loop_W {
  id=#36
  possessed_class=NiagaraActor
  tracks [{
    type=MovieSceneVisibilityTrack
    sections [{type=MovieSceneVisibilitySection, channels [None {type=MovieSceneScriptingBoolChannel, default=false, keys [[3613,true],[8460,false]]}]}]
   }]
 }
 NS_PistolCanary_TripleTentacleFloorAttack_Charge2 {
  id=#24
  possessed_class=NiagaraActor
  tracks [
   {
    type=MovieSceneSpawnTrack
    sections [{
      type=MovieSceneSpawnSection
      channels [None {type=MovieSceneScriptingBoolChannel, default=false, keys [[5244,false],[5277,true],[6226,false]]}]
     }]
   }
   {
    type=MovieScene3DTransformTrack
    sections [{
      type=MovieScene3DTransformSection
      channels [
       Location.X {type=MovieSceneScriptingDoubleChannel, default=0}
       Location.Y {type=MovieSceneScriptingDoubleChannel, default=0}
       Location.Z {type=MovieSceneScriptingDoubleChannel, default=0}
       Rotation.X {type=MovieSceneScriptingDoubleChannel, default=0}
       Rotation.Y {type=MovieSceneScriptingDoubleChannel, default=0}
       Rotation.Z {type=MovieSceneScriptingDoubleChannel, default=0}
       Scale.X {type=MovieSceneScriptingDoubleChannel, default=1}
       Scale.Y {type=MovieSceneScriptingDoubleChannel, default=1}
       Scale.Z {type=MovieSceneScriptingDoubleChannel, default=1}
      ]
     }]
   }
   {type=MovieScene3DAttachTrack, sections [{type=MovieScene3DAttachSection, end_frame=8794, start_frame=5225}]}
  ]
 }
 TripleTentacleRight {id=#37, possessed_class=PistolCanary_TentacleBoss_C, tracks []}
 NS_PistolCanary_TripleTentacleFloorAttack_Charge3 {
  id=#18
  possessed_class=NiagaraActor
  tracks [
   {
    type=MovieSceneSpawnTrack
    sections [{
      type=MovieSceneSpawnSection
      channels [None {type=MovieSceneScriptingBoolChannel, default=false, keys [[5244,false],[5277,true],[6226,false]]}]
     }]
   }
   {
    type=MovieScene3DTransformTrack
    sections [{
      type=MovieScene3DTransformSection
      channels [
       Location.X {type=MovieSceneScriptingDoubleChannel, default=0}
       Location.Y {type=MovieSceneScriptingDoubleChannel, default=0}
       Location.Z {type=MovieSceneScriptingDoubleChannel, default=0}
       Rotation.X {type=MovieSceneScriptingDoubleChannel, default=0}
       Rotation.Y {type=MovieSceneScriptingDoubleChannel, default=0}
       Rotation.Z {type=MovieSceneScriptingDoubleChannel, default=0}
       Scale.X {type=MovieSceneScriptingDoubleChannel, default=1}
       Scale.Y {type=MovieSceneScriptingDoubleChannel, default=1}
       Scale.Z {type=MovieSceneScriptingDoubleChannel, default=1}
      ]
     }]
   }
   {type=MovieScene3DAttachTrack, sections [{type=MovieScene3DAttachSection, end_frame=8794, start_frame=5225}]}
  ]
 }
 NS_TentacleAttackChargeUpLaser {
  id=#38
  possessed_class=NiagaraActor
  tracks [
   {
    type=MovieSceneSpawnTrack
    sections [{type=MovieSceneSpawnSection, channels [None {type=MovieSceneScriptingBoolChannel, default=false, const [false,4]}]}]
   }
   {type=MovieScene3DAttachTrack, sections [{type=MovieScene3DAttachSection, end_frame=14157, start_frame=5654}]}
  ]
 }
 PPV_PistolCanary_VFX {
  id=#39
  possessed_class=PostProcessVolume
  tracks [
   {
    type=MovieSceneFloatTrack
    display_name=Blend Weight
    sections [{type=MovieSceneFloatSection, channels [None {type=MovieSceneScriptingFloatChannel, default=0, keys [[5426,0],[5501,1],[6214,1],[6251,0]]}]}]
   }
   {
    type=MovieSceneFloatTrack
    display_name="Intensity (Settings)"
    sections [{type=MovieSceneFloatSection, channels [None {type=MovieSceneScriptingFloatChannel, default=2, keys [[5433,0.675],[5988,2]]}]}]
   }
   {
    type=MovieSceneFloatTrack
    display_name="Intensity (Settings)"
    sections [{type=MovieSceneFloatSection, channels [None {type=MovieSceneScriptingFloatChannel, default=3, keys [[5431,1.5],[5993,3]]}]}]
   }
  ]
 }
 NiagaraComponent0 {
  id=#40
  parent_id=#41
  possessed_class=NiagaraComponent
  tracks [
   {
    type=MovieSceneNiagaraVectorParameterTrack
    display_name=Stop Location
    sections [{
      type=MovieSceneFloatVectorSection
      channels [
       Vector.X {type=MovieSceneScriptingFloatChannel, default=14122, keys [[5463,14347.3086],[5711,14122],[5946,13308]]}
       Vector.Y {type=MovieSceneScriptingFloatChannel, default=70824, keys [[5463,69706],[5711,70824],[5946,72057]]}
       Vector.Z {type=MovieSceneScriptingFloatChannel, default=12300, keys [[5463,12300],[5711,13309],[5946,13436]]}
      ]
     }]
   }
   {type=MovieSceneNiagaraSystemTrack, sections [{type=MovieSceneNiagaraSystemSpawnSection, end_frame=6022, start_frame=5422}]}
   {
    type=MovieSceneNiagaraFloatParameterTrack
    display_name=BeamImpactScale
    sections [{type=MovieSceneFloatSection, channels [None {type=MovieSceneScriptingFloatChannel, default=2}]}]
   }
   {
    type=MovieSceneNiagaraFloatParameterTrack
    display_name=BeamWidth
    sections [{type=MovieSceneFloatSection, channels [None {type=MovieSceneScriptingFloatChannel, default=2}]}]
   }
   {
    type=MovieSceneNiagaraFloatParameterTrack
    display_name=Duration
    sections [{type=MovieSceneFloatSection, channels [None {type=MovieSceneScriptingFloatChannel, default=10}]}]
   }
   {
    type=MovieSceneParticleTrack
    display_name=FX System
    sections [{type=MovieSceneParticleSection, channels [None {type=MovieSceneScriptingParticleChannel, const ["<ParticleKey.ACTIVATE: 0>", 1]}]}]
   }
  ]
 }
 NiagaraComponent0 {
  id=#42
  parent_id=#43
  possessed_class=NiagaraComponent
  tracks [{
    type=MovieSceneNiagaraFloatParameterTrack
    display_name=Charge
    sections [{
      type=MovieSceneFloatSection
      channels [None {type=MovieSceneScriptingFloatChannel, default=2, keys [[5277,0],[5361,0.25],[6148,2],[6175,0]]}]
     }]
   }]
 }
 TripleTentacleFront {id=#26, possessed_class=PistolCanary_TentacleBoss_C, tracks []}
 NS_PistolCanaryCD_Connecting_Spire_Glow_S {
  id=#44
  possessed_class=NiagaraActor
  tracks [{
    type=MovieSceneVisibilityTrack
    sections [{type=MovieSceneVisibilitySection, channels [None {type=MovieSceneScriptingBoolChannel, default=false, const [false,1]}]}]
   }]
 }
 NS_TentacleEnergyBallDissipate2 {
  id=#45
  possessed_class=NiagaraActor
  tracks [
   {
    type=MovieScene3DTransformTrack
    sections [{
      type=MovieScene3DTransformSection
      channels [
       Location.X {type=MovieSceneScriptingDoubleChannel, default=0}
       Location.Y {type=MovieSceneScriptingDoubleChannel, default=0}
       Location.Z {type=MovieSceneScriptingDoubleChannel, default=0}
       Rotation.X {type=MovieSceneScriptingDoubleChannel, default=0, const [0,1]}
       Rotation.Y {type=MovieSceneScriptingDoubleChannel, default=0, const [0,1]}
       Rotation.Z {type=MovieSceneScriptingDoubleChannel, default=0, const [0,1]}
       Scale.X {type=MovieSceneScriptingDoubleChannel, default=0.2, const [1,1]}
       Scale.Y {type=MovieSceneScriptingDoubleChannel, default=0.2, const [1,1]}
       Scale.Z {type=MovieSceneScriptingDoubleChannel, default=1, const [1,1]}
      ]
     }]
   }
   {type=MovieScene3DAttachTrack, sections [{type=MovieScene3DAttachSection, end_frame=9586, start_frame=5937}]}
  ]
 }
 StaticMeshComponent0 {
  id=#46
  parent_id=#47
  possessed_class=StaticMeshComponent
  tracks [
   {
    type=MovieSceneComponentMaterialTrack
    display_name=Material Slot: top_FB_001
    sections [{
      type=MovieSceneComponentMaterialParameterSection
      end_frame=9586
      start_frame=3600
      channels [
       Core Intensity {type=MovieSceneScriptingFloatChannel, default=200, keys [[8534,200],[8575,1000]]}
       Core ColorR {type=MovieSceneScriptingFloatChannel, keys [[8534,0.3613],[8575,1]]}
       Core ColorG {type=MovieSceneScriptingFloatChannel, const [0,2]}
       Core ColorB {type=MovieSceneScriptingFloatChannel, keys [[8534,0.3005],[8575,0.0612]]}
       Core ColorA {type=MovieSceneScriptingFloatChannel, const [1,2]}
      ]
     }]
   }
   {
    type=MovieScenePrimitiveMaterialTrack
    display_name=Material Slot: top_FB_001
    sections [{
      type=MovieScenePrimitiveMaterialSection
      channels [None {
        type=MovieSceneScriptingObjectPathChannel
        default="<Object '&MI_Hermes_Rocks_MaskMaker_MonolithTop_B_A_Event' (0x852ee93e800) Class 'MaterialInstanceConstant'>"
       }]
     }]
   }
   {
    type=MovieSceneComponentMaterialTrack
    display_name=Material Slot: Bottom_FB
    sections [{
      type=MovieSceneComponentMaterialParameterSection
      end_frame=9586
      start_frame=3600
      channels [
       Core Intensity {type=MovieSceneScriptingFloatChannel, default=200, keys [[8534,200],[8575,1000]]}
       Core ColorR {type=MovieSceneScriptingFloatChannel, keys [[8534,0.3613],[8575,1]]}
       Core ColorG {type=MovieSceneScriptingFloatChannel, const [0,2]}
       Core ColorB {type=MovieSceneScriptingFloatChannel, keys [[8534,0.3005],[8575,0.0612]]}
       Core ColorA {type=MovieSceneScriptingFloatChannel, const [1,2]}
      ]
     }]
   }
   {
    type=MovieScenePrimitiveMaterialTrack
    display_name=Material Slot: Bottom_FB
    sections [{
      type=MovieScenePrimitiveMaterialSection
      channels [None {
        type=MovieSceneScriptingObjectPathChannel
        default="<Object '&MI_Hermes_Rocks_MaskMaker_MonolithBottom_A_A_Event' (0x852ee93f800) Class 'MaterialInstanceConstant'>"
       }]
     }]
   }
  ]
 }
 NiagaraComponent0 {
  id=#48
  parent_id=#7
  possessed_class=NiagaraComponent
  tracks [
   {
    type=MovieSceneNiagaraVectorParameterTrack
    display_name=Stop Location
    sections [{
      type=MovieSceneFloatVectorSection
      channels [
       Vector.X {
        type=MovieSceneScriptingFloatChannel
        default=10421
        keys [[7018,10421],[8058,10928.6064],[8124,26148.9941],[8242,26148.9941],[8276,14540.3877],[8425,14540.3877],[8426,1745.8026],[8456,1745.8026],[8550,14295.3955]]
       }
       Vector.Y {
        type=MovieSceneScriptingFloatChannel
        default=71235
        keys [[7018,71235],[8058,71390.0547],[8124,69656.6719],[8125,70355.2578],[8242,69656.6719],[8243,70355.2578],[8276,86682.6172],[8425,86682.6172],[8426,70526.4375],[8456,70526.4375],[8550,56967.4258]]
       }
       Vector.Z {
        type=MovieSceneScriptingFloatChannel
        default=10121
        keys [[7018,10121],[8058,11000],[8124,7660.5454],[8125,6586.1606],[8242,7660.5454],[8243,6586.1606],[8276,6756.8145],[8425,6756.8145],[8426,6420],[8456,6420],[8550,5867.0913]]
       }
      ]
     }]
   }
   {
    type=MovieSceneNiagaraFloatParameterTrack
    display_name=WidthMult
    sections [{type=MovieSceneFloatSection, channels [None {type=MovieSceneScriptingFloatChannel, default=1}]}]
   }
   {
    type=MovieSceneNiagaraFloatParameterTrack
    display_name=BeamProgress
    sections [{
      type=MovieSceneFloatSection
      channels [None {
        type=MovieSceneScriptingFloatChannel
        default=0
        thinned_from=19
        keys [[6976,0],[6991,1],[7232,1],[7246,0],[8111,0],[8123,1],[8179,1],[8180,0],[8289,0],[8292,1],[8326,1],[8327,0],[8426,0],[8430,1],[8449,1],[8452,0],[8550,0],[8554,1]]
       }]
     }]
   }
   {
    type=MovieSceneNiagaraFloatParameterTrack
    display_name=BeamWidth
    sections [{
      type=MovieSceneFloatSection
      channels [None {type=MovieSceneScriptingFloatChannel, default=0, thinned_from=11, keys [[7214,1],[7228,0],[8111,0],[8113,1],[8575,1],[8578,0]]}]
     }]
   }
   {
    type=MovieSceneNiagaraFloatParameterTrack
    display_name=Firing
    sections [{
      type=MovieSceneFloatSection
      channels [None {
        type=MovieSceneScriptingFloatChannel
        default=0
        thinned_from=21
        keys [[6988,0],[6989,1],[7214,1],[7228,0],[8121,0],[8122,1],[8166,1],[8170,0],[8291,0],[8292,1],[8326,1],[8429,0],[8430,1],[8443,1],[8449,0],[8553,0],[8554,1],[8573,1],[8576,0]]
       }]
     }]
   }
   {
    type=MovieSceneParticleTrack
    display_name=FX System
    sections [{
      type=MovieSceneParticleSection
      channels [None {
        type=MovieSceneScriptingParticleChannel
        keys [
         [6991, "<ParticleKey.ACTIVATE: 0>"]
         [7230, "<ParticleKey.DEACTIVATE: 1>"]
         [8125, "<ParticleKey.ACTIVATE: 0>"]
         [8158, "<ParticleKey.DEACTIVATE: 1>"]
         [8289, "<ParticleKey.ACTIVATE: 0>"]
         [8319, "<ParticleKey.DEACTIVATE: 1>"]
         [8426, "<ParticleKey.ACTIVATE: 0>"]
         [8456, "<ParticleKey.DEACTIVATE: 1>"]
         [8550, "<ParticleKey.ACTIVATE: 0>"]
         [8576, "<ParticleKey.DEACTIVATE: 1>"]
        ]
       }]
     }]
   }
   {type=MovieSceneNiagaraSystemTrack, sections [{type=MovieSceneNiagaraSystemSpawnSection, end_frame=7566, start_frame=6966}]}
  ]
 }
 NS_PS_WealthLamb_Exit_SkyFlare {id=#49, possessed_class=NiagaraActor, tracks []}
 PistolCanary_RockSpireMonolith2 {
  id=#47
  possessed_class=PistolCanary_Monolith_Spire_C
  tracks [{
    type=MovieSceneVisibilityTrack
    sections [{type=MovieSceneVisibilitySection, channels [None {type=MovieSceneScriptingBoolChannel, default=false, keys [[3613,true],[8579,false]]}]}]
   }]
 }
 NS_PC_EnergyBall_Charge_Beam_Tentacle3 {
  id=#50
  possessed_class=NiagaraActor
  tracks [
   {
    type=MovieScene3DTransformTrack
    sections [{
      type=MovieScene3DTransformSection
      channels [
       Location.X {type=MovieSceneScriptingDoubleChannel, default=0}
       Location.Y {type=MovieSceneScriptingDoubleChannel, default=0}
       Location.Z {type=MovieSceneScriptingDoubleChannel, default=0}
       Rotation.X {type=MovieSceneScriptingDoubleChannel, default=0}
       Rotation.Y {type=MovieSceneScriptingDoubleChannel, default=0}
       Rotation.Z {type=MovieSceneScriptingDoubleChannel, default=0}
       Scale.X {type=MovieSceneScriptingDoubleChannel, default=1}
       Scale.Y {type=MovieSceneScriptingDoubleChannel, default=1}
       Scale.Z {type=MovieSceneScriptingDoubleChannel, default=1}
      ]
     }]
   }
   {type=MovieScene3DAttachTrack, sections [{type=MovieScene3DAttachSection, end_frame=9455, start_frame=5415}]}
   {
    type=MovieSceneSpawnTrack
    sections [{type=MovieSceneSpawnSection, channels [None {type=MovieSceneScriptingBoolChannel, default=true, keys [[5411,false],[5419,true]]}]}]
   }
  ]
 }
 NS_LargeSpire_DeathFX_W {
  id=#51
  possessed_class=NiagaraActor
  tracks [{
    type=MovieScene3DTransformTrack
    sections [{
      type=MovieScene3DTransformSection
      channels [
       Location.X {type=MovieSceneScriptingDoubleChannel, default=1638.1653}
       Location.Y {type=MovieSceneScriptingDoubleChannel, default=70412.4252}
       Location.Z {type=MovieSceneScriptingDoubleChannel, default=7128}
       Rotation.X {type=MovieSceneScriptingDoubleChannel, default=0}
       Rotation.Y {type=MovieSceneScriptingDoubleChannel, default=0}
       Rotation.Z {type=MovieSceneScriptingDoubleChannel, default=0}
       Scale.X {type=MovieSceneScriptingDoubleChannel, default=1}
       Scale.Y {type=MovieSceneScriptingDoubleChannel, default=3}
       Scale.Z {type=MovieSceneScriptingDoubleChannel, default=3}
      ]
     }]
   }]
 }
 NiagaraComponent0 {
  id=#52
  parent_id=#45
  possessed_class=NiagaraComponent
  tracks [
   {
    type=MovieSceneNiagaraFloatParameterTrack
    display_name=Scale
    sections [{type=MovieSceneFloatSection, channels [None {type=MovieSceneScriptingFloatChannel, default=1.5, keys [[6202,1.5],[6219,0]]}]}]
   }
   {type=MovieSceneNiagaraSystemTrack, sections [{type=MovieSceneNiagaraSystemSpawnSection, end_frame=6802, start_frame=6202}]}
  ]
 }
 NS_IcicleLarva_Aura {
  id=#53
  possessed_class=NiagaraActor
  tracks [
   {
    type=MovieSceneSpawnTrack
    sections [{
      type=MovieSceneSpawnSection
      channels [None {type=MovieSceneScriptingBoolChannel, default=false, keys [[4907,false],[6052,true],[6595,false],[7184,true],[9299,false]]}]
     }]
   }
   {type=MovieScene3DAttachTrack, sections [{type=MovieScene3DAttachSection, end_frame=13355, start_frame=6042}]}
  ]
 }
 NiagaraComponent0 {
  id=#54
  parent_id=#55
  possessed_class=NiagaraComponent
  tracks [
   {type=MovieSceneNiagaraSystemTrack, sections [{type=MovieSceneNiagaraSystemSpawnSection, end_frame=8420, start_frame=8323}]}
   {
    type=MovieSceneBoolTrack
    display_name=Auto Activate
    sections [{
      type=MovieSceneBoolSection
      channels [None {type=MovieSceneScriptingBoolChannel, default=false, keys [[8291,false],[8323,true],[8464,false]]}]
     }]
   }
  ]
 }
 NS_PistolCanary_InnerSpire_Glow_Loop_S {
  id=#56
  possessed_class=NiagaraActor
  tracks [{
    type=MovieSceneVisibilityTrack
    sections [{type=MovieSceneVisibilitySection, channels [None {type=MovieSceneScriptingBoolChannel, default=false, keys [[3613,true],[8323,false]]}]}]
   }]
 }
 NS_PC_EnergyBall_Charge_Beam_Tentacle2 {
  id=#41
  possessed_class=NiagaraActor
  tracks [
   {
    type=MovieScene3DTransformTrack
    sections [{
      type=MovieScene3DTransformSection
      channels [
       Location.X {type=MovieSceneScriptingDoubleChannel, default=0}
       Location.Y {type=MovieSceneScriptingDoubleChannel, default=0}
       Location.Z {type=MovieSceneScriptingDoubleChannel, default=0}
       Rotation.X {type=MovieSceneScriptingDoubleChannel, default=0}
       Rotation.Y {type=MovieSceneScriptingDoubleChannel, default=0}
       Rotation.Z {type=MovieSceneScriptingDoubleChannel, default=0}
       Scale.X {type=MovieSceneScriptingDoubleChannel, default=1}
       Scale.Y {type=MovieSceneScriptingDoubleChannel, default=1}
       Scale.Z {type=MovieSceneScriptingDoubleChannel, default=1}
      ]
     }]
   }
   {type=MovieScene3DAttachTrack, sections [{type=MovieScene3DAttachSection, end_frame=9455, start_frame=5376}]}
   {
    type=MovieSceneSpawnTrack
    sections [{type=MovieSceneSpawnSection, channels [None {type=MovieSceneScriptingBoolChannel, default=false, keys [[5411,false],[5419,true]]}]}]
   }
  ]
 }
 NiagaraComponent0 {
  id=#57
  parent_id=#50
  possessed_class=NiagaraComponent
  tracks [
   {type=MovieSceneNiagaraSystemTrack, sections [{type=MovieSceneNiagaraSystemSpawnSection, end_frame=6022, start_frame=5422}]}
   {
    type=MovieSceneNiagaraFloatParameterTrack
    display_name=BeamImpactScale
    sections [{type=MovieSceneFloatSection, channels [None {type=MovieSceneScriptingFloatChannel, default=2}]}]
   }
   {
    type=MovieSceneNiagaraFloatParameterTrack
    display_name=BeamWidth
    sections [{type=MovieSceneFloatSection, channels [None {type=MovieSceneScriptingFloatChannel, default=2}]}]
   }
   {
    type=MovieSceneNiagaraVectorParameterTrack
    display_name=Stop Location
    sections [{
      type=MovieSceneFloatVectorSection
      channels [
       Vector.X {type=MovieSceneScriptingFloatChannel, default=14122, keys [[5463,14347.3086],[5711,14122],[5946,13308]]}
       Vector.Y {type=MovieSceneScriptingFloatChannel, default=70824, keys [[5463,69706],[5711,70824],[5946,72057]]}
       Vector.Z {type=MovieSceneScriptingFloatChannel, default=12300, keys [[5463,12300],[5711,13309],[5946,13436]]}
      ]
     }]
   }
   {
    type=MovieSceneParticleTrack
    display_name=FX System
    sections [{type=MovieSceneParticleSection, channels [None {type=MovieSceneScriptingParticleChannel, const ["<ParticleKey.ACTIVATE: 0>", 1]}]}]
   }
  ]
 }
 PistolCanary_RockSpireMonolith3 {
  id=#29
  possessed_class=PistolCanary_Monolith_Spire_C
  tracks [{
    type=MovieSceneVisibilityTrack
    sections [{type=MovieSceneVisibilitySection, channels [None {type=MovieSceneScriptingBoolChannel, default=false, keys [[3613,true],[8323,false]]}]}]
   }]
 }
 NS_TentacleAttackChargeUpLaser2 {
  id=#58
  possessed_class=NiagaraActor
  tracks [
   {
    type=MovieSceneSpawnTrack
    sections [{type=MovieSceneSpawnSection, channels [None {type=MovieSceneScriptingBoolChannel, default=false, const [false,3]}]}]
   }
   {type=MovieScene3DAttachTrack, sections [{type=MovieScene3DAttachSection, end_frame=14157, start_frame=5654}]}
  ]
 }
 SkeletalMesh {
  id=#59
  parent_id=#37
  possessed_class=SkeletalMeshComponent
  tracks [
   {
    type=MovieSceneComponentMaterialTrack
    display_name=Material Slot: Body_A
    sections [{
      type=MovieSceneComponentMaterialParameterSection
      end_frame=9210
      start_frame=3495
      channels [
       PulseSpeed {type=MovieSceneScriptingFloatChannel, default=0.2, thinned_from=6, keys [[5254,0.1],[5258,0],[6393,0],[6400,0.2]]}
       PulseManualOffset {type=MovieSceneScriptingFloatChannel, default=0, keys [[5324,0],[5655,0.6]]}
       PulseRadius {type=MovieSceneScriptingFloatChannel, default=0.33, keys [[5258,0.33],[6176,1],[6393,0.3]]}
       GlowPulseEXP {type=MovieSceneScriptingFloatChannel, default=0.03, keys [[5572,0.015],[5766,0.03],[6176,0.03],[6393,0.015]]}
       EmissiveGlowColorR {
        type=MovieSceneScriptingFloatChannel
        default=8
        keys [[5215,8],[5254,0],[5261,0],[5324,8],[5877,0.7],[6180,0.7],[6291,8],[6393,0],[6400,0],[6433,8]]
       }
       EmissiveGlowColorG {
        type=MovieSceneScriptingFloatChannel
        default=0.3176
        keys [[5215,0.3176],[5254,0],[5261,0],[5324,0.3176],[5877,3],[6180,3],[6291,0.3176],[6393,0],[6400,0],[6433,0.3176]]
       }
       EmissiveGlowColorB {type=MovieSceneScriptingFloatChannel, default=0, keys [[5254,0],[5324,0],[5877,10],[6180,10],[6291,0]]}
       EmissiveGlowColorA {type=MovieSceneScriptingFloatChannel, default=1}
      ]
     }]
   }
   {
    type=MovieSceneComponentMaterialTrack
    display_name=Material Slot: Body_B
    sections [{
      type=MovieSceneComponentMaterialParameterSection
      end_frame=9315
      start_frame=3600
      channels [
       PulseSpeed {type=MovieSceneScriptingFloatChannel, default=0, thinned_from=6, keys [[5254,0.1],[5258,0],[6393,0],[6400,0.2]]}
       PulseManualOffset {type=MovieSceneScriptingFloatChannel, default=0, keys [[5324,0],[5655,0.6]]}
       PulseRadius {type=MovieSceneScriptingFloatChannel, default=0.33, keys [[5258,0.33],[6176,1],[6393,0.3]]}
       GlowPulseEXP {type=MovieSceneScriptingFloatChannel, default=0.03, keys [[5572,0.015],[5766,0.03],[6176,0.03],[6393,0.015]]}
       EmissiveGlowColorR {
        type=MovieSceneScriptingFloatChannel
        default=8
        keys [[5215,8],[5254,0],[5261,0],[5324,8],[5877,0.7],[6180,0.7],[6291,8],[6393,0],[6400,0],[6433,8]]
       }
       EmissiveGlowColorG {
        type=MovieSceneScriptingFloatChannel
        default=0.3176
        keys [[5215,0.3176],[5254,0],[5261,0],[5324,0.3176],[5877,3],[6180,3],[6291,0.3176],[6393,0],[6400,0],[6433,0.3176]]
       }
       EmissiveGlowColorB {type=MovieSceneScriptingFloatChannel, default=0, keys [[5254,0],[5324,0],[5877,10],[6180,10],[6291,0]]}
       EmissiveGlowColorA {type=MovieSceneScriptingFloatChannel, default=1}
      ]
     }]
   }
   {
    type=MovieSceneComponentMaterialTrack
    display_name=Material Slot: Body_C
    sections [{
      type=MovieSceneComponentMaterialParameterSection
      end_frame=9315
      start_frame=3600
      channels [
       PulseSpeed {type=MovieSceneScriptingFloatChannel, default=0, thinned_from=6, keys [[5254,0.1],[5258,0],[6393,0],[6400,0.2]]}
       PulseManualOffset {type=MovieSceneScriptingFloatChannel, default=0.7, keys [[5324,0],[5655,0.6]]}
       PulseRadius {type=MovieSceneScriptingFloatChannel, default=0.3, keys [[5258,0.33],[6176,1],[6393,0.3]]}
       GlowPulseEXP {type=MovieSceneScriptingFloatChannel, default=0.03, keys [[5572,0.015],[5766,0.03],[6176,0.03],[6393,0.015]]}
       EmissiveGlowColorR {
        type=MovieSceneScriptingFloatChannel
        default=0
        keys [[5215,8],[5254,0],[5261,0],[5324,8],[5877,0.7],[6180,0.7],[6291,8],[6393,0],[6400,0],[6433,8]]
       }
       EmissiveGlowColorG {
        type=MovieSceneScriptingFloatChannel
        default=0
        keys [[5215,0.3176],[5254,0],[5261,0],[5324,0.3176],[5877,3],[6180,3],[6291,0.3176],[6393,0],[6400,0],[6433,0.3176]]
       }
       EmissiveGlowColorB {type=MovieSceneScriptingFloatChannel, default=10, keys [[5254,0],[5324,0],[5877,10],[6180,10],[6291,0]]}
       EmissiveGlowColorA {type=MovieSceneScriptingFloatChannel, default=1}
      ]
     }]
   }
   {
    type=MovieSceneComponentMaterialTrack
    display_name=Material Slot: Body_D
    sections [{
      type=MovieSceneComponentMaterialParameterSection
      end_frame=9315
      start_frame=3600
      channels [
       PulseSpeed {type=MovieSceneScriptingFloatChannel, default=0, thinned_from=6, keys [[5254,0.1],[5258,0],[6393,0],[6400,0.2]]}
       PulseManualOffset {type=MovieSceneScriptingFloatChannel, default=0, keys [[5324,0],[5655,0.6]]}
       PulseRadius {type=MovieSceneScriptingFloatChannel, default=0.3399, keys [[5258,0.33],[6176,1],[6393,0.3]]}
       GlowPulseEXP {type=MovieSceneScriptingFloatChannel, default=0.03, keys [[5572,0.015],[5766,0.03],[6176,0.03],[6393,0.015]]}
       EmissiveGlowColorR {
        type=MovieSceneScriptingFloatChannel
        default=8
        keys [[5215,8],[5254,0],[5261,0],[5324,8],[5877,0.7],[6180,0.7],[6291,8],[6393,0],[6400,0],[6433,8]]
       }
       EmissiveGlowColorG {
        type=MovieSceneScriptingFloatChannel
        default=0.3176
        keys [[5215,0.3176],[5254,0],[5261,0],[5324,0.3176],[5877,3],[6180,3],[6291,0.3176],[6393,0],[6400,0],[6433,0.3176]]
       }
       EmissiveGlowColorB {type=MovieSceneScriptingFloatChannel, default=0, keys [[5254,0],[5324,0],[5877,10],[6180,10],[6291,0]]}
       EmissiveGlowColorA {type=MovieSceneScriptingFloatChannel, default=1}
      ]
     }]
   }
  ]
 }
 PistolCanary_RockSpireMonolith4 {
  id=#60
  possessed_class=PistolCanary_Monolith_Spire_C
  tracks [{
    type=MovieSceneVisibilityTrack
    sections [{type=MovieSceneVisibilitySection, channels [None {type=MovieSceneScriptingBoolChannel, default=false, keys [[3613,true],[8166,false]]}]}]
   }]
 }
 NS_PistolCanaryCD_Connecting_Spire_Glow_N {
  id=#61
  possessed_class=NiagaraActor
  tracks [{
    type=MovieSceneVisibilityTrack
    sections [{type=MovieSceneVisibilitySection, channels [None {type=MovieSceneScriptingBoolChannel, default=false, const [false,1]}]}]
   }]
 }
 KalEl {id=#62, possessed_class=PistolCanary_CinematicNPC_Superhero_Main_C, tracks []}
 NS_PistolCanary_TripleTentacleWave {
  id=#63
  possessed_class=NiagaraActor
  tracks [{
    type=MovieScene3DTransformTrack
    sections [{
      type=MovieScene3DTransformSection
      channels [
       Location.X {type=MovieSceneScriptingDoubleChannel, default=14360.2255, const [14240.2256,1]}
       Location.Y {type=MovieSceneScriptingDoubleChannel, default=81825.2175, const [70241.2188,1]}
       Location.Z {type=MovieSceneScriptingDoubleChannel, default=4050.9363, const [4050.9363,1]}
       Rotation.X {type=MovieSceneScriptingDoubleChannel, default=0, const [0,1]}
       Rotation.Y {type=MovieSceneScriptingDoubleChannel, default=0, const [0,1]}
       Rotation.Z {type=MovieSceneScriptingDoubleChannel, default=0, const [0,1]}
       Scale.X {type=MovieSceneScriptingDoubleChannel, default=1, const [1,1]}
       Scale.Y {type=MovieSceneScriptingDoubleChannel, default=1, const [1,1]}
       Scale.Z {type=MovieSceneScriptingDoubleChannel, default=1, const [1,1]}
      ]
     }]
   }]
 }
 NiagaraComponent0 {
  id=#64
  parent_id=#12
  possessed_class=NiagaraComponent
  tracks [
   {
    type=MovieSceneNiagaraFloatParameterTrack
    display_name=Scale
    sections [{
      type=MovieSceneFloatSection
      channels [None {type=MovieSceneScriptingFloatChannel, default=0.7, keys [[5428,0],[5444,0.2],[6076,0.5],[6198,0.7],[6204,0]]}]
     }]
   }
   {
    type=MovieSceneNiagaraFloatParameterTrack
    display_name=PostProcess
    sections [{
      type=MovieSceneFloatSection
      channels [None {type=MovieSceneScriptingFloatChannel, default=0, keys [[5505,0],[6073,0.5],[6159,0.5],[6198,1],[6203,0]]}]
     }]
   }
   {
    type=MovieSceneNiagaraFloatParameterTrack
    display_name=LightRays
    sections [{type=MovieSceneFloatSection, channels [None {type=MovieSceneScriptingFloatChannel, default=4, keys [[6070,0],[6165,4],[6182,0]]}]}]
   }
   {type=MovieSceneNiagaraSystemTrack, sections [{type=MovieSceneNiagaraSystemSpawnSection, end_frame=6252, start_frame=5415}]}
  ]
 }
 NS_PistolCanary_TripleTentacleFloorAttack_Charge {
  id=#43
  possessed_class=NiagaraActor
  tracks [
   {
    type=MovieSceneSpawnTrack
    sections [{
      type=MovieSceneSpawnSection
      channels [None {type=MovieSceneScriptingBoolChannel, default=false, keys [[5244,false],[5277,true],[6226,false]]}]
     }]
   }
   {
    type=MovieScene3DTransformTrack
    sections [{
      type=MovieScene3DTransformSection
      channels [
       Location.X {type=MovieSceneScriptingDoubleChannel, default=168.3642}
       Location.Y {type=MovieSceneScriptingDoubleChannel, default=1.5394}
       Location.Z {type=MovieSceneScriptingDoubleChannel, default=-128.8221}
       Rotation.X {type=MovieSceneScriptingDoubleChannel, default=0}
       Rotation.Y {type=MovieSceneScriptingDoubleChannel, default=0}
       Rotation.Z {type=MovieSceneScriptingDoubleChannel, default=0}
       Scale.X {type=MovieSceneScriptingDoubleChannel, default=1}
       Scale.Y {type=MovieSceneScriptingDoubleChannel, default=1}
       Scale.Z {type=MovieSceneScriptingDoubleChannel, default=1}
      ]
     }]
   }
   {type=MovieScene3DAttachTrack, sections [{type=MovieScene3DAttachSection, end_frame=8794, start_frame=5225}]}
  ]
 }
 NiagaraComponent0 {
  id=#65
  parent_id=#51
  possessed_class=NiagaraComponent
  tracks [
   {type=MovieSceneNiagaraSystemTrack, sections [{type=MovieSceneNiagaraSystemSpawnSection, end_frame=8585, start_frame=8460}]}
   {
    type=MovieSceneBoolTrack
    display_name=Auto Activate
    sections [{
      type=MovieSceneBoolSection
      channels [None {type=MovieSceneScriptingBoolChannel, default=false, keys [[8436,false],[8460,true],[8558,false]]}]
     }]
   }
  ]
 }
 NiagaraComponent0 {
  id=#66
  parent_id=#63
  possessed_class=NiagaraComponent
  tracks [
   {
    type=MovieSceneBoolTrack
    display_name=Auto Activate
    sections [{
      type=MovieSceneBoolSection
      channels [None {type=MovieSceneScriptingBoolChannel, default=false, keys [[4966,false],[5002,true],[5902,false]]}]
     }]
   }
   {type=MovieSceneNiagaraSystemTrack, sections [{type=MovieSceneNiagaraSystemSpawnSection, end_frame=5623, start_frame=5002}]}
  ]
 }
 NiagaraComponent0 {
  id=#67
  parent_id=#68
  possessed_class=NiagaraComponent
  tracks [
   {type=MovieSceneNiagaraSystemTrack, sections [{type=MovieSceneNiagaraSystemSpawnSection, end_frame=6223, start_frame=6126}]}
   {
    type=MovieSceneBoolTrack
    display_name=Auto Activate
    sections [{type=MovieSceneBoolSection, channels [None {type=MovieSceneScriptingBoolChannel, default=false, keys [[6123,false],[6126,true]]}]}]
   }
  ]
 }
 NS_PC_AirCharge {
  id=#68
  possessed_class=NiagaraActor
  tracks [{type=MovieScene3DAttachTrack, sections [{type=MovieScene3DAttachSection, end_frame=9287, start_frame=6126}]}]
 }
 NS_LargeSpire_DeathFX_N {
  id=#33
  possessed_class=NiagaraActor
  tracks [{
    type=MovieScene3DTransformTrack
    sections [{
      type=MovieScene3DTransformSection
      channels [
       Location.X {type=MovieSceneScriptingDoubleChannel, default=14321.0786}
       Location.Y {type=MovieSceneScriptingDoubleChannel, default=57056.0859}
       Location.Z {type=MovieSceneScriptingDoubleChannel, default=7128}
       Rotation.X {type=MovieSceneScriptingDoubleChannel, default=0}
       Rotation.Y {type=MovieSceneScriptingDoubleChannel, default=0}
       Rotation.Z {type=MovieSceneScriptingDoubleChannel, default=0}
       Scale.X {type=MovieSceneScriptingDoubleChannel, default=3}
       Scale.Y {type=MovieSceneScriptingDoubleChannel, default=4}
       Scale.Z {type=MovieSceneScriptingDoubleChannel, default=4}
      ]
     }]
   }]
 }
 TripleTentacleLeft {id=#69, possessed_class=PistolCanary_TentacleBoss_C, tracks []}
 Niagara {
  id=#70
  parent_id=#62
  possessed_class=NiagaraComponent
  tracks [
   {type=MovieSceneNiagaraSystemTrack, sections [{type=MovieSceneNiagaraSystemSpawnSection, end_frame=9039, start_frame=6047}]}
   {
    type=MovieSceneParticleTrack
    display_name=FX System
    sections [{
      type=MovieSceneParticleSection
      channels [None {
        type=MovieSceneScriptingParticleChannel
        keys [
         [6047, "<ParticleKey.ACTIVATE: 0>"]
         [6537, "<ParticleKey.DEACTIVATE: 1>"]
         [6739, "<ParticleKey.ACTIVATE: 0>"]
         [8273, "<ParticleKey.DEACTIVATE: 1>"]
         [8584, "<ParticleKey.ACTIVATE: 0>"]
        ]
       }]
     }]
   }
   {
    type=MovieSceneNiagaraFloatParameterTrack
    display_name=ContrailWidth
    sections [{type=MovieSceneFloatSection, channels [None {type=MovieSceneScriptingFloatChannel, default=5, keys [[6053,1],[6102,5],[6208,5],[6347,1]]}]}]
   }
   {type=MovieScene3DAttachTrack, sections [{type=MovieScene3DAttachSection, end_frame=9091, start_frame=6040}]}
  ]
 }
 NS_PC_TentacleSlice {
  id=#11
  possessed_class=NiagaraActor
  tracks [
   {
    type=MovieScene3DTransformTrack
    sections [{
      type=MovieScene3DTransformSection
      channels [
       Location.X {type=MovieSceneScriptingDoubleChannel, default=531.8238}
       Location.Y {type=MovieSceneScriptingDoubleChannel, default=2.9705}
       Location.Z {type=MovieSceneScriptingDoubleChannel, default=-137.7855}
       Rotation.X {type=MovieSceneScriptingDoubleChannel, default=10.6673}
       Rotation.Y {type=MovieSceneScriptingDoubleChannel, default=-89.136}
       Rotation.Z {type=MovieSceneScriptingDoubleChannel, default=-9.7945}
       Scale.X {type=MovieSceneScriptingDoubleChannel, default=1}
       Scale.Y {type=MovieSceneScriptingDoubleChannel, default=1}
       Scale.Z {type=MovieSceneScriptingDoubleChannel, default=1}
      ]
     }]
   }
   {type=MovieScene3DAttachTrack, sections [{type=MovieScene3DAttachSection, end_frame=9286, start_frame=6198}]}
  ]
 }
 SkeletalMesh {
  id=#71
  parent_id=#69
  possessed_class=SkeletalMeshComponent
  tracks [
   {
    type=MovieSceneComponentMaterialTrack
    display_name=Material Slot: Body_A
    sections [{
      type=MovieSceneComponentMaterialParameterSection
      end_frame=9210
      start_frame=3495
      channels [
       PulseSpeed {type=MovieSceneScriptingFloatChannel, default=0.2, thinned_from=6, keys [[5254,0.1],[5258,0],[6393,0],[6400,0.2]]}
       PulseManualOffset {type=MovieSceneScriptingFloatChannel, default=0, keys [[5324,0],[5655,0.6]]}
       PulseRadius {type=MovieSceneScriptingFloatChannel, default=0.33, keys [[5258,0.33],[6176,1],[6393,0.3]]}
       GlowPulseEXP {type=MovieSceneScriptingFloatChannel, default=0.03, keys [[5572,0.015],[5766,0.03],[6176,0.03],[6393,0.015]]}
       EmissiveGlowColorR {
        type=MovieSceneScriptingFloatChannel
        default=8
        keys [[5215,8],[5254,0],[5261,0],[5324,8],[5877,0.7],[6180,0.7],[6291,8],[6393,0],[6400,0],[6433,8]]
       }
       EmissiveGlowColorG {
        type=MovieSceneScriptingFloatChannel
        default=0.3176
        keys [[5215,0.3176],[5254,0],[5261,0],[5324,0.3176],[5877,3],[6180,3],[6291,0.3176],[6393,0],[6400,0],[6433,0.3176]]
       }
       EmissiveGlowColorB {type=MovieSceneScriptingFloatChannel, default=0, keys [[5254,0],[5324,0],[5877,10],[6180,10],[6291,0]]}
       EmissiveGlowColorA {type=MovieSceneScriptingFloatChannel, default=1}
      ]
     }]
   }
   {
    type=MovieSceneComponentMaterialTrack
    display_name=Material Slot: Body_B
    sections [{
      type=MovieSceneComponentMaterialParameterSection
      end_frame=9315
      start_frame=3600
      channels [
       PulseSpeed {type=MovieSceneScriptingFloatChannel, default=0, thinned_from=6, keys [[5254,0.1],[5258,0],[6393,0],[6400,0.2]]}
       PulseManualOffset {type=MovieSceneScriptingFloatChannel, default=0, keys [[5324,0],[5655,0.6]]}
       PulseRadius {type=MovieSceneScriptingFloatChannel, default=0.33, keys [[5258,0.33],[6176,1],[6393,0.3]]}
       GlowPulseEXP {type=MovieSceneScriptingFloatChannel, default=0.03, keys [[5572,0.015],[5766,0.03],[6176,0.03],[6393,0.015]]}
       EmissiveGlowColorR {
        type=MovieSceneScriptingFloatChannel
        default=8
        keys [[5215,8],[5254,0],[5261,0],[5324,8],[5877,0.7],[6180,0.7],[6291,8],[6393,0],[6400,0],[6433,8]]
       }
       EmissiveGlowColorG {
        type=MovieSceneScriptingFloatChannel
        default=0.3176
        keys [[5215,0.3176],[5254,0],[5261,0],[5324,0.3176],[5877,3],[6180,3],[6291,0.3176],[6393,0],[6400,0],[6433,0.3176]]
       }
       EmissiveGlowColorB {type=MovieSceneScriptingFloatChannel, default=0, keys [[5254,0],[5324,0],[5877,10],[6180,10],[6291,0]]}
       EmissiveGlowColorA {type=MovieSceneScriptingFloatChannel, default=1}
      ]
     }]
   }
   {
    type=MovieSceneComponentMaterialTrack
    display_name=Material Slot: Body_C
    sections [{
      type=MovieSceneComponentMaterialParameterSection
      end_frame=9315
      start_frame=3600
      channels [
       PulseSpeed {type=MovieSceneScriptingFloatChannel, default=0, thinned_from=6, keys [[5254,0.1],[5258,0],[6393,0],[6400,0.2]]}
       PulseManualOffset {type=MovieSceneScriptingFloatChannel, default=0.7, keys [[5324,0],[5655,0.6]]}
       PulseRadius {type=MovieSceneScriptingFloatChannel, default=0.3, keys [[5258,0.33],[6176,1],[6393,0.3]]}
       GlowPulseEXP {type=MovieSceneScriptingFloatChannel, default=0.03, keys [[5572,0.015],[5766,0.03],[6176,0.03],[6393,0.015]]}
       EmissiveGlowColorR {
        type=MovieSceneScriptingFloatChannel
        default=0
        keys [[5215,8],[5254,0],[5261,0],[5324,8],[5877,0.7],[6180,0.7],[6291,8],[6393,0],[6400,0],[6433,8]]
       }
       EmissiveGlowColorG {
        type=MovieSceneScriptingFloatChannel
        default=0
        keys [[5215,0.3176],[5254,0],[5261,0],[5324,0.3176],[5877,3],[6180,3],[6291,0.3176],[6393,0],[6400,0],[6433,0.3176]]
       }
       EmissiveGlowColorB {type=MovieSceneScriptingFloatChannel, default=10, keys [[5254,0],[5324,0],[5877,10],[6180,10],[6291,0]]}
       EmissiveGlowColorA {type=MovieSceneScriptingFloatChannel, default=1}
      ]
     }]
   }
   {
    type=MovieSceneComponentMaterialTrack
    display_name=Material Slot: Body_D
    sections [{
      type=MovieSceneComponentMaterialParameterSection
      end_frame=9315
      start_frame=3600
      channels [
       PulseSpeed {type=MovieSceneScriptingFloatChannel, default=0, thinned_from=6, keys [[5254,0.1],[5258,0],[6393,0],[6400,0.2]]}
       PulseManualOffset {type=MovieSceneScriptingFloatChannel, default=0, keys [[5324,0],[5655,0.6]]}
       PulseRadius {type=MovieSceneScriptingFloatChannel, default=0.3399, keys [[5258,0.33],[6176,1],[6393,0.3]]}
       GlowPulseEXP {type=MovieSceneScriptingFloatChannel, default=0.03, keys [[5572,0.015],[5766,0.03],[6176,0.03],[6393,0.015]]}
       EmissiveGlowColorR {
        type=MovieSceneScriptingFloatChannel
        default=8
        keys [[5215,8],[5254,0],[5261,0],[5324,8],[5877,0.7],[6180,0.7],[6291,8],[6393,0],[6400,0],[6433,8]]
       }
       EmissiveGlowColorG {
        type=MovieSceneScriptingFloatChannel
        default=0.3176
        keys [[5215,0.3176],[5254,0],[5261,0],[5324,0.3176],[5877,3],[6180,3],[6291,0.3176],[6393,0],[6400,0],[6433,0.3176]]
       }
       EmissiveGlowColorB {type=MovieSceneScriptingFloatChannel, default=0, keys [[5254,0],[5324,0],[5877,10],[6180,10],[6291,0]]}
       EmissiveGlowColorA {type=MovieSceneScriptingFloatChannel, default=1}
      ]
     }]
   }
  ]
 }
 NS_PistolCanary_Tentacle_Emerge {
  id=#72
  possessed_class=NiagaraActor
  tracks [
   {
    type=MovieSceneSpawnTrack
    sections [{
      type=MovieSceneSpawnSection
      channels [None {
        type=MovieSceneScriptingBoolChannel
        default=false
        keys [[3644,false],[3952,true],[4065,false],[4136,true],[4345,false],[4430,true],[4570,false]]
       }]
     }]
   }
   {
    type=MovieScene3DTransformTrack
    sections [{
      type=MovieScene3DTransformSection
      channels [
       Location.X {type=MovieSceneScriptingDoubleChannel, default=14564, const [14328,1]}
       Location.Y {type=MovieSceneScriptingDoubleChannel, default=70128, const [70128,1]}
       Location.Z {type=MovieSceneScriptingDoubleChannel, default=3788, const [3896,1]}
       Rotation.X {type=MovieSceneScriptingDoubleChannel, default=0, const [0,1]}
       Rotation.Y {type=MovieSceneScriptingDoubleChannel, default=0, const [0,1]}
       Rotation.Z {type=MovieSceneScriptingDoubleChannel, default=0, const [0,1]}
       Scale.X {type=MovieSceneScriptingDoubleChannel, default=1, const [1,1]}
       Scale.Y {type=MovieSceneScriptingDoubleChannel, default=1, const [1,1]}
       Scale.Z {type=MovieSceneScriptingDoubleChannel, default=1, const [1,1]}
      ]
     }]
   }
  ]
 }
 StaticMeshComponent0 {
  id=#73
  parent_id=#60
  possessed_class=StaticMeshComponent
  tracks [
   {
    type=MovieScenePrimitiveMaterialTrack
    display_name=Material Slot: Bottom_FB
    sections [{
      type=MovieScenePrimitiveMaterialSection
      channels [None {
        type=MovieSceneScriptingObjectPathChannel
        default="<Object '&MI_Hermes_Rocks_MaskMaker_MonolithBottom_A_A_Event' (0x852ee93f800) Class 'MaterialInstanceConstant'>"
       }]
     }]
   }
   {
    type=MovieScenePrimitiveMaterialTrack
    display_name=Material Slot: top_FB_001
    sections [{
      type=MovieScenePrimitiveMaterialSection
      channels [None {
        type=MovieSceneScriptingObjectPathChannel
        default="<Object '&MI_Hermes_Rocks_MaskMaker_MonolithTop_B_A_Event' (0x852ee93e800) Class 'MaterialInstanceConstant'>"
       }]
     }]
   }
   {
    type=MovieSceneComponentMaterialTrack
    display_name=Material Slot: Bottom_FB
    sections [{
      type=MovieSceneComponentMaterialParameterSection
      end_frame=9586
      start_frame=3600
      channels [
       Core Intensity {type=MovieSceneScriptingFloatChannel, default=200, keys [[8121,200],[8162,1000]]}
       Core ColorR {type=MovieSceneScriptingFloatChannel, keys [[8121,0.3613],[8162,1]]}
       Core ColorG {type=MovieSceneScriptingFloatChannel, const [0,2]}
       Core ColorB {type=MovieSceneScriptingFloatChannel, keys [[8121,0.3005],[8162,0.0612]]}
       Core ColorA {type=MovieSceneScriptingFloatChannel, const [1,2]}
      ]
     }]
   }
   {
    type=MovieSceneComponentMaterialTrack
    display_name=Material Slot: top_FB_001
    sections [{
      type=MovieSceneComponentMaterialParameterSection
      end_frame=9586
      start_frame=3600
      channels [
       Core Intensity {type=MovieSceneScriptingFloatChannel, default=200, keys [[8121,200],[8162,1000]]}
       Core ColorR {type=MovieSceneScriptingFloatChannel, keys [[8121,0.3613],[8162,1]]}
       Core ColorG {type=MovieSceneScriptingFloatChannel, const [0,2]}
       Core ColorB {type=MovieSceneScriptingFloatChannel, keys [[8121,0.3005],[8162,0.0612]]}
       Core ColorA {type=MovieSceneScriptingFloatChannel, const [1,2]}
      ]
     }]
   }
  ]
 }
 NS_LargeSpire_DeathFX_S {
  id=#55
  possessed_class=NiagaraActor
  tracks [{
    type=MovieScene3DTransformTrack
    sections [{
      type=MovieScene3DTransformSection
      channels [
       Location.X {type=MovieSceneScriptingDoubleChannel, default=14357.4929}
       Location.Y {type=MovieSceneScriptingDoubleChannel, default=86757.8885}
       Location.Z {type=MovieSceneScriptingDoubleChannel, default=7128}
       Rotation.X {type=MovieSceneScriptingDoubleChannel, default=0}
       Rotation.Y {type=MovieSceneScriptingDoubleChannel, default=0}
       Rotation.Z {type=MovieSceneScriptingDoubleChannel, default=0}
       Scale.X {type=MovieSceneScriptingDoubleChannel, default=1}
       Scale.Y {type=MovieSceneScriptingDoubleChannel, default=3}
       Scale.Z {type=MovieSceneScriptingDoubleChannel, default=3}
      ]
     }]
   }]
 }
 NiagaraComponent0 {
  id=#74
  parent_id=#49
  possessed_class=NiagaraComponent
  tracks [{
    type=MovieSceneParticleTrack
    display_name=FX System
    sections [{type=MovieSceneParticleSection, channels [None {type=MovieSceneScriptingParticleChannel, const ["<ParticleKey.ACTIVATE: 0>", 1]}]}]
   }]
 }
]
display_rate {denominator=1, numerator=60}
tick_resolution {denominator=1, numerator=24000}
tracks []
