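# VFX Level Sequence for the spa1000 sequence that animates a tentacle ground portal opening and closing.
# It possesses three placed Niagara actors: single and triple portal variants sharing scale and parameter curves, plus a dissolve effect.
# Visibility and activation keys switch from the single to the triple variant early on.
# Closing ramps down lightning, dissolve and edge parameters while scale shrinks to zero and the separate dissolve effect triggers, then deactivates.
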LevelSequence /PistolCanary/Sequence/spa/spa1000/PistolCanary_spa1000_vfx_portal
bindings [
 NS_PistolCanary_Tentacle_GroundPortal_01_Triple {
  id=#1
  possessed_class=NiagaraActor
  tracks [{
    type=MovieScene3DTransformTrack
    sections [{
      type=MovieScene3DTransformSection
      channels [
       Location.X {type=MovieSceneScriptingDoubleChannel, default=14328, const [14328,1]}
       Location.Y {type=MovieSceneScriptingDoubleChannel, default=70128, const [70128,1]}
       Location.Z {type=MovieSceneScriptingDoubleChannel, default=3800, const [3800,3]}
       Scale.X {type=MovieSceneScriptingDoubleChannel, default=1.5, keys [[3771,0],[4002,1.7],[8462,1.7],[8633,1.25],[8701,0]]}
       Scale.Y {type=MovieSceneScriptingDoubleChannel, default=1.7, keys [[3771,0],[4002,1.7],[8462,1.7],[8633,1.25],[8701,0]]}
       Scale.Z {type=MovieSceneScriptingDoubleChannel, default=1.7, keys [[3771,0],[4002,1.7],[8462,1.7],[8633,1.25],[8701,0]]}
      ]
     }]
   }]
 }
 NS_PistolCanary_Portal_Dissolve {
  id=#2
  possessed_class=NiagaraActor
  tracks [{
    type=MovieScene3DTransformTrack
    sections [{
      type=MovieScene3DTransformSection
      channels [
       Location.X {type=MovieSceneScriptingDoubleChannel, default=14328, const [14328,1]}
       Location.Y {type=MovieSceneScriptingDoubleChannel, default=70128, const [70128,1]}
       Location.Z {type=MovieSceneScriptingDoubleChannel, default=3500, const [3500,1]}
       Rotation.X {type=MovieSceneScriptingDoubleChannel, default=0}
       Rotation.Y {type=MovieSceneScriptingDoubleChannel, default=0}
       Rotation.Z {type=MovieSceneScriptingDoubleChannel, default=0}
       Scale.X {type=MovieSceneScriptingDoubleChannel, default=1}
       Scale.Y {type=MovieSceneScriptingDoubleChannel, default=1}
       Scale.Z {type=MovieSceneScriptingDoubleChannel, default=1}
      ]
     }]
   }]
 }
 NiagaraComponent0 {
  id=#3
  parent_id=#2
  possessed_class=NiagaraComponent
  tracks [{
    type=MovieSceneParticleTrack
    display_name=FX System
    sections [{
      type=MovieSceneParticleSection
      channels [None {type=MovieSceneScriptingParticleChannel, keys [[8552, "<ParticleKey.TRIGGER: 2>"], [9122, "<ParticleKey.DEACTIVATE: 1>"]]}]
     }]
   }]
 }
 NiagaraComponent0 {
  id=#4
  parent_id=#1
  possessed_class=NiagaraComponent
  tracks [
   {
    type=MovieSceneNiagaraFloatParameterTrack
    display_name=Dissolve
    sections [{type=MovieSceneFloatSection, channels [None {type=MovieSceneScriptingFloatChannel, default=0, keys [[8668,1],[9208,0]]}]}]
   }
   {
    type=MovieSceneNiagaraFloatParameterTrack
    display_name=DissolveEdge
    sections [{type=MovieSceneFloatSection, channels [None {type=MovieSceneScriptingFloatChannel, default=0, keys [[8520,1],[8922,0]]}]}]
   }
   {
    type=MovieSceneNiagaraFloatParameterTrack
    display_name=SpawnLightning
    sections [{type=MovieSceneFloatSection, channels [None {type=MovieSceneScriptingFloatChannel, default=15, keys [[8260,12.5],[8708,0]]}]}]
   }
   {
    type=MovieSceneNiagaraFloatParameterTrack
    display_name=SizeCenterPortal
    sections [{type=MovieSceneFloatSection, channels [None {type=MovieSceneScriptingFloatChannel, default=1.2, keys [[3446,1],[3925,1.2]]}]}]
   }
   {
    type=MovieSceneBoolTrack
    display_name=Auto Activate
    sections [{type=MovieSceneBoolSection, channels [None {type=MovieSceneScriptingBoolChannel, default=false, keys [[3597,false],[3606,true]]}]}]
   }
   {
    type=MovieSceneParticleTrack
    display_name=FX System
    sections [{type=MovieSceneParticleSection, channels [None {type=MovieSceneScriptingParticleChannel, const ["<ParticleKey.ACTIVATE: 0>", 1]}]}]
   }
   {type=MovieSceneNiagaraSystemTrack, sections [{type=MovieSceneNiagaraSystemSpawnSection, end_frame=4365, start_frame=3606}]}
  ]
 }
 NS_PistolCanary_Tentacle_GroundPortal_01 {
  id=#5
  possessed_class=NiagaraActor
  tracks [
   {
    type=MovieScene3DTransformTrack
    sections [{
      type=MovieScene3DTransformSection
      channels [
       Location.X {type=MovieSceneScriptingDoubleChannel, default=14328}
       Location.Y {type=MovieSceneScriptingDoubleChannel, default=70128}
       Location.Z {type=MovieSceneScriptingDoubleChannel, default=3800, const [3800,2]}
       Scale.X {type=MovieSceneScriptingDoubleChannel, default=1.5, keys [[3771,0],[4002,1.7],[8462,1.7],[8633,1.25],[8701,0]]}
       Scale.Y {type=MovieSceneScriptingDoubleChannel, default=1.7, keys [[3771,0],[4002,1.7],[8462,1.7],[8633,1.25],[8701,0]]}
       Scale.Z {type=MovieSceneScriptingDoubleChannel, default=1.7, keys [[3771,0],[4002,1.7],[8462,1.7],[8633,1.25],[8701,0]]}
      ]
     }]
   }
   {
    type=MovieSceneVisibilityTrack
    sections [{
      type=MovieSceneVisibilitySection
      channels [None {type=MovieSceneScriptingBoolChannel, default=false, keys [[3264,true],[3600,false],[3970,false]]}]
     }]
   }
  ]
 }
 NiagaraComponent0 {
  id=#6
  parent_id=#5
  possessed_class=NiagaraComponent
  tracks [
   {
    type=MovieSceneNiagaraFloatParameterTrack
    display_name=Dissolve
    sections [{type=MovieSceneFloatSection, channels [None {type=MovieSceneScriptingFloatChannel, default=0, keys [[8668,1],[9208,0]]}]}]
   }
   {
    type=MovieSceneNiagaraFloatParameterTrack
    display_name=DissolveEdge
    sections [{type=MovieSceneFloatSection, channels [None {type=MovieSceneScriptingFloatChannel, default=0, keys [[8520,1],[8922,0]]}]}]
   }
   {
    type=MovieSceneNiagaraFloatParameterTrack
    display_name=SpawnLightning
    sections [{type=MovieSceneFloatSection, channels [None {type=MovieSceneScriptingFloatChannel, default=15, keys [[8260,12.5],[8708,0]]}]}]
   }
   {
    type=MovieSceneNiagaraFloatParameterTrack
    display_name=SizeCenterPortal
    sections [{type=MovieSceneFloatSection, channels [None {type=MovieSceneScriptingFloatChannel, default=1.2, keys [[3446,1],[3925,1.2]]}]}]
   }
  ]
 }
]
display_rate {denominator=1, numerator=60}
tick_resolution {denominator=1, numerator=24000}
tracks []
