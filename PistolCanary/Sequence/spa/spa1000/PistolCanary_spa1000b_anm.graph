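# Level Sequence for cinematic shot spa1000b that animates KalEl, a scaled-up superhero cinematic NPC, in a fast pass along X.
# Transform keys sweep X from positive to far negative while Y and Z dip mid-pass and return, with rotation held constant.
# Three overlapping skeletal animation sections drive the body, and visibility switches on early in the move.
# A child Niagara component's LifetimeScale stays raised through the pass, then drops near the end.
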LevelSequence /PistolCanary/Sequence/spa/spa1000/PistolCanary_spa1000b_anm
bindings [
 KalEl {
  id=#1
  possessed_class=PistolCanary_CinematicNPC_Superhero_Main_C
  tracks [
   {
    type=MovieSceneSkeletalAnimationTrack
    sections [
     {type=MovieSceneSkeletalAnimationSection, end_frame=3720, start_frame=3600}
     {type=MovieSceneSkeletalAnimationSection, end_frame=3818, start_frame=3698}
     {type=MovieSceneSkeletalAnimationSection, end_frame=3923, start_frame=3803}
    ]
   }
   {
    type=MovieSceneVisibilityTrack
    sections [{
      type=MovieSceneVisibilitySection
      channels [None {type=MovieSceneScriptingBoolChannel, default=true, keys [[3604,false],[3626,true],[3834,true]]}]
     }]
   }
   {
    type=MovieScene3DTransformTrack
    sections [{
      type=MovieScene3DTransformSection
      end_frame=7541
      channels [
       Location.X {type=MovieSceneScriptingDoubleChannel, default=47, keys [[3624,41556.4883],[3675,47],[3751,-66032.5462]]}
       Location.Y {type=MovieSceneScriptingDoubleChannel, default=-650, keys [[3600,-91.0372],[3675,-650],[3792,-91.0382]]}
       Location.Z {type=MovieSceneScriptingDoubleChannel, default=5200, keys [[3600,6632.5981],[3675,5200],[3792,6632.5981]]}
       Rotation.X {type=MovieSceneScriptingDoubleChannel, default=9.7172, const [9.7172,3]}
       Rotation.Y {type=MovieSceneScriptingDoubleChannel, default=6.0492, const [6.0492,3]}
       Rotation.Z {type=MovieSceneScriptingDoubleChannel, default=447.918, const [447.918,3]}
       Scale.X {type=MovieSceneScriptingDoubleChannel, default=3.4654}
       Scale.Y {type=MovieSceneScriptingDoubleChannel, default=3.4654}
       Scale.Z {type=MovieSceneScriptingDoubleChannel, default=3.4654}
      ]
     }]
   }
  ]
 }
 Niagara {
  id=#2
  parent_id=#1
  possessed_class=NiagaraComponent
  tracks [{
    type=MovieSceneNiagaraFloatParameterTrack
    display_name=LifetimeScale
    sections [{type=MovieSceneFloatSection, channels [None {type=MovieSceneScriptingFloatChannel, default=15, keys [[3619,15],[3854,15],[3887,1]]}]}]
   }]
 }
]
display_rate {denominator=1, numerator=60}
tick_resolution {denominator=1, numerator=24000}
tracks []
